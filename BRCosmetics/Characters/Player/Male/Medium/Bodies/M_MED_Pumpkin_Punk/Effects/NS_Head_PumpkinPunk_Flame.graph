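# Niagara effect for the Pumpkin Punk character's head: flames, glow particles and flickering eye and mouth lights, tracking the skeletal mesh.
# Flame ribbons and Glow sprites alternate between two sockets, with speed-based drag and kill spheres keeping them close during movement.
# Eye and mouth lights use layered flicker curves and time-of-day color, with mouth spawning also gated by time of day.
# bisHeadHidden zeroes ribbon width and sets other emitters' visibility tags.
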NiagaraSystem /BRCosmetics/Characters/Player/Male/Medium/Bodies/M_MED_Pumpkin_Punk/Effects/NS_Head_PumpkinPunk_Flame
dependencies {
 usedDataInterfaces [/Script/Niagara.NiagaraDataInterfaceSimpleCounter, &NiagaraDataInterfaceCurve, &NiagaraDataInterfaceColorCurve]
 usedDynamicInputs [&RandomRangeLinearColor]
 usedModules [&SystemState]
}
emitters {
 Flame {
  emitterUpdateScript [
   &EmitterState()
   &SpawnRate(SpawnRate=75, Use Spawn Probability=false)
  ]
  particleSpawnScript [
   &InitializeRibbon(
    Color=<rgba 10 1.75 0.2 0.3>
    Lifetime=&RandomRangeFloat(Minimum=0.2, Maximum=0.25)
    Module.Write Color=true
    Module.Write Lifetime=true
    Module.Write Mass=true
    Module.Write Material Random=true
    Module.Write Position=true
    Module.Write RibbonTwist=false
    Module.Write RibbonWidth=true
    Ribbon Width=&RandomRangeFloat(Minimum=1, Maximum=2.5)
   )
   /SetVariables(Particles.SpriteSize=&Vector2DFromFloat(Value=1)) {enabled=false, isSetParameters=true}
   /SetVariables(Particles.AlternateSockets:NiagaraInt32=&ModuloInt(
     A=/Niagara/DynamicInputs/DataInterface/GenerateIntFromCounter(SimpleCounter:NiagaraDataInterfaceSimpleCounter=DataInterface() {propertyValues {GpuSyncMode=None, InitialValue=0}})
     B=2
    )) {isSetParameters=true}
   &SkeletalMeshLocation(
    Bone / Socket Index=$Particles.AlternateSockets
    Inherit Velocity=true
    MaskSpawnGroup=false
    Mesh Sampling Type=::NewEnumerator5
    Sampled Position Offset Coordinate Space:ENiagaraMeshTransforms=::NewEnumerator3
    Skeletal Mesh=$User.SkeletalMesh
    Socket Sampling Mode=::NewEnumerator5
    VelocitySpeedLimit=false
   )
   /SetVariables(Particles.RibbonLinkOrder:NiagaraFloat=&MakeFloatFromInt(Int=&GenerateIntFromCounter(SimpleCounter:NiagaraDataInterfaceSimpleCounter=DataInterface() {propertyValues {GpuSyncMode=None, InitialValue=0}}))) {isSetParameters=true}
  ]
  particleUpdateScript [
   &ParticleState()
   &SkeletalMeshLocation(
    Apply To Particle Position:NiagaraBool=&SetBoolByFloatComparison(A=$Particles.NormalizedAge, B=0.2, Comparison Type=::NewEnumerator6)
    Bone / Socket Index=$Particles.Initial.AlternateSockets
    Inherit Velocity=false
    Interpolate Mesh Sampling=false
    MaskSpawnGroup=false
    Mesh Sampling Type=::NewEnumerator5
    Skeletal Mesh=$User.SkeletalMesh
    Socket Sampling Mode=::NewEnumerator5
   ) {module=SkeletalMeshLocation001}
   &GravityForce(Gravity=&RandomRangeVector(Minimum=<0 0 100>, Maximum=<1 1 250>))
   /Niagara/Modules/Update/Forces/CurlNoiseForce(Noise Strength=25, Pan Noise=false, Random Seed:NiagaraInt32=&RandomRangeInt(Minimum=0, Maximum=100))
   &Drag(
    Drag=&Multiply_Float(A=/Niagara/DynamicInputs/TypeConversions/VectorLengthSafe(Vector:Vector3f=$Output.SkeletalMeshLocation001.SampledVelocity), B=0.9)
    Use Linear Drag=true
    Use Rotational Drag=false
   )
   &SolveForcesAndVelocity(Clamp Velocity=false, Limit Acceleration=false)
   &ScaleColor(
    Scale Alpha=&FloatFromCurve(FloatCurve=DataInterface() {propertyValues {Curve {keys [<at 0 1 RCIM_Cubic>, {arriveTangent=-1, interpMode=RCIM_Cubic, leaveTangent=-1, time=0.2, value=1}, <at 1 0 RCIM_Cubic>]}}})
    ScaleA=true
    ScaleRGB=true
   )
   /SetVariables(Particles.DistanceScale:NiagaraFloat=&Add_Float(
     A=1
     B=&ClampFloat(
      Float=&Multiply_Float(
       A=-0.1
       B=&VectorLength(Vector=&Subtract_Vector(
         A=Linked() {linkedVariable {name=Particles.Position, type=NiagaraPosition}}
         B=Linked() {linkedVariable {name=Particles.Initial.Position, type=NiagaraPosition}}
        ))
      )
      Min=-0.6
      Max=0
     )
    )) {isSetParameters=true}
   &KillParticlesInVolume(
    Invert Volume=true
    Origin Offset:Vector3f=&Multiply_VectorByFloat(Vector=$Output.SkeletalMeshLocation001.SampledVelocity, Float=0.01)
    Sphere Radius=7
    Volume Origin:Vector3f=$Output.SkeletalMeshLocation001.SampledPosition
   )
   &ScaleRibbonWidth(Ribbon Width Scale=&FloatFromCurve(
     FloatCurve=DataInterface() {propertyValues {Curve {keys [
         {arriveTangent=6.0445, interpMode=RCIM_Cubic, leaveTangent=6.0445, tangentMode=RCTM_User, time=0, value=0.5}
         {arriveTangent=-0.5, interpMode=RCIM_Cubic, leaveTangent=-0.5, time=0.2036, value=1}
         <at 1 0>
        ]}}}
     Scale Curve=&MakeCustomFloatFromBool(
      Bool=$User.bisHeadHidden
      True Float=0
      False Float=&MakeCustomFloatFromBool(
       Bool=&SetBoolByFloatComparison(A=$Particles.DistanceScale, B=0.5)
       True Float=1
       False Float=$Particles.DistanceScale
      )
     )
    ))
   &DynamicMaterialParameters(
    Index 0 Param 1=$Particles.VelocityScale
    Module.Param0WriteEnabled=true
    Module.Param1WriteEnabled=true
    Module.Param2WriteEnabled=true
    Module.Param3WriteEnabled=true
   )
  ]
  rendererClasses [&NiagaraSpriteRendererProperties, &NiagaraRibbonRendererProperties]
  renderers [
   {rendererClass=&NiagaraSpriteRendererProperties}
   {rendererClass=&NiagaraRibbonRendererProperties, rendererIndex=1}
  ]
 }
 Glow {
  emitterUpdateScript [
   &EmitterState()
   &SpawnRate(SpawnRate=65, Use Spawn Probability=false)
  ]
  particleSpawnScript [
   /Niagara/Modules/Spawn/Initialization/InitializeParticle(
    Lifetime=&UniformRangedFloat(Minimum=0.2, Maximum=0.35)
    Module.Write Color=false
    Module.Write Lifetime=true
    Module.Write Mass=true
    Module.Write Material Random=true
    Module.Write Position=true
    Module.Write Scale=false
    Module.Write SpriteRotation=false
    Module.Write SpriteSize=true
    Position=&SimulationPosition()
    Sprite Size=&Vector2DFromFloat(Value=&UniformRangedFloat(Minimum=0.5, Maximum=1))
   )
   &Color(Color=&MultiplyLinearColor(
     Linear Color=<rgba 15 10 1 1>
     Scale Factor=/Niagara/DynamicInputs/UniformRange/UniformRangedVector4(
      Minimum:Vector4f=<1 0.25 0.25 0.01>
      Maximum:Vector4f=<1 1 1 1>
      Randomness Mode:ENiagaraRandomnessMode=::NewEnumerator0
      Random Seed:NiagaraInt32=0
      Fixed Random Seed:NiagaraBool=false
     )
    ))
   /SetVariables(Particles.AlternateSockets:NiagaraInt32=&ModuloInt(
     A=&GenerateIntFromCounter(SimpleCounter:NiagaraDataInterfaceSimpleCounter=DataInterface() {propertyValues {GpuSyncMode=None, InitialValue=0}})
     B=2
    )) {isSetParameters=true}
   &SkeletalMeshLocation(
    Bone / Socket Index=$Particles.AlternateSockets
    Inherit Velocity=true
    MaskSpawnGroup=false
    Mesh Sampling Type=::NewEnumerator5
    Sampled Position Offset Coordinate Space:ENiagaraMeshTransforms=::NewEnumerator3
    Skeletal Mesh=$User.SkeletalMesh
    Socket Sampling Mode=::NewEnumerator5
    VelocitySpeedLimit=false
   )
  ]
  particleUpdateScript [
   &ParticleState()
   &SkeletalMeshLocation(
    Apply To Particle Position:NiagaraBool=&SetBoolByFloatComparison(A=$Particles.NormalizedAge, B=0.8, Comparison Type=::NewEnumerator6)
    Bone / Socket Index=$Particles.AlternateSockets
    Inherit Velocity=false
    Interpolate Mesh Sampling=false
    MaskSpawnGroup=false
    Mesh Sampling Type=::NewEnumerator5
    Sampled Position Offset Coordinate Space:ENiagaraMeshTransforms=::NewEnumerator1
    Skeletal Mesh=$User.SkeletalMesh
    Socket Sampling Mode=::NewEnumerator5
   ) {module=SkeletalMeshLocation001}
   &AddVelocity(Coordinate Space=::World, Velocity=&RandomRangeVector(Minimum=<-1 -1 0>, Maximum=<1 1 15>))
   &Drag(
    Drag=&VectorLengthSafe(Vector:Vector3f=$Output.SkeletalMeshLocation001.SampledVelocity)
    Use Linear Drag=true
    Use Rotational Drag=false
   )
   &SolveForcesAndVelocity(Clamp Velocity=false, Limit Acceleration=false)
   &ScaleColor(
    Scale Alpha=&FloatFromCurve(FloatCurve=DataInterface() {propertyValues {Curve {keys [<at 0.5 1>,<at 1 0>]}}})
    Scale RGB=&Lerp_Vector(A=<1 1 1>, B=<0.1 -1 -1>, Alpha=$Particles.NormalizedAge)
    ScaleA=true
    ScaleRGB=true
   )
   &KillParticlesInVolume(Invert Volume=true, Origin Offset:Vector3f=<0 0 3>, Sphere Radius=4, Volume Origin:Vector3f=$Output.SkeletalMeshLocation001.SampledPosition)
   &ScaleSpriteSize(Scale Factor=&Vector2DFromFloat(Value=&FloatFromCurve(FloatCurve=DataInterface() {propertyValues {Curve {keys [
          {arriveTangent=-0.0036, interpMode=RCIM_Cubic, leaveTangent=-0.0036, tangentMode=RCTM_User, time=0, value=1}
          {arriveTangent=-2.2201, interpMode=RCIM_Cubic, leaveTangent=-2.2201, tangentMode=RCTM_User, time=1, value=0}
         ]}}})))
   &CameraOffset(Camera Offset Amount=&RandomRangeFloat(Minimum=-10, Maximum=-0.5))
   /SetVariables(
    Particles.ColorAlt:LinearColor=&MultiplyLinearColor(
     Linear Color=$Particles.Color
     Scale Factor=&MakeFloat4FromVectorAndFloat(
      "Vector (RGB)"=<1 0.1 0.1>
      "Float (Alpha)"=&FloatFromCurve(
       FloatCurve=DataInterface() {propertyValues {Curve {keys [<at 0 1>,<at 0.5 0>,<at 1 1>]}}}
       CurveIndex=Linked() {linkedVariable {name=NPC.FortniteNPC.TimeOfDay_0-1, type=NiagaraFloat}}
       Scale Curve=0.004
      )
     )
    )
    Particles.SpriteSizeAlt:Vector2f=&Vector2DFromFloat(Value=&FloatFromCurve(
      FloatCurve=DataInterface() {propertyValues {Curve {keys [
          {arriveTangent=5.6057, interpMode=RCIM_Cubic, leaveTangent=5.6057, tangentMode=RCTM_User, time=0, value=0}
          {arriveTangent=0.0205, interpMode=RCIM_Cubic, leaveTangent=0.0205, tangentMode=RCTM_User, time=0.25, value=1}
          {arriveTangent=-2.2201, interpMode=RCIM_Cubic, leaveTangent=-2.2201, tangentMode=RCTM_User, time=1, value=0}
         ]}}}
      Scale Curve=14
     ))
   ) {isSetParameters=true}
   /SetVariables(Particles.VisibilityTag=&MakeCustomIntFromBool(Bool=$User.bisHeadHidden, True INT=1, False INT=0)) {isSetParameters=true, module=SetVariables_#1}
  ]
  rendererClasses [&NiagaraSpriteRendererProperties]
  renderers [
   {rendererClass=&NiagaraSpriteRendererProperties}
   {rendererClass=&NiagaraSpriteRendererProperties, rendererIndex=1}
  ]
 }
 LightEyes {
  emitterUpdateScript [
   &EmitterState()
   &SpawnRate(SpawnRate=4, Use Spawn Probability=false)
  ]
  particleSpawnScript [
   /SetVariables(Particles.Lifetime=&RandomRangeFloat(Minimum=1, Maximum=1.5)) {isSetParameters=true, module=SetVariables_#2}
   /SetVariables(Particles.RandomPosition:Vector3f=&Lerp_Vector(
     A=<3 6 5.85>
     B=<3 6 -5.85>
     Alpha=&MakeFloatFromInt(Int=&ModuloInt(
       A=&GenerateIntFromCounter(SimpleCounter:NiagaraDataInterfaceSimpleCounter=DataInterface() {propertyValues {GpuSyncMode=None, InitialValue=0}})
       B=2
      ))
    )) {isSetParameters=true}
  ]
  particleUpdateScript [
   &ParticleState()
   &SkeletalMeshLocation(
    Bone Sampling Mode=::NewEnumerator4
    Inherit Velocity=false
    Interpolate Mesh Sampling=false
    MaskSpawnGroup=false
    Sampled Position Offset=$Particles.RandomPosition
    Sampled Position Offset Coordinate Space:ENiagaraMeshTransforms=::NewEnumerator3
    Skeletal Mesh=$User.SkeletalMesh
   )
   &FN_Light_v3(
    Create Exponent Attribute=true
    Create Radius Attribute=true
    Enable Particle Light Enable=true
    Exponent=1
    Light Enabled=true
    Radius:NiagaraFloat=&RandomRangeFloat(Minimum=6, Maximum=8.5)
   )
   &Color(Color=&MultiplyLinearColor(
     Linear Color=&MultiplyLinearColor(
      Linear Color=&ColorFromCurve(
       ColorCurve=DataInterface() {propertyValues {
         AlphaCurve {keys [<at 0.5 1>]}
         BlueCurve {keys [<at 0 0 RCIM_Cubic>,<at 0.486 0>,<at 1 0 RCIM_Cubic>]}
         GreenCurve {keys [<at 0 1.2627 RCIM_Cubic>,<at 0.486 0.7216>,<at 1 1.2627 RCIM_Cubic>]}
         RedCurve {keys [<at 0 7 RCIM_Cubic>,<at 0.486 4>,<at 1 7 RCIM_Cubic>]}
        }}
       CurveIndex=Linked() {linkedVariable {name=NPC.FortniteNPC.TimeOfDay_0-1, type=NiagaraFloat}}
      )
      Scale Factor=&MakeVector4_FromFloat(Float=&FloatFromCurve(
        FloatCurve=DataInterface() {propertyValues {Curve {keys [
            <at 0 1 RCIM_Cubic>
            {arriveTangent=-0.195, interpMode=RCIM_Cubic, leaveTangent=-0.195, time=0.1324, value=0.8275}
            {arriveTangent=0.7671, interpMode=RCIM_Cubic, leaveTangent=0.7671, time=0.1944, value=0.9621}
            {arriveTangent=0.4215, interpMode=RCIM_Cubic, leaveTangent=0.4215, time=0.2882, value=0.947}
            {arriveTangent=-1.1376, interpMode=RCIM_Cubic, leaveTangent=-1.1376, time=0.3721, value=1.037}
            {arriveTangent=-0.407, interpMode=RCIM_Cubic, leaveTangent=-0.407, time=0.466, value=0.7447}
            {arriveTangent=1.0678, interpMode=RCIM_Cubic, leaveTangent=1.0678, time=0.5433, value=0.9673}
            {arriveTangent=0.4403, interpMode=RCIM_Cubic, leaveTangent=0.4403, time=0.6442, value=0.9349}
            {arriveTangent=-0.2441, interpMode=RCIM_Cubic, leaveTangent=-0.2441, time=0.7685, value=1.0665}
            {arriveTangent=-0.2873, interpMode=RCIM_Cubic, leaveTangent=-0.2873, time=0.889, value=0.8751}
            <at 1 1 RCIM_Cubic>
           ]}}}
        Scale Curve=&FloatFromCurve(
         FloatCurve=DataInterface() {propertyValues {Curve {keys [
             {arriveTangent=12.3106, interpMode=RCIM_Cubic, leaveTangent=12.3106, tangentMode=RCTM_User, time=0, value=0}
             {arriveTangent=-0.0561, interpMode=RCIM_Cubic, leaveTangent=-0.0561, tangentMode=RCTM_User, time=0.15, value=1}
             {arriveTangent=-0.0041, interpMode=RCIM_Cubic, leaveTangent=-0.0041, tangentMode=RCTM_User, time=0.85, value=1}
             {arriveTangent=-9.2892, interpMode=RCIM_Cubic, leaveTangent=-9.2892, tangentMode=RCTM_User, time=1, value=0}
            ]}}}
         Scale Curve=&FloatFromCurve(FloatCurve=DataInterface() {propertyValues {Curve {keys [
              <at 0 1 RCIM_Cubic>
              {arriveTangent=-0.6274, interpMode=RCIM_Cubic, leaveTangent=-0.6274, time=0.126, value=0.6848}
              {arriveTangent=0.5412, interpMode=RCIM_Cubic, leaveTangent=0.5412, time=0.1934, value=0.8787}
              {arriveTangent=0.8857, interpMode=RCIM_Cubic, leaveTangent=0.8857, time=0.3101, value=0.7844}
              {arriveTangent=-1.0546, interpMode=RCIM_Cubic, leaveTangent=-1.0546, time=0.3721, value=1.037}
              {arriveTangent=-0.407, interpMode=RCIM_Cubic, leaveTangent=-0.407, time=0.4481, value=0.6388}
              {arriveTangent=1.0628, interpMode=RCIM_Cubic, leaveTangent=1.0628, time=0.5433, value=0.9673}
              {arriveTangent=0.8846, interpMode=RCIM_Cubic, leaveTangent=0.8846, time=0.6644, value=0.8686}
              {arriveTangent=-1.8131, interpMode=RCIM_Cubic, leaveTangent=-1.8131, time=0.7866, value=1.1826}
              {arriveTangent=-0.8559, interpMode=RCIM_Cubic, leaveTangent=-0.8559, time=0.8714, value=0.4932}
              <at 1 1 RCIM_Cubic>
             ]}}})
        )
       ))
     )
     Scale Factor=&UniformRangedVector4(
      Minimum:Vector4f=<0.75 0.25 0 1>
      Maximum:Vector4f=<1 1 0 1>
      Randomness Mode:ENiagaraRandomnessMode=::NewEnumerator1
      Random Seed:NiagaraInt32=&MakeIntFromFloat(Float=&MakeFloatFromVector(Vector=$Particles.RandomPosition, Channel=::NewEnumerator2))
      Fixed Random Seed:NiagaraBool=true
     )
    ))
   /SetVariables(Particles.SpriteSize=&Vector2DFromFloat(Value=5)) {isSetParameters=true, module=SetVariables_#3}
   /SetVariables(Particles.VisibilityTag=&MakeCustomIntFromBool(Bool=$User.bisHeadHidden, True INT=1, False INT=0)) {isSetParameters=true, module=SetVariables_#1}
  ]
  rendererClasses [&NiagaraSpriteRendererProperties, &NiagaraLightRendererProperties]
  renderers [
   {rendererClass=&NiagaraSpriteRendererProperties}
   {rendererClass=&NiagaraLightRendererProperties, rendererIndex=1}
  ]
 }
 LightMouth001 {
  emitterUpdateScript [
   &EmitterState()
   /SetVariables(Emitter.TimeOfDayMult:NiagaraFloat=Linked() {linkedVariable {name=NPC.FortniteNPC.TimeOfDay_0-1, type=NiagaraFloat}}) {isSetParameters=true}
   &SpawnRate(
    Spawn Probability=&MakeCustomFloatFromBool(
     Bool=&Boolean_AndOperation(
      Bool1=&SetBoolByFloatComparison(A=Linked() {linkedVariable {name=NPC.FortniteNPC.TimeOfDay_0-1, type=NiagaraFloat}}, B=0.25)
      Bool2=&SetBoolByFloatComparison(A=Linked() {linkedVariable {name=NPC.FortniteNPC.TimeOfDay_0-1, type=NiagaraFloat}}, B=0.75, Comparison Type=::NewEnumerator6)
     )
     True Float=0
     False Float=1
    )
    SpawnRate=4
    Use Spawn Probability=true
   )
  ]
  particleSpawnScript [/SetVariables(Particles.Lifetime=&RandomRangeFloat(Minimum=1, Maximum=2)) {isSetParameters=true, module=SetVariables_#2}]
  particleUpdateScript [
   &ParticleState()
   &SkeletalMeshLocation(
    Bone Sampling Mode=::NewEnumerator4
    Inherit Velocity=false
    Interpolate Mesh Sampling=false
    MaskSpawnGroup=false
    Sampled Position Offset=&RandomRangeVector(Minimum=<-4 5 -7>, Maximum=<-5 7 7>)
    Sampled Position Offset Coordinate Space:ENiagaraMeshTransforms=::NewEnumerator3
    Skeletal Mesh=$User.SkeletalMesh
   )
   &FN_Light_v3(
    Create Exponent Attribute=true
    Create Radius Attribute=true
    Enable Particle Light Enable=true
    Exponent=1
    Light Enabled=true
    Radius:NiagaraFloat=&RandomRangeFloat(Minimum=6, Maximum=6)
   )
   &Color(Color=&MultiplyLinearColor(
     Linear Color=&MultiplyLinearColor(
      Linear Color=&ColorFromCurve(
       ColorCurve=DataInterface() {propertyValues {
         AlphaCurve {keys [<at 0.5 1>]}
         BlueCurve {keys [<at 0 0 RCIM_Cubic>,<at 0.25 0>,<at 0.75 0>,<at 1 0 RCIM_Cubic>]}
         GreenCurve {keys [<at 0 0.5412 RCIM_Cubic>,<at 0.25 0.0078>,<at 0.75 0.0078>,<at 1 0.5412 RCIM_Cubic>]}
         RedCurve {keys [<at 0 3 RCIM_Cubic>,<at 0.25 0.498>,<at 0.75 0.498>,<at 1 3 RCIM_Cubic>]}
        }}
       CurveIndex=Linked() {linkedVariable {name=NPC.FortniteNPC.TimeOfDay_0-1, type=NiagaraFloat}}
      )
      Scale Factor=&MakeVector4_FromFloat(Float=&FloatFromCurve(
        FloatCurve=DataInterface() {propertyValues {Curve {keys [
            <at 0 1 RCIM_Cubic>
            {arriveTangent=-0.195, interpMode=RCIM_Cubic, leaveTangent=-0.195, time=0.1324, value=0.8275}
            {arriveTangent=0.7671, interpMode=RCIM_Cubic, leaveTangent=0.7671, time=0.1944, value=0.9621}
            {arriveTangent=0.4215, interpMode=RCIM_Cubic, leaveTangent=0.4215, time=0.2882, value=0.947}
            {arriveTangent=-1.1376, interpMode=RCIM_Cubic, leaveTangent=-1.1376, time=0.3721, value=1.037}
            {arriveTangent=-0.407, interpMode=RCIM_Cubic, leaveTangent=-0.407, time=0.466, value=0.7447}
            {arriveTangent=1.0678, interpMode=RCIM_Cubic, leaveTangent=1.0678, time=0.5433, value=0.9673}
            {arriveTangent=0.4403, interpMode=RCIM_Cubic, leaveTangent=0.4403, time=0.6442, value=0.9349}
            {arriveTangent=-0.2441, interpMode=RCIM_Cubic, leaveTangent=-0.2441, time=0.7685, value=1.0665}
            {arriveTangent=-0.2873, interpMode=RCIM_Cubic, leaveTangent=-0.2873, time=0.889, value=0.8751}
            <at 1 1 RCIM_Cubic>
           ]}}}
        Scale Curve=&FloatFromCurve(
         FloatCurve=DataInterface() {propertyValues {Curve {keys [
             {arriveTangent=12.3106, interpMode=RCIM_Cubic, leaveTangent=12.3106, tangentMode=RCTM_User, time=0, value=0}
             {arriveTangent=-0.0561, interpMode=RCIM_Cubic, leaveTangent=-0.0561, tangentMode=RCTM_User, time=0.15, value=1}
             {arriveTangent=-0.0041, interpMode=RCIM_Cubic, leaveTangent=-0.0041, tangentMode=RCTM_User, time=0.85, value=1}
             {arriveTangent=-9.2892, interpMode=RCIM_Cubic, leaveTangent=-9.2892, tangentMode=RCTM_User, time=1, value=0}
            ]}}}
         Scale Curve=&FloatFromCurve(FloatCurve=DataInterface() {propertyValues {Curve {keys [
              <at 0 1 RCIM_Cubic>
              {arriveTangent=-0.6274, interpMode=RCIM_Cubic, leaveTangent=-0.6274, time=0.126, value=0.6848}
              {arriveTangent=0.5412, interpMode=RCIM_Cubic, leaveTangent=0.5412, time=0.1934, value=0.8787}
              {arriveTangent=0.8857, interpMode=RCIM_Cubic, leaveTangent=0.8857, time=0.3101, value=0.7844}
              {arriveTangent=-1.0546, interpMode=RCIM_Cubic, leaveTangent=-1.0546, time=0.3721, value=1.037}
              {arriveTangent=-0.407, interpMode=RCIM_Cubic, leaveTangent=-0.407, time=0.4481, value=0.6388}
              {arriveTangent=1.0628, interpMode=RCIM_Cubic, leaveTangent=1.0628, time=0.5433, value=0.9673}
              {arriveTangent=0.8846, interpMode=RCIM_Cubic, leaveTangent=0.8846, time=0.6644, value=0.8686}
              {arriveTangent=-1.8131, interpMode=RCIM_Cubic, leaveTangent=-1.8131, time=0.7866, value=1.1826}
              {arriveTangent=-0.8559, interpMode=RCIM_Cubic, leaveTangent=-0.8559, time=0.8714, value=0.4932}
              <at 1 1 RCIM_Cubic>
             ]}}})
        )
       ))
     )
     Scale Factor=&UniformRangedVector4(
      Minimum:Vector4f=<0.75 0.25 0 1>
      Maximum:Vector4f=<1 1 0 1>
      Randomness Mode:ENiagaraRandomnessMode=::NewEnumerator1
      Random Seed:NiagaraInt32=&MakeIntFromFloat(Float=&MakeFloatFromVector(Vector=$Particles.RandomPosition, Channel=::NewEnumerator2))
      Fixed Random Seed:NiagaraBool=true
     )
    ))
   /SetVariables(Particles.SpriteSize=&Vector2DFromFloat(Value=5)) {isSetParameters=true, module=SetVariables_#3}
   /SetVariables(Particles.VisibilityTag=&MakeCustomIntFromBool(Bool=$User.bisHeadHidden, True INT=1, False INT=0)) {isSetParameters=true, module=SetVariables_#1}
  ]
  rendererClasses [&NiagaraSpriteRendererProperties, &NiagaraLightRendererProperties]
  renderers [
   {rendererClass=&NiagaraSpriteRendererProperties}
   {rendererClass=&NiagaraLightRendererProperties, rendererIndex=1}
  ]
 }
}
summary {
 systemName=NS_Head_PumpkinPunk_Flame
 emitters [
  {emitterName=Glow, rendererClasses [&NiagaraSpriteRendererProperties]}
  {emitterName=Flame, rendererClasses [&NiagaraSpriteRendererProperties, &NiagaraRibbonRendererProperties]}
  {emitterName=LightEyes, rendererClasses [&NiagaraSpriteRendererProperties, &NiagaraLightRendererProperties]}
  {emitterName=LightMouth001, rendererClasses [&NiagaraSpriteRendererProperties, &NiagaraLightRendererProperties]}
 ]
 userVariables [
  User.bisHeadHidden {type=NiagaraBool}
  User.SkeletalMesh {
   type=NiagaraDataInterfaceSkeletalMesh
   defaultValue {
    struct=NiagaraExt_VariableValue_DataInterface
    value {dataInterface=&self:NiagaraDataInterfaceSkeletalMesh_1}
   }
  }
 ]
}
