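# Physics asset for the crouch variant of the male medium Pumpkin Punk player cosmetic body, covering the lower body and its dangling flaps.
# Capsules on pelvis, thighs, calves and feet support chained flap bodies: three back chains from the pelvis, two per thigh-mounted pocket.
# Tiny MSTR anchor capsules root each chain, mostly with free angular motion.
# Flap links have narrow twist limits, loosened on the longer center-back tail.
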PhysicsAsset /BRCosmetics/Characters/Player/Male/Medium/Bodies/M_MED_Pumpkin_Punk/Meshes/M_MED_Pumpkin_Punk_Crouch_Physics
bodies {
 calf_l [{shapeType=Capsule, length=21.1915, radius=7.8545, center=<-19.6494 -1.3006 -0.1399>, rotation=<rot 0 87.5954 -89.474>}]
 calf_r [{shapeType=Capsule, length=21.1844, radius=7.6665, center=<19.598 0.9722 0.0533>, rotation=<rot 0 87.8298 -89.671>}]
 dyn_BK_r_MSTR [{shapeType=Capsule, length=0.505, radius=0.505}]
 dyn_bk_MSTR [{shapeType=Capsule, length=0.505, radius=0.505}]
 dyn_bk_flap [{shapeType=Capsule, length=9.5, radius=2.75, center=<2.29 0 0>, rotation=<rot 0 0 90>}]
 dyn_bk_flap_01 [{shapeType=Capsule, length=9.5, radius=2.75, center=<2.29 0 0>, rotation=<rot 0 0 90>}]
 dyn_bk_flap_02 [{shapeType=Capsule, length=10, radius=2.75, center=<2.29 0 0>, rotation=<rot 0 0 90>}]
 dyn_bk_flap_03 [{shapeType=Capsule, length=23, radius=3, shapeName=dyn_bk_flap_03_capsule, center=<1.1363 0 0.4221>, rotation=<rot 0 0 90>}]
 dyn_bk_flap_04 [{shapeType=Capsule, length=20, radius=3, shapeName=dyn_bk_flap_04_capsule, center=<0 0 -0.1993>, rotation=<rot 0 0 90>}]
 dyn_bk_flap_05 [{shapeType=Capsule, length=12, radius=3, shapeName=dyn_bk_flap_05_capsule, center=<0 0 -0.7379>, rotation=<rot 0 0 90>}]
 dyn_bk_flap_l [{shapeType=Capsule, length=1, radius=3.5, center=<2.3027 0 0>, rotation=<rot 0 0 103.7103>}]
 dyn_bk_flap_l_01 [{shapeType=Capsule, length=2, radius=2.75, center=<2.3027 0 0>, rotation=<rot 0 0 103.7148>}]
 dyn_bk_flap_l_02 [{shapeType=Capsule, length=3, radius=2.25, center=<2.3027 0 0>, rotation=<rot 0 0 90.0756>}]
 dyn_bk_flap_r [{shapeType=Capsule, length=2.5, radius=3, center=<1.7992 0 0>, rotation=<rot 0 0 72.8317>}]
 dyn_bk_flap_r_01 [{shapeType=Capsule, length=2.5, radius=3, center=<1.7992 0 0>, rotation=<rot 0 0 72.8375>}]
 dyn_bk_flap_r_02 [{shapeType=Capsule, length=2.5, radius=3, center=<1.7992 0 0>, rotation=<rot 0 0 87.7005>}]
 dyn_bk_l_MSTR [{shapeType=Capsule, length=0.505, radius=0.505}]
 dyn_l_a_MSTR [{shapeType=Capsule, length=0.505, radius=0.505}]
 dyn_l_b_MSTR [{shapeType=Capsule, length=0.505, radius=0.505}]
 dyn_l_flap_a [{shapeType=Capsule, length=7.75, radius=3.25, center=<2.6597 0.1093 -0.3002>, rotation=<rot 0 0 70>}]
 dyn_l_flap_a_01 [{shapeType=Capsule, length=8.25, radius=3.5, center=<2.4426 0 -0.0001>, rotation=<rot 0 0 70.195>}]
 dyn_l_flap_a_02 [{shapeType=Capsule, length=7.25, radius=3.75, center=<3.6942 0 -0.0001>, rotation=<rot 0 0 70.0162>}]
 dyn_l_flap_b [{shapeType=Capsule, length=7.25, radius=3.25, center=<3.3865 -0.1961 -0.7478>, rotation=<rot 0 0.0002 -75.306>}]
 dyn_l_flap_b_01 [{shapeType=Capsule, length=7.7192, radius=4.5308, center=<2.0586 0.2352 -1.0371>, rotation=<rot 0 0 -64.5595>}]
 dyn_l_flap_b_02 [{shapeType=Capsule, length=7.25, radius=4, center=<4.1473 0.9967 0.0137>, rotation=<rot 0 0 -70.36>}]
 dyn_l_pocket_MSTR [{shapeType=Capsule, length=0.505, radius=0.505}]
 dyn_r_a_MSTR [{shapeType=Capsule, length=0.505, radius=0.505}]
 dyn_r_b_MSTR [{shapeType=Capsule, length=0.505, radius=0.505}]
 dyn_r_flap_a [{shapeType=Capsule, length=9.5, radius=3.25, center=<3.6012 -0.0721 -0.3458>, rotation=<rot 0 0 101.7657>}]
 dyn_r_flap_a_01 [{shapeType=Capsule, length=9.5, radius=3.25, center=<3.6012 0 0>, rotation=<rot 0 0 101.7818>}]
 dyn_r_flap_a_02 [{shapeType=Capsule, length=9.5, radius=3.25, center=<3.6012 0 0>, rotation=<rot 0 0 101.7183>}]
 dyn_r_flap_b [{shapeType=Capsule, length=9.75, radius=3.5, center=<3.3894 -0.033 0.1985>, rotation=<rot 0 -0.0002 78.4425>}]
 dyn_r_flap_b_01 [{shapeType=Capsule, length=9.75, radius=3.5, center=<3.2001 0 0>, rotation=<rot 0 0 68.4517>}]
 dyn_r_flap_b_02 [{shapeType=Capsule, length=9.75, radius=3.5, center=<3.2001 0 0>, rotation=<rot 0 0 68.4479>}]
 dyn_r_pocket_MSTR [{shapeType=Capsule, length=0.505, radius=0.505}]
 foot_l [{shapeType=Capsule, length=13.8732, radius=7.7403, center=<-3.9618 7.4037 -0.5186>, rotation=<rot 0 -170.0538 -89.4198>}]
 foot_r [{shapeType=Capsule, length=13.87, radius=7.9114, center=<3.8052 -7.4365 0.5183>, rotation=<rot 0 -170.0101 -89.4581>}]
 pelvis [
  {shapeType=Capsule, length=10.4485, radius=10.8556, center=<-14.356 -0.0806 0.0074>}
  {shapeType=Capsule, length=13.8896, radius=9.1959, center=<-20.3095 1.8206 0>}
  {shapeType=Capsule, length=10.4994, radius=7.7912, center=<-7.7884 3.0617 0>}
 ]
 thigh_l [{shapeType=Capsule, length=26.9944, radius=8.2093, center=<-21.8162 -0.5241 0.7205>, rotation=<rot 0.0002 -89.9996 -90>}]
 thigh_r [{shapeType=Capsule, length=17.2106, radius=9.0994, center=<21.7035 0.6059 -0.742>, rotation=<rot 0.0001 -89.9994 -89.9999>}]
}
constraints [
 {bone1Name=calf_l, bone2Name=thigh_l, swing1LimitDegrees=45, swing2LimitDegrees=45, twistLimitDegrees=45}
 {bone1Name=foot_l, bone2Name=calf_l, swing1LimitDegrees=45, swing2LimitDegrees=45, twistLimitDegrees=45}
 {bone1Name=calf_r, bone2Name=thigh_r, swing1LimitDegrees=45, swing2LimitDegrees=45, twistLimitDegrees=45}
 {bone1Name=foot_r, bone2Name=calf_r, swing1LimitDegrees=45, swing2LimitDegrees=45, twistLimitDegrees=45}
 {
  bone1Name=thigh_r
  bone2Name=pelvis
  swing1LimitDegrees=45
  swing1Motion=Free
  swing2LimitDegrees=45
  swing2Motion=Free
  twistLimitDegrees=45
  twistMotion=Free
 }
 {
  bone1Name=dyn_bk_l_MSTR
  bone2Name=pelvis
  swing1LimitDegrees=45
  swing1Motion=Free
  swing2LimitDegrees=45
  swing2Motion=Free
  twistLimitDegrees=45
  twistMotion=Free
 }
 {
  bone1Name=dyn_bk_MSTR
  bone2Name=pelvis
  swing1LimitDegrees=45
  swing1Motion=Free
  swing2LimitDegrees=45
  swing2Motion=Free
  twistLimitDegrees=45
  twistMotion=Free
 }
 {
  bone1Name=dyn_BK_r_MSTR
  bone2Name=pelvis
  swing1LimitDegrees=45
  swing1Motion=Free
  swing2LimitDegrees=45
  swing2Motion=Free
  twistLimitDegrees=45
  twistMotion=Free
 }
 {bone1Name=dyn_bk_flap, bone2Name=dyn_bk_MSTR, swing1LimitDegrees=30, swing2LimitDegrees=30, twistLimitDegrees=5}
 {bone1Name=dyn_bk_flap_01, bone2Name=dyn_bk_flap, swing1LimitDegrees=30, swing2LimitDegrees=30, twistLimitDegrees=5}
 {bone1Name=dyn_bk_flap_02, bone2Name=dyn_bk_flap_01, swing1LimitDegrees=30, swing2LimitDegrees=30, twistLimitDegrees=5}
 {bone1Name=dyn_bk_flap_r, bone2Name=dyn_BK_r_MSTR, swing1LimitDegrees=30, swing2LimitDegrees=30, twistLimitDegrees=5}
 {bone1Name=dyn_bk_flap_r_01, bone2Name=dyn_bk_flap_r, swing1LimitDegrees=30, swing2LimitDegrees=30, twistLimitDegrees=5}
 {bone1Name=dyn_bk_flap_r_02, bone2Name=dyn_bk_flap_r_01, swing1LimitDegrees=30, swing2LimitDegrees=30, twistLimitDegrees=5}
 {bone1Name=dyn_bk_flap_l, bone2Name=dyn_bk_l_MSTR, swing1LimitDegrees=30, swing2LimitDegrees=30, twistLimitDegrees=5}
 {bone1Name=dyn_bk_flap_l_01, bone2Name=dyn_bk_flap_l, swing1LimitDegrees=30, swing2LimitDegrees=30, twistLimitDegrees=5}
 {bone1Name=dyn_bk_flap_l_02, bone2Name=dyn_bk_flap_l_01, swing1LimitDegrees=30, swing2LimitDegrees=30, twistLimitDegrees=5}
 {bone1Name=dyn_l_flap_b_02, bone2Name=dyn_l_flap_b_01, swing1LimitDegrees=30, swing2LimitDegrees=30, twistLimitDegrees=5}
 {bone1Name=dyn_l_flap_b, bone2Name=dyn_l_a_MSTR, swing1LimitDegrees=30, swing2LimitDegrees=30, twistLimitDegrees=5}
 {bone1Name=dyn_l_flap_b_01, bone2Name=dyn_l_flap_b, swing1LimitDegrees=30, swing2LimitDegrees=30, twistLimitDegrees=5}
 {bone1Name=dyn_l_flap_a_02, bone2Name=dyn_l_flap_a_01, swing1LimitDegrees=30, swing2LimitDegrees=30, twistLimitDegrees=5}
 {bone1Name=dyn_l_flap_a_01, bone2Name=dyn_l_flap_a, swing1LimitDegrees=30, swing2LimitDegrees=30, twistLimitDegrees=5}
 {bone1Name=dyn_l_flap_a, bone2Name=dyn_l_b_MSTR, swing1LimitDegrees=30, swing2LimitDegrees=30, twistLimitDegrees=5}
 {
  bone1Name=dyn_l_a_MSTR
  bone2Name=dyn_l_pocket_MSTR
  swing1LimitDegrees=45
  swing1Motion=Free
  swing2LimitDegrees=45
  swing2Motion=Free
  twistLimitDegrees=45
  twistMotion=Free
 }
 {
  bone1Name=dyn_l_b_MSTR
  bone2Name=dyn_l_pocket_MSTR
  swing1LimitDegrees=45
  swing1Motion=Free
  swing2LimitDegrees=45
  swing2Motion=Free
  twistLimitDegrees=45
  twistMotion=Free
 }
 {
  bone1Name=dyn_l_pocket_MSTR
  bone2Name=thigh_l
  swing1LimitDegrees=45
  swing1Motion=Free
  swing2LimitDegrees=45
  swing2Motion=Free
  twistLimitDegrees=45
  twistMotion=Free
 }
 {bone1Name=dyn_r_flap_a_02, bone2Name=dyn_r_flap_a_01, swing1LimitDegrees=30, swing2LimitDegrees=30, twistLimitDegrees=5}
 {bone1Name=dyn_r_flap_a_01, bone2Name=dyn_r_flap_a, swing1LimitDegrees=30, swing2LimitDegrees=30, twistLimitDegrees=5}
 {bone1Name=dyn_r_flap_a, bone2Name=dyn_r_b_MSTR, swing1LimitDegrees=30, swing2LimitDegrees=30, twistLimitDegrees=5}
 {bone1Name=dyn_r_flap_b_02, bone2Name=dyn_r_flap_b_01, swing1LimitDegrees=30, swing2LimitDegrees=30, twistLimitDegrees=5}
 {bone1Name=dyn_r_flap_b_01, bone2Name=dyn_r_flap_b, swing1LimitDegrees=30, swing2LimitDegrees=30, twistLimitDegrees=5}
 {bone1Name=dyn_r_flap_b, bone2Name=dyn_r_a_MSTR, swing1LimitDegrees=30, swing2LimitDegrees=30, twistLimitDegrees=5}
 {
  bone1Name=dyn_r_a_MSTR
  bone2Name=dyn_r_pocket_MSTR
  swing1LimitDegrees=45
  swing1Motion=Free
  swing2LimitDegrees=45
  swing2Motion=Free
  twistLimitDegrees=45
  twistMotion=Free
 }
 {
  bone1Name=dyn_r_pocket_MSTR
  bone2Name=thigh_r
  swing1LimitDegrees=45
  swing1Motion=Free
  swing2LimitDegrees=45
  swing2Motion=Free
  twistLimitDegrees=45
  twistMotion=Free
 }
 {bone1Name=dyn_r_b_MSTR, bone2Name=dyn_r_pocket_MSTR, swing1LimitDegrees=45, swing2LimitDegrees=45, twistLimitDegrees=15}
 {
  bone1Name=thigh_l
  bone2Name=pelvis
  swing1LimitDegrees=45
  swing1Motion=Free
  swing2LimitDegrees=45
  swing2Motion=Free
  twistLimitDegrees=45
  twistMotion=Free
 }
 {bone1Name=dyn_bk_flap_03, bone2Name=dyn_bk_flap_02, swing1LimitDegrees=40, swing2LimitDegrees=40, twistLimitDegrees=10}
 {bone1Name=dyn_bk_flap_04, bone2Name=dyn_bk_flap_03, swing1LimitDegrees=40, swing2LimitDegrees=40, twistLimitDegrees=10}
 {bone1Name=dyn_bk_flap_05, bone2Name=dyn_bk_flap_04, swing1LimitDegrees=40, swing2LimitDegrees=40, twistLimitDegrees=10}
]
