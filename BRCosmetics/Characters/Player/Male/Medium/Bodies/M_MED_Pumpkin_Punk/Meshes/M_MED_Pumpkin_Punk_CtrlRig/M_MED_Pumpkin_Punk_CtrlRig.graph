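# Control Rig for the Pumpkin Punk character body that procedurally corrects hood collar and jacket bones to follow body motion.
# Head rotation relative to neck_01 drives the back collar, while each lower flap's rotation drives its matching jacket bone.
# Clamped remaps convert Euler angles into additive local offsets, acting as driven keys.
# A separate chain for side, front and back-right collar bones has no link from Begin Execution.
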ControlRigBlueprint /BRCosmetics/Characters/Player/Male/Medium/Bodies/M_MED_Pumpkin_Punk/Meshes/M_MED_Pumpkin_Punk_CtrlRig/M_MED_Pumpkin_Punk_CtrlRig
graphs [
 RigVMFunctionLibrary {links [], nodes []}
 RigVMModel {nodes [
   n1 = &RigVMFunction_MathQuaternionToEuler(RotationOrder=ZYX, Value=&RigUnit_GetRelativeBoneTransform(Bone=head, Space=neck_01).Transform.Rotation)
   n2 = &RigVMFunction_MathQuaternionFromEuler(
    Euler=(X=0,Y=0,Z=0)
    Euler.Y=&RigVMFunction_MathFloatRemap(SourceMaximum=-47, SourceMinimum=-19, TargetMaximum=30, TargetMinimum=0, bClamp=true, Value=@n1.Result.Z).Result
   )
   /Script/ControlRig.RigUnit_ModifyBoneTransforms(
    BoneToModify=((Bone="dyn_hood_collar_r",Transform=(Rotation=(X=0,Y=0,Z=0,W=1),Translation=(X=0,Y=0,Z=0),Scale3D=(X=1,Y=1,Z=1))))
    ExecutePin=()
    Mode=AdditiveLocal
    Weight=1
    WeightMaximum=1
    WeightMinimum=0
    BoneToModify.0.Transform.Rotation=@n2.Result
   ) {is=struct RigVMUnitNode}
   > &RigUnit_ModifyBoneTransforms(
    BoneToModify=((Bone="dyn_hood_collar_l",Transform=(Rotation=(X=0,Y=0,Z=0,W=1),Translation=(X=0,Y=0,Z=0),Scale3D=(X=1,Y=1,Z=1))))
    Mode=AdditiveLocal
    Weight=1
    WeightMaximum=1
    WeightMinimum=0
    BoneToModify.0.Transform.Rotation=@n2.Result
   )
   n3 = &RigVMFunction_MathQuaternionFromEuler(
    Euler=(X=0,Y=0,Z=0)
    Euler.Y=&RigVMFunction_MathFloatRemap(SourceMaximum=3, SourceMinimum=-6, TargetMaximum=20, TargetMinimum=0, bClamp=true, Value=@n1.Result.Z).Result
   )
   > &RigUnit_ModifyBoneTransforms(
    BoneToModify=((Bone="dyn_hood_collar_fr",Transform=(Rotation=(X=0,Y=0,Z=0,W=1),Translation=(X=0,Y=0,Z=0),Scale3D=(X=1,Y=1,Z=1))))
    Mode=AdditiveLocal
    Weight=1
    WeightMaximum=1
    WeightMinimum=0
    BoneToModify.0.Transform.Rotation=@n3.Result
   )
   > &RigUnit_ModifyBoneTransforms(
    BoneToModify=((Bone="dyn_hood_collar_fr_l",Transform=(Rotation=(X=0,Y=0,Z=0,W=1),Translation=(X=0,Y=0,Z=0),Scale3D=(X=1,Y=1,Z=1))))
    Mode=AdditiveLocal
    Weight=1
    WeightMaximum=1
    WeightMinimum=0
    BoneToModify.0.Transform.Rotation=@n3.Result
   )
   > &RigUnit_ModifyBoneTransforms(
    BoneToModify=((Bone="dyn_hood_collar_fr_r",Transform=(Rotation=(X=0,Y=0,Z=0,W=1),Translation=(X=0,Y=0,Z=0),Scale3D=(X=1,Y=1,Z=1))))
    Mode=AdditiveLocal
    Weight=1
    WeightMaximum=1
    WeightMinimum=0
    BoneToModify.0.Transform.Rotation=@n3.Result
   )
   > &RigUnit_ModifyBoneTransforms(
    BoneToModify=((Bone="dyn_hood_collar_l",Transform=(Rotation=(X=0,Y=0,Z=0,W=1),Translation=(X=0,Y=0,Z=0),Scale3D=(X=1,Y=1,Z=1))))
    Mode=AdditiveLocal
    Weight=1
    WeightMaximum=1
    WeightMinimum=0
    BoneToModify.0.Transform.Rotation=&RigVMFunction_MathQuaternionFromEuler(
     Euler=(X=0,Y=0,Z=0)
     Euler.Y=&RigVMFunction_MathFloatRemap(SourceMaximum=3, SourceMinimum=-6, TargetMaximum=25, TargetMinimum=0, bClamp=true, Value=@n1.Result.Z).Result
    ).Result
   )
   n4 = &RigVMFunction_MathQuaternionToEuler(RotationOrder=ZYX, Value=&RigUnit_GetRelativeBoneTransform(Bone=head, Space=neck_01).Transform.Rotation)
   n5 = &RigVMFunction_MathQuaternionFromEuler(
    Euler=(X=0,Y=0,Z=0)
    Euler.Y=&RigVMFunction_MathFloatRemap(SourceMaximum=-15, SourceMinimum=20, TargetMaximum=60, TargetMinimum=0, bClamp=true, Value=@n4.Result.X).Result
   )
   > &RigUnit_ModifyBoneTransforms(
    BoneToModify=((Bone="dyn_hood_collar_l",Transform=(Rotation=(X=0,Y=0,Z=0,W=1),Translation=(X=0,Y=0,Z=0),Scale3D=(X=1,Y=1,Z=1))))
    Mode=AdditiveLocal
    Weight=1
    WeightMaximum=1
    WeightMinimum=0
    BoneToModify.0.Transform.Rotation=@n5.Result
   )
   > &RigUnit_ModifyBoneTransforms(
    BoneToModify=((Bone="dyn_hood_collar_fr_l",Transform=(Rotation=(X=0,Y=0,Z=0,W=1),Translation=(X=0,Y=0,Z=0),Scale3D=(X=1,Y=1,Z=1))))
    Mode=AdditiveLocal
    Weight=1
    WeightMaximum=1
    WeightMinimum=0
    BoneToModify.0.Transform.Rotation=@n5.Result
   )
   n6 = &RigVMFunction_MathQuaternionFromEuler(
    Euler=(X=0,Y=0,Z=0)
    Euler.Y=&RigVMFunction_MathFloatRemap(SourceMaximum=20, SourceMinimum=0, TargetMaximum=30, TargetMinimum=0, bClamp=true, Value=@n4.Result.X).Result
   )
   > &RigUnit_ModifyBoneTransforms(
    BoneToModify=((Bone="dyn_hood_collar_r",Transform=(Rotation=(X=0,Y=0,Z=0,W=1),Translation=(X=0,Y=0,Z=0),Scale3D=(X=1,Y=1,Z=1))))
    Mode=AdditiveLocal
    Weight=1
    WeightMaximum=1
    WeightMinimum=0
    BoneToModify.0.Transform.Rotation=@n6.Result
   )
   > &RigUnit_ModifyBoneTransforms(
    BoneToModify=((Bone="dyn_hood_collar_fr_r",Transform=(Rotation=(X=0,Y=0,Z=0,W=1),Translation=(X=0,Y=0,Z=0),Scale3D=(X=1,Y=1,Z=1))))
    Mode=AdditiveLocal
    Weight=1
    WeightMaximum=1
    WeightMinimum=0
    BoneToModify.0.Transform.Rotation=@n6.Result
   )
   n7 = &RigVMFunction_MathQuaternionFromEuler(
    Euler=(X=0,Y=0,Z=0)
    Euler.Y=&RigVMFunction_MathFloatRemap(SourceMaximum=7, SourceMinimum=-6, TargetMaximum=15, TargetMinimum=0, bClamp=true, Value=@n4.Result.X).Result
   )
   > &RigUnit_ModifyBoneTransforms(
    BoneToModify=((Bone="dyn_hood_collar_r",Transform=(Rotation=(X=0,Y=0,Z=0,W=1),Translation=(X=0,Y=0,Z=0),Scale3D=(X=1,Y=1,Z=1))))
    Mode=AdditiveLocal
    Weight=1
    WeightMaximum=1
    WeightMinimum=0
    BoneToModify.0.Transform.Rotation=@n7.Result
   )
   > &RigUnit_ModifyBoneTransforms(
    BoneToModify=((Bone="dyn_hood_collar_bk_r",Transform=(Rotation=(X=0,Y=0,Z=0,W=1),Translation=(X=0,Y=0,Z=0),Scale3D=(X=1,Y=1,Z=1))))
    Mode=AdditiveLocal
    Weight=1
    WeightMaximum=1
    WeightMinimum=0
    BoneToModify.0.Transform.Rotation=@n7.Result
   )
   RigUnit_BeginExecution = &RigUnit_BeginExecution()
   n8 = &RigVMFunction_MathFloatRemap(SourceMaximum=-55, SourceMinimum=-19, TargetMaximum=-3, TargetMinimum=0, bClamp=true, Value=@n1.Result.Z)
   n9 = &RigVMFunction_MathQuaternionFromEuler(
    Euler=(X=0,Y=0,Z=0)
    Euler.Y=&RigVMFunction_MathFloatRemap(SourceMaximum=-55, SourceMinimum=-25, TargetMaximum=25, TargetMinimum=0, bClamp=true, Value=@n1.Result.Z).Result
   )
   &RigUnit_ModifyBoneTransforms(
    BoneToModify=((Bone="dyn_hood_collar_bk",Transform=(Rotation=(X=0,Y=0,Z=0,W=1),Translation=(X=0,Y=0,Z=0),Scale3D=(X=1,Y=1,Z=1))))
    Mode=AdditiveLocal
    Weight=1
    WeightMaximum=1
    WeightMinimum=0
    BoneToModify.0.Transform.Translation.Z=@n8.Result
    BoneToModify.0.Transform.Rotation=@n9.Result
    ExecutePin=@RigUnit_BeginExecution.ExecutePin
   )
   > &RigUnit_ModifyBoneTransforms(
    BoneToModify=((Bone="dyn_hood_collar_bk_l",Transform=(Rotation=(X=0,Y=0,Z=0,W=1),Translation=(X=0,Y=0,Z=0),Scale3D=(X=1,Y=1,Z=1))))
    Mode=AdditiveLocal
    Weight=1
    WeightMaximum=1
    WeightMinimum=0
    BoneToModify.0.Transform.Translation.Z=@n8.Result
    BoneToModify.0.Transform.Rotation=@n9.Result
   )
   > &RigUnit_ModifyBoneTransforms(
    BoneToModify=((Bone="dyn_hood_collar_bk_r",Transform=(Rotation=(X=0,Y=0,Z=0,W=1),Translation=(X=0,Y=0,Z=0),Scale3D=(X=1,Y=1,Z=1))))
    Mode=AdditiveLocal
    Weight=1
    WeightMaximum=1
    WeightMinimum=0
    BoneToModify.0.Transform.Translation.Z=@n8.Result
    BoneToModify.0.Transform.Rotation=&RigVMFunction_MathQuaternionFromEuler(
     Euler=(X=0,Y=0,Z=0)
     Euler.Y=&RigVMFunction_MathFloatRemap(SourceMaximum=-55, SourceMinimum=-27, TargetMaximum=40, TargetMinimum=0, bClamp=true, Value=@n1.Result.Z).Result
    ).Result
   )
   n10 = &RigVMFunction_MathQuaternionToEuler(RotationOrder=ZYX, Value=&RigUnit_GetRelativeBoneTransform(Bone=dyn_bk_flap, Space=pelvis).Transform.Rotation)
   > &RigUnit_ModifyBoneTransforms(
    BoneToModify=((Bone="dyn_bk_jacket",Transform=(Rotation=(X=0,Y=0,Z=0,W=1),Translation=(X=0,Y=0,Z=0),Scale3D=(X=1,Y=1,Z=1))))
    Mode=AdditiveLocal
    Weight=1
    WeightMaximum=1
    WeightMinimum=0
    BoneToModify.0.Transform.Rotation=&RigVMFunction_MathQuaternionFromEuler(
     Euler=(X=0,Y=0,Z=0)
     Euler.Y=&RigVMFunction_MathFloatRemap(SourceMaximum=45, SourceMinimum=1, TargetMaximum=-40, TargetMinimum=-10, bClamp=true, Value=@n10.Result.Y).Result
    ).Result
   )
   n11 = &RigVMFunction_MathQuaternionToEuler(RotationOrder=ZYX, Value=&RigUnit_GetBoneTransform(Bone=dyn_bk_flap_l, Space=LocalSpace).Transform.Rotation)
   n12 = &RigVMFunction_MathQuaternionToEuler(RotationOrder=ZYX, Value=&RigUnit_GetRelativeBoneTransform(Bone=dyn_bk_flap_l, Space=pelvis).Transform.Rotation)
   > &RigUnit_ModifyBoneTransforms(
    BoneToModify=((Bone="dyn_bk_jacket_l",Transform=(Rotation=(X=0,Y=0,Z=0,W=1),Translation=(X=0,Y=0,Z=0),Scale3D=(X=1,Y=1,Z=1))))
    Mode=AdditiveLocal
    Weight=1
    WeightMaximum=1
    WeightMinimum=0
    BoneToModify.0.Transform.Rotation=&RigVMFunction_MathQuaternionFromEuler(
     Euler=(X=0,Y=0,Z=0)
     Euler.Z=&RigVMFunction_MathFloatRemap(SourceMaximum=-75, SourceMinimum=-90, TargetMaximum=40, TargetMinimum=10, bClamp=true, Value=@n11.Result.Z).Result
     Euler.Y=&RigVMFunction_MathFloatRemap(SourceMaximum=45, SourceMinimum=1, TargetMaximum=-40, TargetMinimum=-5, bClamp=true, Value=@n12.Result.Y).Result
    ).Result
   )
   n13 = &RigVMFunction_MathQuaternionToEuler(RotationOrder=ZYX, Value=&RigUnit_GetBoneTransform(Bone=dyn_bk_flap_r, Space=LocalSpace).Transform.Rotation)
   n14 = &RigVMFunction_MathQuaternionToEuler(RotationOrder=ZYX, Value=&RigUnit_GetRelativeBoneTransform(Bone=dyn_bk_flap_r, Space=pelvis).Transform.Rotation)
   > &RigUnit_ModifyBoneTransforms(
    BoneToModify=((Bone="dyn_bk_jacket_r",Transform=(Rotation=(X=0,Y=0,Z=0,W=1),Translation=(X=0,Y=0,Z=0),Scale3D=(X=1,Y=1,Z=1))))
    Mode=AdditiveLocal
    Weight=1
    WeightMaximum=1
    WeightMinimum=0
    BoneToModify.0.Transform.Rotation=&RigVMFunction_MathQuaternionFromEuler(
     Euler=(X=0,Y=0,Z=0)
     Euler.Z=&RigVMFunction_MathFloatRemap(SourceMaximum=-75, SourceMinimum=-90, TargetMaximum=-40, TargetMinimum=-10, bClamp=true, Value=@n13.Result.Z).Result
     Euler.Y=&RigVMFunction_MathFloatRemap(SourceMaximum=45, SourceMinimum=1, TargetMaximum=-40, TargetMinimum=-5, bClamp=true, Value=@n14.Result.Y).Result
    ).Result
   )
   n15 = &RigVMFunction_MathQuaternionToEuler(RotationOrder=ZYX, Value=&RigUnit_GetRelativeBoneTransform(Bone=dyn_l_flap_b, Space=pelvis).Transform.Rotation)
   > &RigUnit_ModifyBoneTransforms(
    BoneToModify=((Bone="dyn_l_jacket_b",Transform=(Rotation=(X=0,Y=0,Z=0,W=1),Translation=(X=0,Y=0,Z=0),Scale3D=(X=1,Y=1,Z=1))))
    Mode=AdditiveLocal
    Weight=1
    WeightMaximum=1
    WeightMinimum=0
    BoneToModify.0.Transform.Rotation=&RigVMFunction_MathQuaternionFromEuler(
     Euler=(X=0,Y=0,Z=0)
     Euler.Y=&RigVMFunction_MathFloatRemap(SourceMaximum=45, SourceMinimum=1, TargetMaximum=-20, TargetMinimum=0, bClamp=true, Value=@n15.Result.Y).Result
    ).Result
   )
   n16 = &RigVMFunction_MathQuaternionToEuler(RotationOrder=ZYX, Value=&RigUnit_GetRelativeBoneTransform(Bone=dyn_l_flap_a, Space=pelvis).Transform.Rotation)
   > &RigUnit_ModifyBoneTransforms(
    BoneToModify=((Bone="dyn_l_jacket_a",Transform=(Rotation=(X=0,Y=0,Z=0,W=1),Translation=(X=0,Y=0,Z=0),Scale3D=(X=1,Y=1,Z=1))))
    Mode=AdditiveLocal
    Weight=1
    WeightMaximum=1
    WeightMinimum=0
    BoneToModify.0.Transform.Rotation=&RigVMFunction_MathQuaternionFromEuler(
     Euler=(X=0,Y=0,Z=0)
     Euler.Y=&RigVMFunction_MathFloatRemap(SourceMaximum=45, SourceMinimum=1, TargetMaximum=-20, TargetMinimum=0, bClamp=true, Value=@n16.Result.Y).Result
    ).Result
   )
   n17 = &RigVMFunction_MathQuaternionToEuler(RotationOrder=ZYX, Value=&RigUnit_GetRelativeBoneTransform(Bone=dyn_r_flap_b, Space=pelvis).Transform.Rotation)
   > &RigUnit_ModifyBoneTransforms(
    BoneToModify=((Bone="dyn_r_jacket_b",Transform=(Rotation=(X=0,Y=0,Z=0,W=1),Translation=(X=0,Y=0,Z=0),Scale3D=(X=1,Y=1,Z=1))))
    Mode=AdditiveLocal
    Weight=1
    WeightMaximum=1
    WeightMinimum=0
    BoneToModify.0.Transform.Rotation=&RigVMFunction_MathQuaternionFromEuler(
     Euler=(X=0,Y=0,Z=0)
     Euler.Y=&RigVMFunction_MathFloatRemap(SourceMaximum=45, SourceMinimum=1, TargetMaximum=-20, TargetMinimum=0, bClamp=true, Value=@n17.Result.Y).Result
    ).Result
   )
   n18 = &RigVMFunction_MathQuaternionToEuler(RotationOrder=ZYX, Value=&RigUnit_GetRelativeBoneTransform(Bone=dyn_r_flap_a, Space=pelvis).Transform.Rotation)
   > &RigUnit_ModifyBoneTransforms(
    BoneToModify=((Bone="dyn_r_jacket_a",Transform=(Rotation=(X=0,Y=0,Z=0,W=1),Translation=(X=0,Y=0,Z=0),Scale3D=(X=1,Y=1,Z=1))))
    Mode=AdditiveLocal
    Weight=1
    WeightMaximum=1
    WeightMinimum=0
    BoneToModify.0.Transform.Rotation=&RigVMFunction_MathQuaternionFromEuler(
     Euler=(X=0,Y=0,Z=0)
     Euler.Y=&RigVMFunction_MathFloatRemap(SourceMaximum=45, SourceMinimum=1, TargetMaximum=-20, TargetMinimum=0, bClamp=true, Value=@n18.Result.Y).Result
    ).Result
   )
   n19 = &RigVMFunction_MathQuaternionToEuler(RotationOrder=ZYX, Value=&RigUnit_GetRelativeBoneTransform(Bone=dyn_fr_flap, Space=pelvis).Transform.Rotation)
   > &RigUnit_ModifyBoneTransforms(
    BoneToModify=((Bone="dyn_fr_jacket",Transform=(Rotation=(X=0,Y=0,Z=0,W=1),Translation=(X=0,Y=0,Z=0),Scale3D=(X=1,Y=1,Z=1))))
    Mode=AdditiveLocal
    Weight=1
    WeightMaximum=1
    WeightMinimum=0
    BoneToModify.0.Transform.Rotation=&RigVMFunction_MathQuaternionFromEuler(
     Euler=(X=0,Y=0,Z=0)
     Euler.Y=&RigVMFunction_MathFloatRemap(SourceMaximum=80, SourceMinimum=12, TargetMaximum=-50, TargetMinimum=-10, bClamp=true, Value=@n19.Result.Y).Result
    ).Result
   )
   n20 = &RigVMFunction_MathQuaternionToEuler(RotationOrder=ZYX, Value=&RigUnit_GetBoneTransform(Bone=dyn_fr_flap_l, Space=LocalSpace).Transform.Rotation)
   n21 = &RigVMFunction_MathQuaternionToEuler(RotationOrder=ZYX, Value=&RigUnit_GetRelativeBoneTransform(Bone=dyn_fr_flap_l, Space=pelvis).Transform.Rotation)
   > &RigUnit_ModifyBoneTransforms(
    BoneToModify=((Bone="dyn_fr_jacket_l",Transform=(Rotation=(X=0,Y=0,Z=0,W=1),Translation=(X=0,Y=0,Z=0),Scale3D=(X=1,Y=1,Z=1))))
    Mode=AdditiveLocal
    Weight=1
    WeightMaximum=1
    WeightMinimum=0
    BoneToModify.0.Transform.Rotation=&RigVMFunction_MathQuaternionFromEuler(
     Euler=(X=0,Y=0,Z=0)
     Euler.Z=&RigVMFunction_MathFloatRemap(SourceMaximum=75, SourceMinimum=90, TargetMaximum=-30, TargetMinimum=0, bClamp=true, Value=@n20.Result.Z).Result
     Euler.Y=&RigVMFunction_MathFloatRemap(SourceMaximum=70, SourceMinimum=20, TargetMaximum=-40, TargetMinimum=0, bClamp=true, Value=@n21.Result.Y).Result
    ).Result
   )
   n22 = &RigVMFunction_MathQuaternionToEuler(RotationOrder=ZYX, Value=&RigUnit_GetBoneTransform(Bone=dyn_fr_flap_r, Space=LocalSpace).Transform.Rotation)
   n23 = &RigVMFunction_MathQuaternionToEuler(RotationOrder=ZYX, Value=&RigUnit_GetRelativeBoneTransform(Bone=dyn_fr_flap_r, Space=pelvis).Transform.Rotation)
   > &RigUnit_ModifyBoneTransforms(
    BoneToModify=((Bone="dyn_fr_jacket_r",Transform=(Rotation=(X=0,Y=0,Z=0,W=1),Translation=(X=0,Y=0,Z=0),Scale3D=(X=1,Y=1,Z=1))))
    Mode=AdditiveLocal
    Weight=1
    WeightMaximum=1
    WeightMinimum=0
    BoneToModify.0.Transform.Rotation=&RigVMFunction_MathQuaternionFromEuler(
     Euler=(X=0,Y=0,Z=0)
     Euler.Z=&RigVMFunction_MathFloatRemap(SourceMaximum=75, SourceMinimum=90, TargetMaximum=30, TargetMinimum=0, bClamp=true, Value=@n22.Result.Z).Result
     Euler.Y=&RigVMFunction_MathFloatRemap(SourceMaximum=70, SourceMinimum=20, TargetMaximum=-40, TargetMinimum=0, bClamp=true, Value=@n23.Result.Y).Result
    ).Result
   )
  ]}
]
hierarchy [
 root {loc=<0 0 0>, rot=<0 0 0 1>}
 pelvis {parent=root, loc=<0 -0.7071 93.1191>, rot=<0.0127 0.707 -0.0127 -0.707>}
 spine_01 {parent=pelvis, loc=<6.7494 -0.6388 0>, rot=<0 0 0.034 -0.9994>}
 spine_02 {parent=spine_01, loc=<7.2518 -0.035 0>, rot=<0 0 0.0238 -0.9997>}
 spine_03 {parent=spine_02, loc=<6.0104 0 0>, rot=<0 0 0.0149 -0.9999>}
 spine_04 {parent=spine_03, loc=<8.1024 0 0>, rot=<0 0 -0.0132 -0.9999>}
 spine_05 {parent=spine_04, loc=<10.0146 0 0>, rot=<0 0 -0.0015 -1>}
 clavicle_l {parent=spine_05, loc=<4.3575 -0.1466 -2.3223>, rot=<0.0465 -0.7289 -0.009 -0.6829>}
 upperarm_l {parent=clavicle_l, loc=<15.5111 0 0>, rot=<0.0226 -0.3234 -0.0088 -0.9459>}
 lowerarm_l {parent=upperarm_l, loc=<26.6514 0 0>, rot=<0.0391 -0.0099 -0.2453 -0.9686>}
 hand_l {parent=lowerarm_l, loc=<22.3491 0 0>, rot=<-0.7081 -0.0507 0.0975 -0.6975>}
 index_metacarpal_l {parent=hand_l, loc=<1.6535 0.0183 -2.96>, rot=<0.0158 -0.0124 -0.0173 -0.9996>}
 index_01_l {parent=index_metacarpal_l, loc=<8.3008 0.0357 -0.0524>, rot=<-0.0044 0.0485 0.0924 -0.9945>}
 index_02_l {parent=index_01_l, loc=<3.3587 -0.3004 0.0452>, rot=<-0.002 0.0027 0.3225 -0.9466>}
 index_03_l {parent=index_02_l, loc=<2.5578 -0.123 -0.0252>, rot=<-0.002 0.0067 0.2558 -0.9667>}
 middle_metacarpal_l {parent=hand_l, loc=<2.1856 -0.1725 -0.5984>, rot=<0.006 0.0168 -0.0329 -0.9993>}
 middle_01_l {parent=middle_metacarpal_l, loc=<7.3275 -0.2447 0.0129>, rot=<-0.0208 0.0042 0.0858 -0.9961>}
 middle_02_l {parent=middle_01_l, loc=<4.0123 0.1401 0.1175>, rot=<-0.0218 0.0258 0.3512 -0.9357>}
 middle_03_l {parent=middle_02_l, loc=<2.6077 0.0534 0.1373>, rot=<0.0082 0.0188 0.1708 -0.9851>}
 pinky_metacarpal_l {parent=hand_l, loc=<1.7616 -0.9753 3.3814>, rot=<-0.0397 0.0814 -0.069 -0.9935>}
 pinky_01_l {parent=pinky_metacarpal_l, loc=<7.1709 -0.4831 -0.1094>, rot=<-0.0654 -0.0035 0.2242 -0.9723>}
 pinky_02_l {parent=pinky_01_l, loc=<2.8274 0.0893 -0.1043>, rot=<0.0037 0.0011 0.2854 -0.9584>}
 pinky_03_l {parent=pinky_02_l, loc=<2.2046 0.0307 -0.0078>, rot=<-0.0002 0.0098 0.1445 -0.9895>}
 ring_metacarpal_l {parent=hand_l, loc=<2.2074 -0.3625 1.3638>, rot=<-0.0284 0.0722 -0.0458 -0.9959>}
 ring_01_l {parent=ring_metacarpal_l, loc=<7.0295 -0.6197 -0.1446>, rot=<-0.0489 -0.0252 0.1892 -0.9804>}
 ring_02_l {parent=ring_01_l, loc=<3.6174 0.4646 -0.1027>, rot=<0 0 0.2806 -0.9598>}
 ring_03_l {parent=ring_02_l, loc=<2.4152 0.1141 0.0199>, rot=<0 0 0.1189 -0.9929>}
 thumb_01_l {parent=hand_l, loc=<1.3118 -0.3191 -2.459>, rot=<0.6546 -0.4042 0.0664 -0.6354>}
 thumb_02_l {parent=thumb_01_l, loc=<5.9336 -0.0448 0.0401>, rot=<-0.0037 0.0213 0.298 -0.9543>}
 thumb_03_l {parent=thumb_02_l, loc=<2.2223 0.3682 -0.0907>, rot=<0.0118 -0.013 0.0066 -0.9998>}
 lowerarm_twist_01_l {parent=lowerarm_l, loc=<17.8549 0 0>, rot=<0 0 0 1>}
 lowerarm_twist_02_l {parent=lowerarm_l, loc=<9.6199 0 0>, rot=<0 0 0 1>}
 upperarm_twist_01_l {parent=upperarm_l, loc=<0.0041 0 0>, rot=<0 0 0 1>}
 upperarm_twist_02_l {parent=upperarm_l, loc=<13.7941 0 0>, rot=<0 0 0 1>}
 dyn_hood_l {parent=clavicle_l, loc=<2.7995 -9.6799 6.7638>, rot=<-0.349 -0.4583 0.5375 -0.6158>}
 dyn_hood_l_01 {parent=dyn_hood_l, loc=<7.9354 -0.3357 0>, rot=<-0.0784 -0.0973 -0.0471 -0.991>}
 clavicle_r {parent=spine_05, loc=<4.3575 -0.1466 2.3223>, rot=<0.7289 0.0465 -0.6829 0.009>}
 upperarm_r {parent=clavicle_r, loc=<-15.5111 0 0>, rot=<0.0226 -0.3234 -0.0088 -0.9459>}
 lowerarm_r {parent=upperarm_r, loc=<-26.6514 0 0>, rot=<0.0391 -0.0099 -0.2453 -0.9686>}
 hand_r {parent=lowerarm_r, loc=<-22.3492 0 0>, rot=<-0.7081 -0.0507 0.0975 -0.6975>}
 index_metacarpal_r {parent=hand_r, loc=<-1.6534 -0.0181 2.9601>, rot=<0.0158 -0.0124 -0.0173 -0.9996>}
 index_01_r {parent=index_metacarpal_r, loc=<-8.3008 -0.0357 0.0524>, rot=<-0.0044 0.0485 0.0924 -0.9945>}
 index_02_r {parent=index_01_r, loc=<-3.3589 0.3003 -0.0453>, rot=<-0.002 0.0027 0.3225 -0.9466>}
 index_03_r {parent=index_02_r, loc=<-2.5578 0.1231 0.0251>, rot=<-0.002 0.0067 0.2558 -0.9667>}
 middle_metacarpal_r {parent=hand_r, loc=<-2.1856 0.1726 0.5984>, rot=<0.006 0.0168 -0.0329 -0.9993>}
 middle_01_r {parent=middle_metacarpal_r, loc=<-7.3273 0.2447 -0.0129>, rot=<-0.0208 0.0042 0.0858 -0.9961>}
 middle_02_r {parent=middle_01_r, loc=<-4.0142 -0.1411 -0.07>, rot=<-0.0217 0.0259 0.3512 -0.9357>}
 middle_03_r {parent=middle_02_r, loc=<-2.6074 -0.0534 -0.1312>, rot=<0.0083 0.0188 0.1708 -0.9851>}
 pinky_metacarpal_r {parent=hand_r, loc=<-1.7617 0.9753 -3.3814>, rot=<-0.0397 0.0814 -0.069 -0.9935>}
 pinky_01_r {parent=pinky_metacarpal_r, loc=<-7.1707 0.4832 0.1094>, rot=<-0.0654 -0.0035 0.2242 -0.9723>}
 pinky_02_r {parent=pinky_01_r, loc=<-2.8274 -0.0892 0.1043>, rot=<0.0037 0.0011 0.2854 -0.9584>}
 pinky_03_r {parent=pinky_02_r, loc=<-2.2045 -0.031 0.0078>, rot=<-0.0002 0.0098 0.1445 -0.9895>}
 ring_metacarpal_r {parent=hand_r, loc=<-2.2073 0.3627 -1.3638>, rot=<-0.0284 0.0722 -0.0458 -0.9959>}
 ring_01_r {parent=ring_metacarpal_r, loc=<-7.0297 0.6197 0.1446>, rot=<-0.0489 -0.0252 0.1892 -0.9804>}
 ring_02_r {parent=ring_01_r, loc=<-3.6172 -0.4648 0.1026>, rot=<0 0 0.2806 -0.9598>}
 ring_03_r {parent=ring_02_r, loc=<-2.4152 -0.1141 -0.0199>, rot=<0 0 0.1189 -0.9929>}
 thumb_01_r {parent=hand_r, loc=<-1.3119 0.3192 2.459>, rot=<0.6546 -0.4042 0.0664 -0.6354>}
 thumb_02_r {parent=thumb_01_r, loc=<-5.9335 0.0447 -0.0401>, rot=<-0.0037 0.0213 0.298 -0.9543>}
 thumb_03_r {parent=thumb_02_r, loc=<-2.2223 -0.3682 0.0907>, rot=<0.0118 -0.013 0.0066 -0.9998>}
 lowerarm_twist_01_r {parent=lowerarm_r, loc=<-17.855 0 0>, rot=<0 0 0 1>}
 lowerarm_twist_02_r {parent=lowerarm_r, loc=<-9.6199 0 0>, rot=<0 0 0 1>}
 upperarm_twist_01_r {parent=upperarm_r, loc=<-0.0041 0 0>, rot=<0 0 0 1>}
 upperarm_twist_02_r {parent=upperarm_r, loc=<-13.7942 0 0>, rot=<0 0 0 1>}
 dyn_hood_r {parent=clavicle_r, loc=<-3.7103 9.6371 -6.8187>, rot=<0.6397 0.5381 0.4442 -0.3224>}
 dyn_hood_r_01 {parent=dyn_hood_r, loc=<8.3355 0 0>, rot=<-0.1268 -0.1383 0.0371 -0.9815>}
 neck_01 {parent=spine_05, loc=<11.3138 0.1529 0>, rot=<0 0 -0.2003 -0.9797>}
 neck_02 {parent=neck_01, loc=<4.5627 0 0>, rot=<0 0 0.0504 -0.9987>}
 head {parent=neck_02, loc=<5.3914 0 0>, rot=<0 0 0.115 -0.9934>}
 dyn_hood_collar_fr_r {parent=spine_05, loc=<7.0095 7.6674 9.4659>, rot=<0.3499 0.3539 0.8629 -0.0881>}
 dyn_hood_collar_fr_l {parent=spine_05, loc=<7.0095 7.6674 -9.3971>, rot=<-0.0677 -0.6075 -0.395 -0.6858>}
 dyn_hood_collar_fr {parent=spine_05, loc=<3.8217 12.3442 -0.1834>, rot=<0.4795 0.5196 0.5196 0.4795>}
 dyn_hood_collar_bk {parent=spine_05, loc=<11.9751 -13.9769 0.0582>, rot=<0.5196 -0.4795 0.4795 -0.5196>}
 dyn_hood_collar_l {parent=spine_05, loc=<11.4348 0.3715 -12.326>, rot=<0.0284 -0.7065 -0.0284 -0.7065>}
 dyn_hood_collar_bk_l {parent=spine_05, loc=<11.6877 -10.4016 -10.3562>, rot=<0.3233 -0.6289 0.2718 -0.6528>}
 dyn_hood_collar_r {parent=spine_05, loc=<11.4348 0.3715 12.9572>, rot=<0.7065 0.0284 0.7065 -0.0284>}
 dyn_hood_collar_bk_r {parent=spine_05, loc=<11.7697 -10.4132 9.6588>, rot=<0.634 -0.3132 0.6068 -0.363>}
 thigh_l {parent=pelvis, loc=<-2.5693 -0.0109 -7.7501>, rot=<0.0012 0.0572 0.0155 -0.9982>}
 calf_l {parent=thigh_l, loc=<-39.0484 0 0>, rot=<0.0006 -0.0161 -0.0391 -0.9991>}
 calf_twist_01_l {parent=calf_l, loc=<-35.8131 -0.0967 0.0397>, rot=<0.0237 0.0095 -0.0035 -0.9997>}
 calf_twist_02_l {parent=calf_l, loc=<-18.8064 -0.0967 0.0397>, rot=<0.0237 0.0095 -0.0035 -0.9997>}
 foot_l {parent=calf_l, loc=<-42.561 0.1077 0.4292>, rot=<0.0418 -0.0455 0.0343 -0.9975>}
 ball_l {parent=foot_l, loc=<-6.1599 13.8264 0.6806>, rot=<0.0001 -0.0001 -0.7063 -0.7079>}
 thigh_twist_01_l {parent=thigh_l, loc=<-21.6303 0 0>, rot=<0.0013 0 0 -1>}
 dyn_l_pocket_MSTR {parent=thigh_l, loc=<0 0 0>, rot=<0 0 0 1>}
 dyn_l_b_MSTR {parent=dyn_l_pocket_MSTR, loc=<-4.0817 -3.702 -7.9803>, rot=<-0.0019 -0.7464 0.0016 -0.6655>}
 dyn_l_flap_a {parent=dyn_l_b_MSTR, loc=<0.3552 0 -0.1952>, rot=<0 -0.6749 0 -0.7379>}
 dyn_l_flap_a_01 {parent=dyn_l_flap_a, loc=<7.7327 0 0>, rot=<0 0.0321 0 -0.9995>}
 dyn_l_flap_a_02 {parent=dyn_l_flap_a_01, loc=<7.2643 0 0>, rot=<0 -0.0766 0 -0.9971>}
 dyn_l_a_MSTR {parent=dyn_l_pocket_MSTR, loc=<-4.0313 2.3056 -7.7054>, rot=<-0.0019 -0.7464 0.0016 -0.6655>}
 dyn_l_flap_b {parent=dyn_l_a_MSTR, loc=<0.3552 0 -0.1952>, rot=<0 -0.6749 0 -0.7379>}
 dyn_l_flap_b_01 {parent=dyn_l_flap_b, loc=<7.7327 0 0>, rot=<0 0.0321 0 -0.9995>}
 dyn_l_flap_b_02 {parent=dyn_l_flap_b_01, loc=<7.2643 0 0>, rot=<0 -0.0766 0 -0.9971>}
 dyn_l_jacket_b_MSTR {parent=thigh_l, loc=<3.5176 3.2025 -9.321>, rot=<-0.0024 -0.9999 -0.0001 0.0126>}
 dyn_l_jacket_b {parent=dyn_l_jacket_b_MSTR, loc=<0.7223 0 0>, rot=<0 0 0 1>}
 dyn_l_jacket_b_01 {parent=dyn_l_jacket_b, loc=<6.8818 0 0.2384>, rot=<0 0.0321 0 -0.9995>}
 dyn_fr_jacket_l_MSTR {parent=thigh_l, loc=<4.0208 10.0367 -4.5064>, rot=<-0.0692 -0.7068 -0.704 0.0116>}
 dyn_fr_jacket_l {parent=dyn_fr_jacket_l_MSTR, loc=<0.9605 0 0>, rot=<0 0 0 1>}
 dyn_fr_jacket_l_01 {parent=dyn_fr_jacket_l, loc=<7.6204 -0.4588 -0.5772>, rot=<0 -0.0156 0 -0.9999>}
 drn_fr_l_MSTR {parent=thigh_l, loc=<-2.8259 8.7545 -2.8974>, rot=<-0.0019 -0.7464 0.0016 -0.6655>}
 dyn_fr_flap_l {parent=drn_fr_l_MSTR, loc=<0 0.3178 -0.3401>, rot=<0.4804 -0.4804 -0.5188 -0.5188>}
 dyn_fr_flap_l_01 {parent=dyn_fr_flap_l, loc=<6.1527 0 0>, rot=<0 -0.0156 0 -0.9999>}
 dyn_fr_flap_l_02 {parent=dyn_fr_flap_l_01, loc=<6.6107 0 0>, rot=<-0.1096 -0.0158 0.0161 -0.9937>}
 dyn_fr_flap_l_03 {parent=dyn_fr_flap_l_02, loc=<7.0175 0 0>, rot=<-0.0517 0.0044 0.005 -0.9986>}
 dyn_fr_flap_l_04 {parent=dyn_fr_flap_l_03, loc=<5.1831 0 0>, rot=<-0.0202 -0.0016 0.0008 -0.9998>}
 dyn_fr_flap_l_05 {parent=dyn_fr_flap_l_04, loc=<4.8947 0 0>, rot=<0.0322 -0.0184 -0.0006 -0.9993>}
 thigh_r {parent=pelvis, loc=<-2.5691 -0.0109 7.7501>, rot=<0.0572 -0.0012 0.9982 0.0155>}
 calf_r {parent=thigh_r, loc=<39.0487 0 0>, rot=<0.0006 -0.0161 -0.0391 -0.9991>}
 calf_twist_01_r {parent=calf_r, loc=<35.8132 0.0967 -0.0397>, rot=<0.0237 0.0095 -0.0035 -0.9997>}
 calf_twist_02_r {parent=calf_r, loc=<18.8064 0.0967 -0.0397>, rot=<0.0237 0.0095 -0.0035 -0.9997>}
 foot_r {parent=calf_r, loc=<42.5611 -0.1077 -0.4292>, rot=<0.0418 -0.0455 0.0343 -0.9975>}
 ball_r {parent=foot_r, loc=<6.1598 -13.8264 -0.6806>, rot=<0.0001 -0.0001 -0.7063 -0.7079>}
 thigh_twist_01_r {parent=thigh_r, loc=<21.6303 0 0>, rot=<0.0013 0 0 -1>}
 dyn_r_pocket_MSTR {parent=thigh_r, loc=<0 0 0>, rot=<0 0 0 1>}
 dyn_r_a_MSTR {parent=dyn_r_pocket_MSTR, loc=<3.7745 -2.587 7.3>, rot=<-0.6655 0.0016 0.7464 0.0019>}
 dyn_r_flap_b {parent=dyn_r_a_MSTR, loc=<-0.5221 0 -0.4221>, rot=<-0.6528 0 0.7575 0>}
 dyn_r_flap_b_01 {parent=dyn_r_flap_b, loc=<7.759 0 0>, rot=<0 -0.0028 0 -1>}
 dyn_r_flap_b_02 {parent=dyn_r_flap_b_01, loc=<7.2526 0 0>, rot=<0 -0.0712 0 -0.9975>}
 dyn_r_b_MSTR {parent=dyn_r_pocket_MSTR, loc=<3.8259 3.757 7.6624>, rot=<-0.6655 0.0016 0.7464 0.0019>}
 dyn_r_flap_a {parent=dyn_r_b_MSTR, loc=<-0.5221 0 -0.4221>, rot=<-0.6528 0 0.7575 0>}
 dyn_r_flap_a_01 {parent=dyn_r_flap_a, loc=<7.759 0 0>, rot=<0 -0.0028 0 -1>}
 dyn_r_flap_a_02 {parent=dyn_r_flap_a_01, loc=<7.2526 0 0>, rot=<0 -0.0712 0 -0.9975>}
 dyn_fr_jacket_r_MSTR {parent=thigh_r, loc=<-4.1058 -10.2303 3.9043>, rot=<-0.704 -0.0116 0.0692 -0.7068>}
 dyn_fr_jacket_r {parent=dyn_fr_jacket_r_MSTR, loc=<0.9605 0 0>, rot=<0 0 0 1>}
 dyn_fr_jacket_r_01 {parent=dyn_fr_jacket_r, loc=<7.3991 0 -0.1615>, rot=<0 -0.0156 0 -0.9999>}
 dyn_r_jacket_b_MSTR {parent=thigh_r, loc=<-4.0185 -3.7076 9.9067>, rot=<0 0.0169 0.0024 -0.9999>}
 dyn_r_jacket_b {parent=dyn_r_jacket_b_MSTR, loc=<1.145 0 0>, rot=<0 0 0 1>}
 dyn_r_jacket_b_01 {parent=dyn_r_jacket_b, loc=<6.5201 0 -0.4372>, rot=<0 -0.0028 0 -1>}
 dyn_fr_r_MSTR {parent=thigh_r, loc=<2.85 -8.7546 3.1061>, rot=<-0.6655 0.0016 0.7464 0.0019>}
 dyn_fr_flap_r {parent=dyn_fr_r_MSTR, loc=<0 0.3178 -0.3401>, rot=<0.4804 -0.4804 -0.5188 -0.5188>}
 dyn_fr_flap_r_01 {parent=dyn_fr_flap_r, loc=<6.1527 0 0>, rot=<0 -0.0156 0 -0.9999>}
 dyn_fr_flap_r_02 {parent=dyn_fr_flap_r_01, loc=<6.6107 0 0>, rot=<0.0634 -0.0223 0.0028 -0.9977>}
 dyn_fr_flap_r_03 {parent=dyn_fr_flap_r_02, loc=<7.1776 0 0>, rot=<-0.0049 0.0034 -0.0261 -0.9996>}
 dyn_fr_flap_r_04 {parent=dyn_fr_flap_r_03, loc=<6.1585 0 0>, rot=<-0.0176 -0.0005 -0.0003 -0.9998>}
 dyn_fr_flap_r_05 {parent=dyn_fr_flap_r_04, loc=<6.1535 -0.0179 -0.0018>, rot=<0.0045 -0.0024 0.0058 -1>}
 dyn_bk_l_MSTR {parent=pelvis, loc=<-5.0826 -9.8847 -11.1506>, rot=<-0.0127 -0.707 0.0127 -0.707>}
 dyn_bk_flap_l {parent=dyn_bk_l_MSTR, loc=<0 -0.6743 -0.6735>, rot=<-0.4816 -0.4816 0.5178 -0.5178>}
 dyn_bk_flap_l_01 {parent=dyn_bk_flap_l, loc=<6.143 0 0>, rot=<0 -0.0123 0 -0.9999>}
 dyn_bk_flap_l_02 {parent=dyn_bk_flap_l_01, loc=<6.5786 0 0>, rot=<0.1389 -0.036 0.02 -0.9894>}
 dyn_bk_flap_l_03 {parent=dyn_bk_flap_l_02, loc=<4.2716 0 0>, rot=<-0.0741 -0.0089 0.0991 -0.9923>}
 dyn_bk_flap_l_04 {parent=dyn_bk_flap_l_03, loc=<7.1459 -0.7556 0>, rot=<-0.0865 0.0144 0.1243 -0.9884>}
 dyn_bk_flap_l_05 {parent=dyn_bk_flap_l_04, loc=<6.6545 -1.1512 -0.053>, rot=<-0.1362 0.0585 0.2618 -0.9537>}
 dyn_bk_flap_l_06 {parent=dyn_bk_flap_l_05, loc=<4.5188 -0.3264 0>, rot=<0.7063 0.2638 -0.4211 -0.5042>}
 dyn_bk_MSTR {parent=pelvis, loc=<-5.1881 -12.8269 0.0931>, rot=<-0.0127 -0.707 0.0127 -0.707>}
 dyn_bk_flap {parent=dyn_bk_MSTR, loc=<0 -0.6743 -0.6735>, rot=<-0.5 -0.5 0.5 -0.5>}
 dyn_bk_flap_01 {parent=dyn_bk_flap, loc=<6.1269 0 0>, rot=<0 0 0 1>}
 dyn_bk_flap_02 {parent=dyn_bk_flap_01, loc=<6.571 0 0>, rot=<0 0 0 1>}
 dyn_BK_r_MSTR {parent=pelvis, loc=<-5.0826 -9.8847 11.4497>, rot=<-0.0127 -0.707 0.0127 -0.707>}
 dyn_bk_flap_r {parent=dyn_BK_r_MSTR, loc=<0 -0.6743 -0.6735>, rot=<-0.4816 -0.4816 0.5178 -0.5178>}
 dyn_bk_flap_r_01 {parent=dyn_bk_flap_r, loc=<6.143 0 0>, rot=<0 -0.0123 0 -0.9999>}
 dyn_bk_flap_r_02 {parent=dyn_bk_flap_r_01, loc=<6.5786 0 0>, rot=<-0.1172 -0.0381 -0.0052 -0.9924>}
 dyn_bk_flap_r_03 {parent=dyn_bk_flap_r_02, loc=<4.3651 0 0>, rot=<-0.0208 -0.0071 -0.1356 -0.9905>}
 dyn_bk_flap_r_04 {parent=dyn_bk_flap_r_03, loc=<7.3565 0.7146 0>, rot=<0.0718 -0.0352 -0.1347 -0.9876>}
 dyn_bk_flap_r_05 {parent=dyn_bk_flap_r_04, loc=<6.8901 1.0416 -0.0946>, rot=<0.003 -0.0519 -0.234 -0.9708>}
 dyn_bk_flap_r_06 {parent=dyn_bk_flap_r_05, loc=<4.6576 0.5404 0.0228>, rot=<0.6436 -0.3203 0.384 -0.5794>}
 dyn_fr_MSTR {parent=pelvis, loc=<-4.699 10.8877 -0.0961>, rot=<-0.0127 -0.707 0.0127 -0.707>}
 dyn_fr_flap {parent=dyn_fr_MSTR, loc=<0 0.3178 -0.3401>, rot=<0.4974 -0.4974 -0.5026 -0.5026>}
 dyn_fr_flap_01 {parent=dyn_fr_flap, loc=<6.1349 0 0>, rot=<0 0.0087 0 -1>}
 dyn_fr_flap_02 {parent=dyn_fr_flap_01, loc=<6.6064 0 0>, rot=<0 -0.0138 0 -0.9999>}
 dyn_fr_jacket_MSTR {parent=pelvis, loc=<2.5888 12.6196 -0.0961>, rot=<-0.0163 -0.7069 -0.7069 -0.0163>}
 dyn_fr_jacket {parent=dyn_fr_jacket_MSTR, loc=<0.9605 0 0>, rot=<0 0 0 1>}
 dyn_fr_jacket_01 {parent=dyn_fr_jacket, loc=<7.8558 0 0.2948>, rot=<0 0.0087 0 -1>}
 dyn_r_jacket_a_MSTR {parent=pelvis, loc=<2.0538 -5.8754 16.5967>, rot=<0.074 -0.0013 0.9971 0.0179>}
 dyn_r_jacket_a {parent=dyn_r_jacket_a_MSTR, loc=<1.145 0 0>, rot=<0 0 0 1>}
 dyn_r_jacket_a_01 {parent=dyn_r_jacket_a, loc=<6.3402 0 -0.1598>, rot=<0 -0.0028 0 -1>}
 dyn_l_jacket_a_MSTR {parent=pelvis, loc=<1.8136 -4.8493 -16.8855>, rot=<-0.0179 -0.9988 0.0008 -0.0446>}
 dyn_l_jacket_a {parent=dyn_l_jacket_a_MSTR, loc=<0.7223 0 0>, rot=<0 0 0 1>}
 dyn_l_jacket_a_01 {parent=dyn_l_jacket_a, loc=<6.8818 0 0.2384>, rot=<0 0.0321 0 -0.9995>}
 dyn_bk_jacket_r_MSTR {parent=pelvis, loc=<2.1439 -11.9856 10.2656>, rot=<0.0129 -0.707 0.707 -0.0129>}
 dyn_bk_jacket_r {parent=dyn_bk_jacket_r_MSTR, loc=<1.3004 0 0.4823>, rot=<0 0 0 1>}
 dyn_bk_jacket_r_01 {parent=dyn_bk_jacket_r, loc=<6.7399 0 0.4059>, rot=<0 -0.0123 0 -0.9999>}
 dyn_bk_jacket_MSTR {parent=pelvis, loc=<2.9756 -15.0265 -0.4413>, rot=<-0.0179 -0.9998 0 0>}
 dyn_bk_jacket {parent=dyn_bk_jacket_MSTR, loc=<1.3004 0 0>, rot=<-0.7071 0 0 -0.7071>}
 dyn_bk_jacket_01 {parent=dyn_bk_jacket, loc=<7.2732 -0.5716 0.6405>, rot=<0 0 0 1>}
 dyn_bk_jacket_l_MSTR {parent=pelvis, loc=<2.2542 -12.8014 -9.785>, rot=<0.0129 -0.707 0.707 -0.0129>}
 dyn_bk_jacket_l {parent=dyn_bk_jacket_l_MSTR, loc=<1.3004 0 0.0396>, rot=<0 0 0 1>}
 dyn_bk_jacket_l_01 {parent=dyn_bk_jacket_l, loc=<6.7368 0 0.2823>, rot=<0 -0.0123 0 -0.9999>}
 dyn_zipper_r_MSTR {parent=pelvis, loc=<-1.8528 8.5726 16.7488>, rot=<0.7539 0.3795 0.5248 0.1102>}
 dyn_zipper_r {parent=dyn_zipper_r_MSTR, loc=<0 0 0>, rot=<0 0 0 1>}
 dyn_zipper_l_MSTR {parent=pelvis, loc=<-1.8528 8.5726 -16.7455>, rot=<-0.095 -0.513 -0.4183 -0.7435>}
 dyn_zipper_l {parent=dyn_zipper_l_MSTR, loc=<0 0 0>, rot=<0 0 0 1>}
]
