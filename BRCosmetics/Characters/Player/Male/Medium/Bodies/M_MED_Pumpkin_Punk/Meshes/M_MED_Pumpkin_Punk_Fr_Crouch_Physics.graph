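# Physics asset for the front crotch flaps of the Pumpkin Punk medium male player body.
# Capsule chains form left and right flaps with angular limits, alongside a short center flap, each starting from a small master body.
# The center and right masters link to the pelvis and the left to the left thigh.
# Pelvis, spine and thigh shapes supply body volume, and a free-swinging constraint cross-links the two side chains.
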PhysicsAsset /BRCosmetics/Characters/Player/Male/Medium/Bodies/M_MED_Pumpkin_Punk/Meshes/M_MED_Pumpkin_Punk_Fr_Crouch_Physics
bodies {
 drn_fr_l_MSTR [{shapeType=Capsule, length=0.505, radius=0.505}]
 dyn_fr_MSTR [{shapeType=Capsule, length=0.505, radius=1.5714}]
 dyn_fr_flap [{shapeType=Capsule, length=8, radius=3, center=<2.8071 0 -0.0001>, rotation=<rot 0 0 90.1504>}]
 dyn_fr_flap_l [{shapeType=Capsule, length=2.5, radius=3.25, center=<2.6522 0 0>, rotation=<rot 0 0 76.7751>}]
 dyn_fr_flap_l_01 [{shapeType=Capsule, length=8, radius=3.5, center=<2.1361 4.6913 2.5245>, rotation=<rot 0.0001 0 85>}]
 dyn_fr_flap_l_02 [{shapeType=Capsule, length=8, radius=3.5, center=<2.1361 5.2419 1.412>, rotation=<rot 0.0008 0.0002 95>}]
 dyn_fr_flap_l_03 [{shapeType=Capsule, length=3, radius=3, center=<1.4013 0 0.0086>, rotation=<rot 89.6231 -179.9994 -179.9995>}]
 dyn_fr_flap_l_04 [{shapeType=Capsule, length=3, radius=3, center=<1.9143 0 0.0128>, rotation=<rot 89.5975 -179.9992 -179.9992>}]
 dyn_fr_flap_l_05 [{shapeType=Capsule, length=3, radius=3, center=<2.3148 0 0.017>, rotation=<rot 89.567 -179.9983 -179.9984>}]
 dyn_fr_flap_r [{shapeType=Capsule, length=2.75, radius=3, center=<2.571 0 0>, rotation=<rot 0 0 102.429>}]
 dyn_fr_flap_r_01 [{shapeType=Capsule, length=8, radius=3.5, center=<2.1361 -4.8476 1.6625>, rotation=<rot 0 0 -85>}]
 dyn_fr_flap_r_02 [{shapeType=Capsule, length=8, radius=3.5, center=<2.1361 -5.0901 1.2481>, rotation=<rot 0 0 -85>}]
 dyn_fr_flap_r_03 [{shapeType=Capsule, length=3, radius=3, center=<1.896 0 0.0108>, rotation=<rot 89.6649 179.9995 179.9996>}]
 dyn_fr_flap_r_04 [{shapeType=Capsule, length=3, radius=3, center=<2.3149 0 0.0131>, rotation=<rot 89.6761 -179.9996 -179.9996>}]
 dyn_fr_flap_r_05 [{shapeType=Capsule, length=3, radius=3, center=<2.3149 0 0.013>, rotation=<rot 89.678 179.9998 179.9998>}]
 dyn_fr_r_MSTR [{shapeType=Capsule, length=0.505, radius=0.505}]
 pelvis [
  {shapeType=Sphere, radius=9.25, shapeName=pelvis_sphere, center=<8.3408 11.6708 0>}
  {shapeType=Capsule, length=9.4391, radius=10.6601, center=<-4.365 -0.8914 0.0075>}
 ]
 spine_01 [{shapeType=Capsule, length=10.4816, radius=13.0083, rotation=<rot 0 -90 0>}]
 thigh_l [{shapeType=Capsule, length=25.1921, radius=9.25, center=<-25.3384 -0.9222 0.9859>, rotation=<rot 0.0002 -100.0005 -89.9995>}]
 thigh_r [{shapeType=Capsule, length=25.8042, radius=9.5, center=<24.7934 0.7041 -0.7421>, rotation=<rot 0.0001 -89.9994 -89.9999>}]
}
constraints [
 {
  bone1Name=thigh_r
  bone2Name=pelvis
  swing1LimitDegrees=45
  swing1Motion=Free
  swing2LimitDegrees=45
  swing2Motion=Free
  twistLimitDegrees=45
  twistMotion=Free
 }
 {
  bone1Name=dyn_fr_MSTR
  bone2Name=pelvis
  swing1LimitDegrees=45
  swing1Motion=Free
  swing2LimitDegrees=45
  swing2Motion=Free
  twistLimitDegrees=45
  twistMotion=Free
 }
 {bone1Name=dyn_fr_flap, bone2Name=dyn_fr_MSTR, swing1LimitDegrees=20, swing2LimitDegrees=10, twistLimitDegrees=5}
 {bone1Name=dyn_fr_flap_l_05, bone2Name=dyn_fr_flap_l_04, swing1LimitDegrees=30, swing2LimitDegrees=30, twistLimitDegrees=15}
 {bone1Name=dyn_fr_flap_l, bone2Name=drn_fr_l_MSTR, swing1LimitDegrees=30, swing2LimitDegrees=30, twistLimitDegrees=5}
 {bone1Name=dyn_fr_flap_l_01, bone2Name=dyn_fr_flap_l, swing1LimitDegrees=30, swing2LimitDegrees=30, twistLimitDegrees=15}
 {bone1Name=dyn_fr_flap_l_02, bone2Name=dyn_fr_flap_l_01, swing1LimitDegrees=30, swing2LimitDegrees=30, twistLimitDegrees=15}
 {bone1Name=dyn_fr_flap_l_03, bone2Name=dyn_fr_flap_l_02, swing1LimitDegrees=40, swing2LimitDegrees=40, twistLimitDegrees=5}
 {bone1Name=dyn_fr_flap_l_04, bone2Name=dyn_fr_flap_l_03, swing1LimitDegrees=30, swing2LimitDegrees=30, twistLimitDegrees=15}
 {bone1Name=dyn_fr_flap_r, bone2Name=dyn_fr_r_MSTR, swing1LimitDegrees=10, swing2LimitDegrees=20, twistLimitDegrees=15}
 {bone1Name=dyn_fr_flap_r_01, bone2Name=dyn_fr_flap_r, swing1LimitDegrees=10, swing2LimitDegrees=20, twistLimitDegrees=15}
 {bone1Name=dyn_fr_flap_r_02, bone2Name=dyn_fr_flap_r_01, swing1LimitDegrees=10, swing2LimitDegrees=20, twistLimitDegrees=15}
 {bone1Name=dyn_fr_flap_r_03, bone2Name=dyn_fr_flap_r_02, swing1LimitDegrees=10, swing2LimitDegrees=20, twistLimitDegrees=15}
 {bone1Name=dyn_fr_flap_r_04, bone2Name=dyn_fr_flap_r_03, swing1LimitDegrees=30, swing2LimitDegrees=30, twistLimitDegrees=15}
 {bone1Name=dyn_fr_flap_r_05, bone2Name=dyn_fr_flap_r_04, swing1LimitDegrees=30, swing2LimitDegrees=30, twistLimitDegrees=15}
 {
  bone1Name=drn_fr_l_MSTR
  bone2Name=thigh_l
  swing1LimitDegrees=45
  swing1Motion=Free
  swing2LimitDegrees=45
  swing2Motion=Free
  twistLimitDegrees=45
  twistMotion=Free
 }
 {
  bone1Name=dyn_fr_r_MSTR
  bone2Name=pelvis
  swing1LimitDegrees=45
  swing1Motion=Free
  swing2LimitDegrees=45
  swing2Motion=Free
  twistLimitDegrees=45
  twistMotion=Free
 }
 {
  bone1Name=thigh_l
  bone2Name=pelvis
  swing1LimitDegrees=45
  swing1Motion=Free
  swing2LimitDegrees=45
  swing2Motion=Free
  twistLimitDegrees=45
  twistMotion=Free
 }
 {
  bone1Name=spine_01
  bone2Name=pelvis
  swing1LimitDegrees=45
  swing1Motion=Free
  swing2LimitDegrees=45
  swing2Motion=Free
  twistLimitDegrees=45
  twistMotion=Free
 }
 {
  bone1Name=dyn_fr_flap_l_02
  bone2Name=dyn_fr_flap_r_02
  swing1LimitDegrees=45
  swing1Motion=Free
  swing2LimitDegrees=45
  swing2Motion=Free
  twistLimitDegrees=45
  twistMotion=Free
 }
]
