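# Physics asset for the hood of the Pumpkin Punk hoodie on the medium male player body, covering its secondary-motion bones.
# Large capsules on the head, spine_04 and spine_05 form the body volumes, joined by constraints with free swing and twist.
# Two-bone left and right hood chains use small capsules, and the back collar uses a capsule flanked by two spheres.
# Angle-limited constraints anchor each hood chain and collar bone to spine_05, with tighter swing on the collar.
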PhysicsAsset /BRCosmetics/Characters/Player/Male/Medium/Bodies/M_MED_Pumpkin_Punk/Meshes/M_MED_Pumpkin_Punk_Hoodie_Physics
bodies {
 dyn_hood_collar_bk [{
   shapeType=Capsule
   length=8.2562
   radius=4.0923
   shapeName=dyn_hood_collar_bk_capsule
   center=<2.3336 -0.0323 5.5202>
   rotation=<rot 0 177.2501 -89.3099>
  }]
 dyn_hood_collar_bk_l [{shapeType=Sphere, radius=4.4356, shapeName=dyn_hood_collar_bk_l_sphere, center=<3.6177 0.7163 4.9199>}]
 dyn_hood_collar_bk_r [{shapeType=Sphere, radius=3.6293, shapeName=dyn_hood_collar_bk_r_sphere, center=<3.3797 -0.5839 4.6851>}]
 dyn_hood_l [{shapeType=Capsule, length=7, radius=2, shapeName=dyn_hood_l_capsule, center=<3.0249 -0.538 3.051>, rotation=<rot -0.0005 0 80>}]
 dyn_hood_l_01 [{shapeType=Capsule, length=7, radius=2, shapeName=dyn_hood_l_01_capsule, center=<2.1096 0 0.3811>, rotation=<rot 0 0 90>}]
 dyn_hood_r [{shapeType=Capsule, length=7, radius=2, shapeName=dyn_hood_r_capsule, center=<2.9899 -1.0605 2.9469>, rotation=<rot -0.8684 -4.9241 75.0373>}]
 dyn_hood_r_01 [{shapeType=Capsule, length=7, radius=2, shapeName=dyn_hood_r_01_capsule, center=<2.1096 0 0.3811>, rotation=<rot 0 0 90>}]
 head [{shapeType=Capsule, length=9.25, radius=11.5, shapeName=head_capsule, center=<2.646 -4.6252 0>, rotation=<rot 0 -90 0>}]
 spine_04 [{shapeType=Capsule, length=9.7926, radius=14.384, shapeName=spine_04_capsule, center=<1.4585 0.1551 0.6884>}]
 spine_05 [{shapeType=Capsule, length=17.7476, radius=11.2727, shapeName=spine_05_capsule, center=<3.6047 -1.5009 0.352>}]
}
constraints [
 {
  bone1Name=spine_05
  bone2Name=spine_04
  swing1LimitDegrees=45
  swing1Motion=Free
  swing2LimitDegrees=45
  swing2Motion=Free
  twistLimitDegrees=45
  twistMotion=Free
 }
 {
  bone1Name=head
  bone2Name=spine_05
  swing1LimitDegrees=45
  swing1Motion=Free
  swing2LimitDegrees=45
  swing2Motion=Free
  twistLimitDegrees=45
  twistMotion=Free
 }
 {bone1Name=dyn_hood_r, bone2Name=spine_05, swing1LimitDegrees=30, swing2LimitDegrees=45, twistLimitDegrees=20}
 {bone1Name=dyn_hood_r_01, bone2Name=dyn_hood_r, swing1LimitDegrees=30, swing2LimitDegrees=45, twistLimitDegrees=10}
 {bone1Name=dyn_hood_l_01, bone2Name=dyn_hood_l, swing1LimitDegrees=30, swing2LimitDegrees=45, twistLimitDegrees=20}
 {bone1Name=dyn_hood_l, bone2Name=spine_05, swing1LimitDegrees=30, swing2LimitDegrees=45, twistLimitDegrees=20}
 {bone1Name=dyn_hood_collar_bk_r, bone2Name=spine_05, swing1LimitDegrees=20, swing2LimitDegrees=30, twistLimitDegrees=20}
 {bone1Name=dyn_hood_collar_bk_l, bone2Name=spine_05, swing1LimitDegrees=20, swing2LimitDegrees=30, twistLimitDegrees=20}
 {bone1Name=dyn_hood_collar_bk, bone2Name=spine_05, swing1LimitDegrees=20, swing2LimitDegrees=30, twistLimitDegrees=20}
]
