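# Physics asset for the male medium Pumpkin Punk body cosmetic, covering dangling clothing: jacket tails, flaps, thigh-pocket flaps, hood, collar and zippers.
# Capsules and spheres on pelvis, spine, head and thighs form the body volume.
# Small MSTR anchor bodies, mostly free-constrained, link flap and jacket chains to pelvis or thighs.
# Chain segments use limited swing and twist; hood and collar attach to spine_05, zippers to pelvis.
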PhysicsAsset /BRCosmetics/Characters/Player/Male/Medium/Bodies/M_MED_Pumpkin_Punk/Meshes/M_MED_Pumpkin_Punk_Physics
bodies {
 drn_fr_l_MSTR [{shapeType=Capsule, length=0.505, radius=0.505}]
 dyn_BK_r_MSTR [{shapeType=Capsule, length=0.505, radius=0.505}]
 dyn_bk_MSTR [{shapeType=Capsule, length=0.505, radius=0.505}]
 dyn_bk_flap [{shapeType=Capsule, length=5.75, radius=1.75, center=<2.5408 0 0>, rotation=<rot 0 0 90>}]
 dyn_bk_flap_01 [{shapeType=Capsule, length=5.75, radius=1.75, center=<2.5408 0 0>, rotation=<rot 0 0 90>}]
 dyn_bk_flap_02 [{shapeType=Capsule, length=5.75, radius=1.75, center=<2.5408 0 -0.3589>, rotation=<rot 0 0 90>}]
 dyn_bk_flap_03 [{shapeType=Capsule, length=23, radius=3, shapeName=dyn_bk_flap_03_capsule, center=<1.1363 0 0.4221>, rotation=<rot 0 0 90>}]
 dyn_bk_flap_04 [{shapeType=Capsule, length=20, radius=3, shapeName=dyn_bk_flap_04_capsule, center=<0 0 -0.1993>, rotation=<rot 0 0 90>}]
 dyn_bk_flap_05 [{shapeType=Capsule, length=12, radius=3, shapeName=dyn_bk_flap_05_capsule, center=<0 0 -0.7379>, rotation=<rot 0 0 90>}]
 dyn_bk_flap_l [{shapeType=Capsule, length=4.75, radius=2.75, center=<1.6619 0 -1.2016>, rotation=<rot 0 0 90>}]
 dyn_bk_flap_l_01 [{shapeType=Capsule, length=3.5, radius=3.25, center=<1.6619 0 -0.3309>, rotation=<rot 0 0 90>}]
 dyn_bk_flap_l_02 [{shapeType=Capsule, length=5.25, radius=3.5, center=<1.6619 0 0.5747>, rotation=<rot 0 0 90>}]
 dyn_bk_flap_r [{shapeType=Capsule, length=4.75, radius=2.75, center=<1.6619 0 0>, rotation=<rot 0 0 90>}]
 dyn_bk_flap_r_01 [{shapeType=Capsule, length=3.5, radius=3.25, center=<1.6619 0 -0.6172>, rotation=<rot 0 0 90>}]
 dyn_bk_flap_r_02 [{shapeType=Capsule, length=5.25, radius=3.5, center=<1.6619 0 -1.0557>, rotation=<rot 0 0 90>}]
 dyn_bk_jacket [{shapeType=Capsule, length=5, radius=3, shapeName=dyn_bk_jacket_capsule, center=<1.3767 0 0.1764>, rotation=<rot 0 0 90>}]
 dyn_bk_jacket_01 [{shapeType=Capsule, length=7, radius=4, shapeName=dyn_bk_jacket_01_capsule, center=<0.4073 0 -0.5101>, rotation=<rot 0 0 90>}]
 dyn_bk_jacket_MSTR [{shapeType=Capsule, length=0.505, radius=0.505, shapeName=dyn_bk_jacket_MSTR_capsule}]
 dyn_bk_jacket_l [{shapeType=Capsule, length=5, radius=3, shapeName=dyn_bk_jacket_l_capsule, center=<1.3767 0 0.1764>, rotation=<rot 0 0 90>}]
 dyn_bk_jacket_l_01 [{shapeType=Capsule, length=7, radius=4, shapeName=dyn_bk_jacket_l_01_capsule, center=<0.4073 0 0.0196>, rotation=<rot 0 0 90>}]
 dyn_bk_jacket_l_MSTR [{shapeType=Capsule, length=0.505, radius=0.505, shapeName=dyn_bk_jacket_l_MSTR_capsule}]
 dyn_bk_jacket_r [{shapeType=Capsule, length=5, radius=3, shapeName=dyn_bk_jacket_r_capsule, center=<1.3767 0 0.1764>, rotation=<rot 0 0 90>}]
 dyn_bk_jacket_r_01 [{shapeType=Capsule, length=7, radius=4, shapeName=dyn_bk_jacket_r_01_capsule, center=<0.4073 0 0.0196>, rotation=<rot 0 0 90>}]
 dyn_bk_jacket_r_MSTR [{shapeType=Capsule, length=0.505, radius=0.505, shapeName=dyn_bk_jacket_r_MSTR_capsule}]
 dyn_bk_l_MSTR [{shapeType=Sphere, radius=0.505, shapeName=dyn_bk_l_MSTR_sphere}]
 dyn_fr_MSTR [{shapeType=Capsule, length=0.505, radius=1.5714}]
 dyn_fr_flap [{shapeType=Capsule, length=8.507, radius=1.993, center=<5.3986 0 0.0329>, rotation=<rot 0 0 90>}]
 dyn_fr_flap_l [{shapeType=Capsule, length=7.6166, radius=0.8, center=<3.4867 0 0>, rotation=<rot 0 0 90>}]
 dyn_fr_flap_l_01 [{shapeType=Capsule, length=7.3666, radius=1.55, center=<2.5392 0.2863 1.302>, rotation=<rot -0.576 -2.375 78.7557>}]
 dyn_fr_flap_l_02 [{shapeType=Capsule, length=8.3077, radius=1.8, center=<2.649 0 0.3584>, rotation=<rot -0.1438 1.5547 101.0227>}]
 dyn_fr_flap_l_03 [{shapeType=Capsule, length=3.3544, radius=2.2498, shapeName=dyn_fr_flap_l_03_capsule, center=<3.4883 0 0>, rotation=<rot 90 0 0>}]
 dyn_fr_flap_l_04 [{shapeType=Capsule, length=7.5992, radius=2.537, shapeName=dyn_fr_flap_l_04_capsule, center=<6.2556 0 0>, rotation=<rot 90 0 0>}]
 dyn_fr_flap_r [{shapeType=Capsule, length=4.25, radius=3.2, center=<2.5408 0 1.9887>, rotation=<rot 0 0 90>}]
 dyn_fr_flap_r_01 [{shapeType=Capsule, length=7.5, radius=3.1, center=<3.4999 0 1.3366>, rotation=<rot 0.0001 0 92.8125>}]
 dyn_fr_flap_r_02 [{shapeType=Capsule, length=10.25, radius=2.5, center=<2.9266 0 0.5642>, rotation=<rot 0.0004 0 75.9375>}]
 dyn_fr_flap_r_03 [{shapeType=Capsule, length=3.8544, radius=1.7498, shapeName=dyn_fr_flap_l_03_capsule, center=<3.4883 0 0>, rotation=<rot 90 0 0>}]
 dyn_fr_flap_r_04 [{shapeType=Capsule, length=8.3492, radius=1.787, shapeName=dyn_fr_flap_l_04_capsule, center=<6.2556 0 0>, rotation=<rot 90 0 0>}]
 dyn_fr_jacket_MSTR [{shapeType=Capsule, length=0.505, radius=0.505, shapeName=dyn_fr_jacket_MSTR_capsule}]
 dyn_fr_r_MSTR [{shapeType=Capsule, length=0.505, radius=0.505}]
 dyn_hood_collar_bk [{
   shapeType=Capsule
   length=8.2562
   radius=4.0923
   shapeName=dyn_hood_collar_bk_capsule
   center=<2.3336 -0.0323 5.5202>
   rotation=<rot 0 177.2501 -89.3099>
  }]
 dyn_hood_collar_bk_l [{shapeType=Sphere, radius=4.4356, shapeName=dyn_hood_collar_bk_l_sphere, center=<3.6177 0.7163 4.9199>}]
 dyn_hood_collar_bk_r [{shapeType=Sphere, radius=3.6293, shapeName=dyn_hood_collar_bk_r_sphere, center=<3.3797 -0.5839 4.6851>}]
 dyn_hood_l [{shapeType=Capsule, length=7, radius=2, shapeName=dyn_hood_l_capsule, center=<3.0249 -0.538 3.051>, rotation=<rot -0.0005 0.0001 80>}]
 dyn_hood_l_01 [{shapeType=Capsule, length=7, radius=2, shapeName=dyn_hood_l_01_capsule, center=<2.1096 0 0.3811>, rotation=<rot 0 0 90>}]
 dyn_hood_r [{shapeType=Capsule, length=7, radius=2, shapeName=dyn_hood_r_capsule, center=<2.9899 -1.0605 2.9469>, rotation=<rot -0.8684 -4.9241 75.0373>}]
 dyn_hood_r_01 [{shapeType=Capsule, length=7, radius=2, shapeName=dyn_hood_r_01_capsule, center=<2.1096 0 0.3811>, rotation=<rot 0 0 90>}]
 dyn_l_a_MSTR [{shapeType=Capsule, length=0.505, radius=0.505, rotation=<rot 0 -90 0>}]
 dyn_l_b_MSTR [{shapeType=Capsule, length=0.505, radius=0.505, rotation=<rot 0 -90 0>}]
 dyn_l_flap_a [{shapeType=Capsule, length=5.0506, radius=2.4494, center=<2.7381 0 -1.0962>, rotation=<rot 0 0 90>}]
 dyn_l_flap_a_01 [{shapeType=Capsule, length=4.8934, radius=2.6066, center=<2.5408 0 -0.1211>, rotation=<rot 0 0 90>}]
 dyn_l_flap_a_02 [{shapeType=Capsule, length=4.8934, radius=3.8566, center=<2.0145 0 1.1495>, rotation=<rot 0 0 90>}]
 dyn_l_flap_b [{shapeType=Capsule, length=6.1034, radius=1.3966, center=<2.7381 0.0436 1.1015>, rotation=<rot 0 0 90>}]
 dyn_l_flap_b_01 [{shapeType=Capsule, length=4.25, radius=3, center=<2.5408 0 0.5277>, rotation=<rot 0 0 90>}]
 dyn_l_flap_b_02 [{shapeType=Capsule, length=4.5, radius=4.25, center=<2.5408 0 1.8031>, rotation=<rot 0 0 90>}]
 dyn_l_pocket_MSTR [{shapeType=Capsule, length=0.505, radius=0.505, rotation=<rot 0 -90 0>}]
 dyn_r_a_MSTR [{shapeType=Capsule, length=0.505, radius=0.505}]
 dyn_r_b_MSTR [{shapeType=Capsule, length=0.505, radius=0.505}]
 dyn_r_flap_a [{shapeType=Capsule, length=6.5, radius=2.75, center=<3.4935 -0.5423 -1.9554>, rotation=<rot 0 0 90>}]
 dyn_r_flap_a_01 [{shapeType=Capsule, length=6.75, radius=3, center=<2.5408 0 0.0404>, rotation=<rot 0 0 90>}]
 dyn_r_flap_a_02 [{shapeType=Capsule, length=7, radius=2.75, center=<2.5408 0 0.3276>, rotation=<rot 0 0 90>}]
 dyn_r_flap_b [{shapeType=Capsule, length=6.5, radius=2.75, center=<3.4935 0.2376 -1.1501>, rotation=<rot 0 0 90>}]
 dyn_r_flap_b_01 [{shapeType=Capsule, length=6.5, radius=3, center=<2.5408 0 0.2084>, rotation=<rot 0 0 90>}]
 dyn_r_flap_b_02 [{shapeType=Capsule, length=6, radius=2.75, center=<2.5408 -0.935 1.2092>, rotation=<rot 0 -0.0001 85>}]
 dyn_r_pocket_MSTR [{shapeType=Capsule, length=0.505, radius=0.505}]
 dyn_zipper_l [{shapeType=Capsule, length=1.032, radius=1.968, center=<0 0 -2.2449>}]
 dyn_zipper_r [{shapeType=Capsule, length=1.032, radius=1.968, center=<0 0 -2.2449>}]
 head [{shapeType=Capsule, length=9.25, radius=11.5, shapeName=head_capsule, center=<2.646 -4.6252 0>, rotation=<rot 0 -90 0>}]
 pelvis [
  {shapeType=Capsule, length=8.7817, radius=10.8069, center=<-14.356 -1.3676 0.0074>}
  {shapeType=Capsule, length=7.5, radius=9, shapeName=pelvis_capsule, center=<-11.8085 -2.1968 0>}
 ]
 spine_01 [{shapeType=Capsule, length=8.8839, radius=11.2366, center=<-4.5139 0 0>, rotation=<rot 0 -90 0>}]
 spine_04 [{shapeType=Capsule, length=9.7926, radius=14.384, center=<1.4585 0.1551 0.6884>}]
 spine_05 [{shapeType=Capsule, length=17.7476, radius=11.2727, shapeName=spine_05_capsule, center=<3.6047 -1.5009 0.352>}]
 thigh_l [
  {shapeType=Capsule, length=29.4566, radius=8.8895, center=<-21.4474 0.765 0.656>, rotation=<rot 0.0001 -89.9996 -89.9999>}
  {shapeType=Capsule, length=35.4324, radius=8, center=<-19.7639 -2.8733 0.7313>, rotation=<rot -90 0.1582 -0.1583>}
 ]
 thigh_r [
  {shapeType=Capsule, length=19.6197, radius=8.9402, center=<26.0304 0.3938 -0.5171>, rotation=<rot 0.0001 -89.9994 -89.9999>}
  {shapeType=Capsule, length=36.3793, radius=7.1695, center=<19.6051 2.7562 -0.3151>, rotation=<rot -90 -0.0198 0.0198>}
 ]
}
constraints [
 {
  bone1Name=thigh_l
  bone2Name=pelvis
  swing1LimitDegrees=45
  swing1Motion=Free
  swing2LimitDegrees=45
  swing2Motion=Free
  twistLimitDegrees=45
  twistMotion=Free
 }
 {
  bone1Name=thigh_r
  bone2Name=pelvis
  swing1LimitDegrees=45
  swing1Motion=Free
  swing2LimitDegrees=45
  swing2Motion=Free
  twistLimitDegrees=45
  twistMotion=Free
 }
 {
  bone1Name=dyn_fr_MSTR
  bone2Name=pelvis
  swing1LimitDegrees=45
  swing1Motion=Free
  swing2LimitDegrees=45
  swing2Motion=Free
  twistLimitDegrees=45
  twistMotion=Free
 }
 {bone1Name=dyn_fr_flap, bone2Name=dyn_fr_MSTR, swing1LimitDegrees=20, swing2LimitDegrees=45, twistLimitDegrees=3}
 {
  bone1Name=dyn_bk_l_MSTR
  bone2Name=pelvis
  swing1LimitDegrees=45
  swing1Motion=Free
  swing2LimitDegrees=45
  swing2Motion=Free
  twistLimitDegrees=45
  twistMotion=Free
 }
 {
  bone1Name=dyn_bk_MSTR
  bone2Name=pelvis
  swing1LimitDegrees=45
  swing1Motion=Free
  swing2LimitDegrees=45
  swing2Motion=Free
  twistLimitDegrees=45
  twistMotion=Free
 }
 {
  bone1Name=dyn_BK_r_MSTR
  bone2Name=pelvis
  swing1LimitDegrees=45
  swing1Motion=Free
  swing2LimitDegrees=45
  swing2Motion=Free
  twistLimitDegrees=45
  twistMotion=Free
 }
 {bone1Name=dyn_bk_flap, bone2Name=dyn_bk_MSTR, swing1LimitDegrees=20, swing2LimitDegrees=30, twistLimitDegrees=5}
 {bone1Name=dyn_bk_flap_01, bone2Name=dyn_bk_flap, swing1LimitDegrees=20, swing2LimitDegrees=30, twistLimitDegrees=5, twistMotion=Locked}
 {bone1Name=dyn_bk_flap_02, bone2Name=dyn_bk_flap_01, swing1LimitDegrees=20, swing2LimitDegrees=30, twistLimitDegrees=5, twistMotion=Locked}
 {bone1Name=dyn_bk_flap_r, bone2Name=dyn_BK_r_MSTR, swing1LimitDegrees=20, swing2LimitDegrees=30, twistLimitDegrees=5}
 {bone1Name=dyn_bk_flap_r_01, bone2Name=dyn_bk_flap_r, swing1LimitDegrees=20, swing2LimitDegrees=30, twistLimitDegrees=5}
 {bone1Name=dyn_bk_flap_r_02, bone2Name=dyn_bk_flap_r_01, swing1LimitDegrees=20, swing2LimitDegrees=30, twistLimitDegrees=5}
 {bone1Name=dyn_bk_flap_l, bone2Name=dyn_bk_l_MSTR, swing1LimitDegrees=20, swing2LimitDegrees=30, twistLimitDegrees=5}
 {bone1Name=dyn_bk_flap_l_01, bone2Name=dyn_bk_flap_l, swing1LimitDegrees=20, swing2LimitDegrees=60, twistLimitDegrees=5}
 {bone1Name=dyn_bk_flap_l_02, bone2Name=dyn_bk_flap_l_01, swing1LimitDegrees=20, swing2LimitDegrees=60, twistLimitDegrees=5}
 {bone1Name=dyn_fr_flap_l, bone2Name=drn_fr_l_MSTR, swing1LimitDegrees=15, swing2LimitDegrees=20, twistLimitDegrees=5}
 {bone1Name=dyn_fr_flap_l_01, bone2Name=dyn_fr_flap_l, swing1LimitDegrees=15, swing2LimitDegrees=20, twistLimitDegrees=5}
 {bone1Name=dyn_fr_flap_l_02, bone2Name=dyn_fr_flap_l_01, swing1LimitDegrees=50, swing2LimitDegrees=50, twistLimitDegrees=10}
 {bone1Name=dyn_fr_flap_r, bone2Name=dyn_fr_r_MSTR, swing1LimitDegrees=20, swing2LimitDegrees=30, twistLimitDegrees=5}
 {bone1Name=dyn_fr_flap_r_01, bone2Name=dyn_fr_flap_r, swing1LimitDegrees=30, swing2LimitDegrees=30, twistLimitDegrees=10}
 {bone1Name=dyn_fr_flap_r_02, bone2Name=dyn_fr_flap_r_01, swing1LimitDegrees=30, swing2LimitDegrees=45, twistLimitDegrees=15}
 {
  bone1Name=drn_fr_l_MSTR
  bone2Name=pelvis
  swing1LimitDegrees=45
  swing1Motion=Free
  swing2LimitDegrees=45
  swing2Motion=Free
  twistLimitDegrees=45
  twistMotion=Free
 }
 {
  bone1Name=dyn_fr_r_MSTR
  bone2Name=pelvis
  swing1LimitDegrees=45
  swing1Motion=Free
  swing2LimitDegrees=45
  swing2Motion=Free
  twistLimitDegrees=45
  twistMotion=Free
 }
 {bone1Name=dyn_r_flap_a_02, bone2Name=dyn_r_flap_a_01, swing1LimitDegrees=30, swing2LimitDegrees=45, twistLimitDegrees=5}
 {bone1Name=dyn_r_flap_a_01, bone2Name=dyn_r_flap_a, swing1LimitDegrees=30, swing2LimitDegrees=45, twistLimitDegrees=5}
 {bone1Name=dyn_r_flap_a, bone2Name=dyn_r_b_MSTR, swing1LimitDegrees=30, swing2LimitDegrees=5, twistLimitDegrees=5}
 {bone1Name=dyn_r_flap_b_02, bone2Name=dyn_r_flap_b_01, swing1LimitDegrees=30, swing2LimitDegrees=45, twistLimitDegrees=5}
 {bone1Name=dyn_r_flap_b_01, bone2Name=dyn_r_flap_b, swing1LimitDegrees=30, swing2LimitDegrees=45, twistLimitDegrees=5}
 {bone1Name=dyn_r_flap_b, bone2Name=dyn_r_a_MSTR, swing1LimitDegrees=30, swing2LimitDegrees=5, twistLimitDegrees=5}
 {
  bone1Name=dyn_r_a_MSTR
  bone2Name=dyn_r_pocket_MSTR
  swing1LimitDegrees=45
  swing1Motion=Free
  swing2LimitDegrees=45
  swing2Motion=Free
  twistLimitDegrees=45
  twistMotion=Free
 }
 {
  bone1Name=dyn_r_pocket_MSTR
  bone2Name=thigh_r
  swing1LimitDegrees=45
  swing1Motion=Free
  swing2LimitDegrees=45
  swing2Motion=Free
  twistLimitDegrees=45
  twistMotion=Free
 }
 {bone1Name=dyn_r_b_MSTR, bone2Name=dyn_r_pocket_MSTR, swing1LimitDegrees=45, swing2LimitDegrees=45, twistLimitDegrees=15}
 {
  bone1Name=dyn_l_pocket_MSTR
  bone2Name=thigh_l
  swing1LimitDegrees=45
  swing1Motion=Free
  swing2LimitDegrees=45
  swing2Motion=Free
  twistLimitDegrees=45
  twistMotion=Free
 }
 {bone1Name=dyn_l_flap_a, bone2Name=dyn_l_b_MSTR, swing1LimitDegrees=30, swing2LimitDegrees=25, twistLimitDegrees=5}
 {bone1Name=dyn_l_flap_a_01, bone2Name=dyn_l_flap_a, swing1LimitDegrees=30, swing2LimitDegrees=45, twistLimitDegrees=5}
 {bone1Name=dyn_l_flap_a_02, bone2Name=dyn_l_flap_a_01, swing1LimitDegrees=30, swing2LimitDegrees=45, twistLimitDegrees=5}
 {bone1Name=dyn_l_flap_b, bone2Name=dyn_l_a_MSTR, swing1LimitDegrees=30, swing2LimitDegrees=25, twistLimitDegrees=5}
 {bone1Name=dyn_l_flap_b_01, bone2Name=dyn_l_flap_b, swing1LimitDegrees=30, swing2LimitDegrees=45, twistLimitDegrees=5}
 {bone1Name=dyn_l_flap_b_02, bone2Name=dyn_l_flap_b_01, swing1LimitDegrees=30, swing2LimitDegrees=45, twistLimitDegrees=5}
 {
  bone1Name=dyn_l_b_MSTR
  bone2Name=dyn_l_pocket_MSTR
  swing1LimitDegrees=45
  swing1Motion=Free
  swing2LimitDegrees=45
  swing2Motion=Free
  twistLimitDegrees=45
  twistMotion=Free
 }
 {
  bone1Name=dyn_l_a_MSTR
  bone2Name=dyn_l_pocket_MSTR
  swing1LimitDegrees=45
  swing1Motion=Free
  swing2LimitDegrees=45
  swing2Motion=Free
  twistLimitDegrees=45
  twistMotion=Free
 }
 {
  bone1Name=spine_04
  bone2Name=spine_01
  swing1LimitDegrees=45
  swing1Motion=Free
  swing2LimitDegrees=45
  swing2Motion=Free
  twistLimitDegrees=45
  twistMotion=Free
 }
 {
  bone1Name=spine_01
  bone2Name=pelvis
  swing1LimitDegrees=45
  swing1Motion=Free
  swing2LimitDegrees=45
  swing2Motion=Free
  twistLimitDegrees=45
  twistMotion=Free
 }
 {bone1Name=dyn_zipper_r, bone2Name=pelvis, swing1LimitDegrees=10, swing2LimitDegrees=45, twistLimitDegrees=1}
 {bone1Name=dyn_zipper_l, bone2Name=pelvis, swing1LimitDegrees=10, swing2LimitDegrees=45, twistLimitDegrees=5}
 {bone1Name=dyn_hood_l, bone2Name=spine_05, swing1LimitDegrees=30, swing2LimitDegrees=45, twistLimitDegrees=20}
 {bone1Name=dyn_hood_l_01, bone2Name=dyn_hood_l, swing1LimitDegrees=30, swing2LimitDegrees=45, twistLimitDegrees=10}
 {bone1Name=dyn_hood_r, bone2Name=spine_05, swing1LimitDegrees=30, swing2LimitDegrees=45, twistLimitDegrees=20}
 {bone1Name=dyn_hood_r_01, bone2Name=dyn_hood_r, swing1LimitDegrees=30, swing2LimitDegrees=45, twistLimitDegrees=10}
 {
  bone1Name=dyn_fr_jacket_MSTR
  bone2Name=pelvis
  swing1LimitDegrees=45
  swing1Motion=Free
  swing2LimitDegrees=45
  swing2Motion=Free
  twistLimitDegrees=45
  twistMotion=Free
 }
 {
  bone1Name=spine_05
  bone2Name=spine_04
  swing1LimitDegrees=45
  swing1Motion=Free
  swing2LimitDegrees=45
  swing2Motion=Free
  twistLimitDegrees=45
  twistMotion=Free
 }
 {bone1Name=dyn_bk_jacket_r, bone2Name=dyn_bk_jacket_r_MSTR, swing1LimitDegrees=50, swing2LimitDegrees=50, twistLimitDegrees=10}
 {bone1Name=dyn_bk_jacket_r_01, bone2Name=dyn_bk_jacket_r, swing1LimitDegrees=50, swing2LimitDegrees=50, twistLimitDegrees=10}
 {bone1Name=dyn_bk_jacket_l, bone2Name=dyn_bk_jacket_l_MSTR, swing1LimitDegrees=50, swing2LimitDegrees=50, twistLimitDegrees=10}
 {bone1Name=dyn_bk_jacket_l_01, bone2Name=dyn_bk_jacket_l, swing1LimitDegrees=50, swing2LimitDegrees=50, twistLimitDegrees=10}
 {bone1Name=dyn_bk_jacket, bone2Name=dyn_bk_jacket_MSTR, swing1LimitDegrees=50, swing2LimitDegrees=50, twistLimitDegrees=10}
 {bone1Name=dyn_bk_jacket_01, bone2Name=dyn_bk_jacket, swing1LimitDegrees=50, swing2LimitDegrees=50, twistLimitDegrees=10}
 {
  bone1Name=dyn_bk_jacket_MSTR
  bone2Name=pelvis
  swing1LimitDegrees=45
  swing1Motion=Free
  swing2LimitDegrees=45
  swing2Motion=Free
  twistLimitDegrees=45
  twistMotion=Free
 }
 {
  bone1Name=dyn_bk_jacket_l_MSTR
  bone2Name=pelvis
  swing1LimitDegrees=45
  swing1Motion=Free
  swing2LimitDegrees=45
  swing2Motion=Free
  twistLimitDegrees=45
  twistMotion=Free
 }
 {
  bone1Name=dyn_bk_jacket_r_MSTR
  bone2Name=pelvis
  swing1LimitDegrees=45
  swing1Motion=Free
  swing2LimitDegrees=45
  swing2Motion=Free
  twistLimitDegrees=45
  twistMotion=Free
 }
 {bone1Name=dyn_hood_collar_bk, bone2Name=spine_05, swing1LimitDegrees=20, swing2LimitDegrees=30, twistLimitDegrees=20}
 {
  bone1Name=head
  bone2Name=spine_05
  swing1LimitDegrees=45
  swing1Motion=Free
  swing2LimitDegrees=45
  swing2Motion=Free
  twistLimitDegrees=45
  twistMotion=Free
 }
 {bone1Name=dyn_hood_collar_bk_l, bone2Name=spine_05, swing1LimitDegrees=20, swing2LimitDegrees=30, twistLimitDegrees=20}
 {bone1Name=dyn_hood_collar_bk_r, bone2Name=spine_05, swing1LimitDegrees=20, swing2LimitDegrees=30, twistLimitDegrees=20}
 {bone1Name=dyn_bk_flap_03, bone2Name=dyn_bk_flap_02, swing1LimitDegrees=40, swing2LimitDegrees=60, twistLimitDegrees=10}
 {bone1Name=dyn_bk_flap_04, bone2Name=dyn_bk_flap_03, swing1LimitDegrees=40, swing2LimitDegrees=40, twistLimitDegrees=10}
 {bone1Name=dyn_bk_flap_05, bone2Name=dyn_bk_flap_04, swing1LimitDegrees=40, swing2LimitDegrees=40, twistLimitDegrees=10}
 {bone1Name=dyn_fr_flap_l_03, bone2Name=dyn_fr_flap_l_02, swing1LimitDegrees=60, swing2LimitDegrees=60, twistLimitDegrees=15}
 {bone1Name=dyn_fr_flap_l_04, bone2Name=dyn_fr_flap_l_03, swing1LimitDegrees=60, swing2LimitDegrees=60, twistLimitDegrees=15}
 {bone1Name=dyn_fr_flap_r_03, bone2Name=dyn_fr_flap_r_02, swing1LimitDegrees=60, swing2LimitDegrees=60, twistLimitDegrees=15}
 {bone1Name=dyn_fr_flap_r_04, bone2Name=dyn_fr_flap_r_03, swing1LimitDegrees=60, swing2LimitDegrees=60, twistLimitDegrees=15}
 {
  bone1Name=dyn_fr_flap_r_01
  bone2Name=dyn_fr_flap_l_01
  swing1LimitDegrees=45
  swing1Motion=Free
  swing2LimitDegrees=45
  swing2Motion=Free
  twistLimitDegrees=45
  twistMotion=Free
 }
 {
  bone1Name=dyn_l_flap_b
  bone2Name=dyn_l_flap_b_02
  swing1LimitDegrees=45
  swing1Motion=Free
  swing2LimitDegrees=45
  swing2Motion=Free
  twistLimitDegrees=45
  twistMotion=Free
 }
]
