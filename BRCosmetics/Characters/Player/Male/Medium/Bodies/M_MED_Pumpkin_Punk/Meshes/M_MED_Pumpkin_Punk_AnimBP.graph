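# Animation Blueprint for the Pumpkin Punk male medium cosmetic body, simulating and correcting its jacket, flaps, pockets and hood.
# A cached pose feeds two blended RigidBody branches: one using crouch and hoodie physics assets, the other the main asset, with falling/jumping pre-adjustments.
# Bone offsets and constraints around the simulations correct cloth shape, some driven by thigh, movement-direction and emote variables.
# Alpha-driven constraints then tie hood collar bones to spine_05.
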AnimBlueprint /BRCosmetics/Characters/Player/Male/Medium/Bodies/M_MED_Pumpkin_Punk/Meshes/M_MED_Pumpkin_Punk_AnimBP
target_skeleton=&/M_MED_Pumpkin_Punk_Skeleton
graphs [
 AnimGraph {nodes [
   n1 = "Constraint - dyn_bk_flap_l_03 (P)"(Alpha=1, ConstraintWeights_0=1) {
    alphaBoolBlend {blendOption=Linear}
    boneToModify {boneName=dyn_bk_flap_l_03}
    constraintSetup [{targetBone {boneName=thigh_l}}]
    constraintWeights [1]
   }
   "Constraint - dyn_bk_flap_r_03 (P)"(ComponentPose=@n1, Alpha=1, ConstraintWeights_0=1) {
    alphaBoolBlend {blendOption=Linear}
    boneToModify {boneName=dyn_bk_flap_r_03}
    constraintSetup [{targetBone {boneName=thigh_r}}]
    constraintWeights [1]
   }
   n2 = "Constraint - dyn_bk_flap_l_01 (P)"(Alpha=1, ConstraintWeights_0=1) {
    alphaBoolBlend {blendOption=Linear}
    boneToModify {boneName=dyn_bk_flap_l_01}
    constraintSetup [{targetBone {boneName=thigh_l}}]
    constraintWeights [1]
   }
   "Constraint - dyn_bk_flap_r_01 (P)"(ComponentPose=@n2, Alpha=1, ConstraintWeights_0=1) {
    alphaBoolBlend {blendOption=Linear}
    boneToModify {boneName=dyn_bk_flap_r_01}
    constraintSetup [{targetBone {boneName=thigh_r}}]
    constraintWeights [1]
   }
   "Constraint - dyn_bk_jacket_r_MSTR (TTTTTTTTTTT)"!(
    ComponentPose
    bAlphaBoolEnabled:bool=True
    ConstraintWeights_0=0.002
    ConstraintWeights_1:float=0.14
    ConstraintWeights_2:float=0.087
    ConstraintWeights_3:float=0.03
    ConstraintWeights_4:float=0.238
    ConstraintWeights_5:float=0.045
    ConstraintWeights_6:float=0.002
    ConstraintWeights_7:float=0.002
    ConstraintWeights_8:float=0.358
    ConstraintWeights_9:float=0.009
    ConstraintWeights_10:float=0.087
   ) -> (Pose) {
    alphaInputType=Bool
    alphaBoolBlend {blendOption=Linear}
    boneToModify {boneName=dyn_bk_jacket_r_MSTR}
    constraintSetup [
     {transformType=Translation, targetBone {boneName=dyn_bk_flap}}
     {transformType=Translation, targetBone {boneName=dyn_bk_flap_01}}
     {transformType=Translation, targetBone {boneName=dyn_bk_flap_02}}
     {transformType=Translation, targetBone {boneName=dyn_bk_flap_r}}
     {transformType=Translation, targetBone {boneName=dyn_bk_flap_r_01}}
     {transformType=Translation, targetBone {boneName=dyn_bk_flap_r_02}}
     {transformType=Translation, targetBone {boneName=dyn_bk_jacket_r}}
     {transformType=Translation, targetBone {boneName=dyn_bk_jacket_r_01}}
     {transformType=Translation, targetBone {boneName=pelvis}}
     {transformType=Translation, targetBone {boneName=thigh_twist_01_l}}
     {transformType=Translation, targetBone {boneName=thigh_twist_01_r}}
    ]
    constraintWeights [1,1,1,1,1,1,1,1,1,1,1]
   }
   n3 = "Constraint - dyn_r_jacket_a_01 (TTTT)"!(
    ComponentPose
    Alpha=1
    ConstraintWeights_0=0.005
    ConstraintWeights_1:float=0.128
    ConstraintWeights_2:float=0.863
    ConstraintWeights_3:float=0.031
   ) -> (Pose) {
    alphaBoolBlend {blendOption=Linear}
    boneToModify {boneName=dyn_r_jacket_a_01}
    constraintSetup [
     {transformType=Translation, targetBone {boneName=dyn_r_flap_a}}
     {transformType=Translation, targetBone {boneName=dyn_r_flap_a_02}}
     {transformType=Translation, targetBone {boneName=pelvis}}
     {transformType=Translation, targetBone {boneName=thigh_twist_01_r}}
    ]
    constraintWeights [1,1,1,1]
   }
   "Constraint - dyn_bk_jacket_r_01 (TTTTTTTTTT)"!(
    ComponentPose=@n3
    Alpha=1
    ConstraintWeights_0=0.001
    ConstraintWeights_1:float=0.012
    ConstraintWeights_2:float=0.027
    ConstraintWeights_3:float=0.017
    ConstraintWeights_4:float=0.024
    ConstraintWeights_5:float=0.336
    ConstraintWeights_6:float=0.43
    ConstraintWeights_7:float=0.002
    ConstraintWeights_8:float=0.132
    ConstraintWeights_9:float=0.018
   ) -> (Pose) {
    alphaBoolBlend {blendOption=Linear}
    boneToModify {boneName=dyn_bk_jacket_r_01}
    constraintSetup [
     {transformType=Translation, targetBone {boneName=dyn_BK_r_MSTR}}
     {transformType=Translation, targetBone {boneName=dyn_bk_flap}}
     {transformType=Translation, targetBone {boneName=dyn_bk_flap_01}}
     {transformType=Translation, targetBone {boneName=dyn_bk_flap_02}}
     {transformType=Translation, targetBone {boneName=dyn_bk_flap_r}}
     {transformType=Translation, targetBone {boneName=dyn_bk_flap_r_01}}
     {transformType=Translation, targetBone {boneName=pelvis}}
     {transformType=Translation, targetBone {boneName=thigh_twist_01_l}}
     {transformType=Translation, targetBone {boneName=thigh_twist_01_r}}
     {transformType=Translation, targetBone {boneName=dyn_bk_flap_r_02}}
    ]
    constraintWeights [1,1,1,1,1,1,1,1,1,1]
   }
   "Constraint - dyn_bk_flap_r_01 (TTTTTTTTT)"!(
    ComponentPose
    Alpha=1
    ConstraintWeights_0=0.004
    ConstraintWeights_1:float=0.062
    ConstraintWeights_2:float=0.047
    ConstraintWeights_3:float=0.041
    ConstraintWeights_4:float=0.355
    ConstraintWeights_5:float=0.05
    ConstraintWeights_6:float=0.324
    ConstraintWeights_7:float=0.004
    ConstraintWeights_8:float=0.113
   ) -> (Pose) {
    alphaBoolBlend {blendOption=Linear}
    boneToModify {boneName=dyn_bk_flap_r_01}
    constraintSetup [
     {transformType=Translation, targetBone {boneName=dyn_bk_flap}}
     {transformType=Translation, targetBone {boneName=dyn_bk_flap_01}}
     {transformType=Translation, targetBone {boneName=dyn_bk_flap_02}}
     {transformType=Translation, targetBone {boneName=dyn_bk_flap_r}}
     {transformType=Translation, targetBone {boneName=dyn_bk_flap_r_01}}
     {transformType=Translation, targetBone {boneName=dyn_bk_flap_r_02}}
     {transformType=Translation, targetBone {boneName=pelvis}}
     {transformType=Translation, targetBone {boneName=thigh_twist_01_l}}
     {transformType=Translation, targetBone {boneName=thigh_twist_01_r}}
    ]
    constraintWeights [1,1,1,1,1,1,1,1,1]
   }
   "Transform (Modify) Bone - Bone: dyn_bk_flap_r_01"!(ComponentPose, bAlphaBoolEnabled:bool=True, Translation="0,0,6", Rotation="0,0,0", Scale="1,1,1") -> (Pose) {
    alphaInputType=Bool
    rotationMode=BMM_Additive
    rotationSpace=BCS_BoneSpace
    translationMode=BMM_Additive
    translationSpace=BCS_BoneSpace
    alphaBoolBlend {blendOption=Linear}
    boneToModify {boneName=dyn_bk_flap_r_01}
    translation {z=6}
   }
   "Transform (Modify) Bone - Bone: dyn_l_pocket_MSTR"(Alpha=1, Translation="1,-2,0.5", Rotation="0.000,-15,30", Scale="0.9,0.9,0.9") {
    rotationMode=BMM_Additive
    scaleMode=BMM_Additive
    translationMode=BMM_Additive
    alphaBoolBlend {blendOption=Linear}
    boneToModify {boneName=dyn_l_pocket_MSTR}
    rotation {roll=30, yaw=-15}
    scale=<0.9 0.9 0.9>
    translation=<1 -2 0.5>
   }
   n4 = "Constraint - dyn_bk_flap_l_01 (TTTT)"!(
    ComponentPose
    bAlphaBoolEnabled:bool=True
    ConstraintWeights_0=0.63
    ConstraintWeights_1:float=0.014
    ConstraintWeights_2:float=0.352
    ConstraintWeights_3:float=0.004
   ) -> (Pose) {
    alphaInputType=Bool
    alphaBoolBlend {blendOption=Linear}
    boneToModify {boneName=dyn_bk_flap_l_01}
    constraintSetup [
     {transformType=Translation, targetBone {boneName=pelvis}}
     {transformType=Translation, targetBone {boneName=thigh_l}}
     {transformType=Translation, targetBone {boneName=thigh_twist_01_l}}
     {transformType=Translation, targetBone {boneName=thigh_twist_01_r}}
    ]
    constraintWeights [1,1,1,1]
   }
   n5 = "Transform (Modify) Bone - Bone: dyn_bk_flap_l_01"!(ComponentPose=@n4, bAlphaBoolEnabled:bool=True, Translation="0,5,15", Rotation="30,0,0", Scale="1,1,1") -> (Pose) {
    alphaInputType=Bool
    rotationMode=BMM_Additive
    rotationSpace=BCS_BoneSpace
    translationMode=BMM_Additive
    translationSpace=BCS_BoneSpace
    alphaBoolBlend {blendOption=Linear}
    boneToModify {boneName=dyn_bk_flap_l_01}
    rotation {pitch=30}
    translation {y=5, z=15}
   }
   n6 = "Transform (Modify) Bone - Bone: dyn_bk_jacket_l_01"!(ComponentPose=@n5, bAlphaBoolEnabled:bool=True, Translation="0,10,5", Rotation="30,0,0", Scale="1,1,1") -> (Pose) {
    alphaInputType=Bool
    rotationMode=BMM_Additive
    rotationSpace=BCS_BoneSpace
    translationMode=BMM_Additive
    translationSpace=BCS_BoneSpace
    alphaBoolBlend {blendOption=Linear}
    boneToModify {boneName=dyn_bk_jacket_l_01}
    rotation {pitch=30}
    translation {y=10, z=5}
   }
   "Transform (Modify) Bone - Bone: dyn_l_flap_a_01"!(ComponentPose=@n6, bAlphaBoolEnabled:bool=True, Translation="0,0,5", Rotation="30,30,0", Scale="1,1,1") -> (Pose) {
    alphaInputType=Bool
    rotationMode=BMM_Additive
    rotationSpace=BCS_BoneSpace
    translationMode=BMM_Additive
    translationSpace=BCS_BoneSpace
    alphaBoolBlend {blendOption=Linear}
    boneToModify {boneName=dyn_l_flap_a_01}
    rotation {pitch=30, yaw=30}
    translation {z=5}
   }
   n7 = "Transform (Modify) Bone - Bone: dyn_bk_flap"(Alpha=1, Translation="0,0,-3.5", Rotation="20,0.000,0.000", Scale="1,1,1") {
    rotationMode=BMM_Replace
    rotationSpace=BCS_BoneSpace
    scaleSpace=BCS_BoneSpace
    translationMode=BMM_Replace
    translationSpace=BCS_BoneSpace
    alphaBoolBlend {blendOption=Linear}
    boneToModify {boneName=dyn_bk_flap}
    rotation {pitch=20}
    translation {z=-3.5}
   }
   n8 = "Transform (Modify) Bone - Bone: dyn_bk_flap_r"(ComponentPose=@n7, Alpha=1, Translation="0,0,0", Rotation="20,0.000,0.000", Scale="1,1,1") {
    rotationMode=BMM_Replace
    rotationSpace=BCS_BoneSpace
    scaleSpace=BCS_BoneSpace
    translationMode=BMM_Replace
    translationSpace=BCS_BoneSpace
    alphaBoolBlend {blendOption=Linear}
    boneToModify {boneName=dyn_bk_flap_r}
    rotation {pitch=20}
   }
   n9 = "Transform (Modify) Bone - Bone: dyn_bk_jacket_r_MSTR"!(ComponentPose=@n8, bAlphaBoolEnabled:bool=True, Translation="0,-1,0", Rotation="0,0,30", Scale="1,1,1") -> (Pose) {
    alphaInputType=Bool
    rotationMode=BMM_Additive
    translationMode=BMM_Additive
    alphaBoolBlend {blendOption=Linear}
    boneToModify {boneName=dyn_bk_jacket_r_MSTR}
    rotation {roll=30}
    translation {y=-1}
   }
   n10 = "Transform (Modify) Bone - Bone: dyn_bk_flap_l"(ComponentPose=@n9, Alpha=1, Translation="0,0,-1", Rotation="20,0.000,0.000", Scale="1,1,1") {
    rotationMode=BMM_Replace
    rotationSpace=BCS_BoneSpace
    scaleSpace=BCS_BoneSpace
    translationMode=BMM_Replace
    translationSpace=BCS_BoneSpace
    alphaBoolBlend {blendOption=Linear}
    boneToModify {boneName=dyn_bk_flap_l}
    rotation {pitch=20}
    translation {z=-1}
   }
   n11 = "Transform (Modify) Bone - Bone: dyn_bk_jacket_MSTR"!(ComponentPose=@n10, bAlphaBoolEnabled:bool=True, Translation="0,0,0", Rotation="0,0,30", Scale="1,1,1") -> (Pose) {
    alphaInputType=Bool
    rotationMode=BMM_Additive
    translationMode=BMM_Additive
    alphaBoolBlend {blendOption=Linear}
    boneToModify {boneName=dyn_bk_jacket_MSTR}
    rotation {roll=30}
   }
   n12 = "Transform (Modify) Bone - Bone: dyn_r_b_MSTR"!(ComponentPose=@n11, bAlphaBoolEnabled:bool=True, Translation="5,2,-4", Rotation="-30,10,0", Scale="1,1,1") -> (Pose) {
    alphaInputType=Bool
    rotationMode=BMM_Additive
    rotationSpace=BCS_BoneSpace
    translationMode=BMM_Additive
    translationSpace=BCS_BoneSpace
    alphaBoolBlend {blendOption=Linear}
    boneToModify {boneName=dyn_r_b_MSTR}
    rotation {pitch=-30, yaw=10}
    translation=<5 2 -4>
   }
   n13 = "Transform (Modify) Bone - Bone: dyn_l_b_MSTR"!(ComponentPose=@n12, bAlphaBoolEnabled:bool=True, Translation="-4,-1,-4", Rotation="20,5,20", Scale="1,1,1") -> (Pose) {
    alphaInputType=Bool
    rotationMode=BMM_Additive
    translationMode=BMM_Additive
    alphaBoolBlend {blendOption=Linear}
    boneToModify {boneName=dyn_l_b_MSTR}
    rotation=<rot 20 5 20>
    translation=<-4 -1 -4>
   }
   n14 = "Transform (Modify) Bone - Bone: dyn_bk_jacket_l_MSTR"!(ComponentPose=@n13, bAlphaBoolEnabled:bool=True, Translation="0,0,0", Rotation="30,0,0", Scale="1,1,1") -> (Pose) {
    alphaInputType=Bool
    rotationMode=BMM_Additive
    translationMode=BMM_Additive
    alphaBoolBlend {blendOption=Linear}
    boneToModify {boneName=dyn_bk_jacket_l_MSTR}
    rotation {pitch=30}
   }
   n15 = "Transform (Modify) Bone - Bone: dyn_r_a_MSTR"!(ComponentPose=@n14, bAlphaBoolEnabled:bool=True, Translation="0,2,0", Rotation="15,0,0", Scale="1,1,1") -> (Pose) {
    alphaInputType=Bool
    rotationMode=BMM_Additive
    rotationSpace=BCS_BoneSpace
    translationMode=BMM_Additive
    translationSpace=BCS_BoneSpace
    alphaBoolBlend {blendOption=Linear}
    boneToModify {boneName=dyn_r_a_MSTR}
    rotation {pitch=15}
    translation {y=2}
   }
   n16 = "Transform (Modify) Bone - Bone: dyn_l_a_MSTR"!(ComponentPose=@n15, bAlphaBoolEnabled:bool=True, Translation="0,0,0", Rotation="0,0,15", Scale="1,1,1") -> (Pose) {
    alphaInputType=Bool
    rotationMode=BMM_Additive
    translationMode=BMM_Additive
    alphaBoolBlend {blendOption=Linear}
    boneToModify {boneName=dyn_l_a_MSTR}
    rotation {roll=15}
   }
   "Transform (Modify) Bone - Bone: dyn_r_jacket_b_MSTR"!(ComponentPose=@n16, bAlphaBoolEnabled:bool=True, Translation="0,0,1", Rotation="10,0,0", Scale="1,1,1") -> (Pose) {
    alphaInputType=Bool
    rotationMode=BMM_Additive
    rotationSpace=BCS_BoneSpace
    translationMode=BMM_Additive
    translationSpace=BCS_BoneSpace
    alphaBoolBlend {blendOption=Linear}
    boneToModify {boneName=dyn_r_jacket_b_MSTR}
    rotation {pitch=10}
    translation {z=1}
   }
   n17 = "Transform (Modify) Bone - Bone: dyn_fr_MSTR"(
    ComponentPose=Local To Component(LocalPose=Use cached pose 'RBAN'())
    Alpha=1
    Translation="-2,3,10"
    Rotation="10,15.000,0.000"
    Scale="1,1,1"
   ) {
    rotationMode=BMM_Additive
    translationMode=BMM_Additive
    alphaBoolBlend {blendOption=Linear}
    boneToModify {boneName=dyn_fr_MSTR}
    rotation {pitch=10, yaw=15}
    translation=<-2 3 10>
   }
   n18 = "Transform (Modify) Bone - Bone: drn_fr_l_MSTR"(ComponentPose=@n17, Alpha=1, Translation="-3,4.000,-1.000", Rotation="0,20.000,0.000", Scale="1,1,1") {
    rotationMode=BMM_Additive
    translationMode=BMM_Additive
    alphaBoolBlend {blendOption=Linear}
    boneToModify {boneName=drn_fr_l_MSTR}
    rotation {yaw=20}
    translation=<-3 4 -1>
   }
   n19 = "Transform (Modify) Bone - Bone: dyn_l_pocket_MSTR"(ComponentPose=@n18, Alpha=1, Translation="0,-2,-0.5", Rotation="0.000,-15,30", Scale="0.9,0.9,0.9") {
    rotationMode=BMM_Additive
    scaleMode=BMM_Additive
    translationMode=BMM_Additive
    alphaBoolBlend {blendOption=Linear}
    boneToModify {boneName=dyn_l_pocket_MSTR}
    rotation {roll=30, yaw=-15}
    scale=<0.9 0.9 0.9>
    translation {y=-2, z=-0.5}
   }
   n20 = "Transform (Modify) Bone - Bone: dyn_l_jacket_a"(ComponentPose=@n19, Alpha=1, Translation="0,-8,10", Rotation="0,0,40", Scale="1,1,1") {
    rotationMode=BMM_Additive
    rotationSpace=BCS_BoneSpace
    scaleMode=BMM_Additive
    scaleSpace=BCS_BoneSpace
    translationMode=BMM_Additive
    translationSpace=BCS_BoneSpace
    alphaBoolBlend {blendOption=Linear}
    alphaScaleBiasClamp {interpSpeedDecreasing=1, interpSpeedIncreasing=20}
    boneToModify {boneName=dyn_l_jacket_a}
    rotation {roll=40}
    translation {y=-8, z=10}
   }
   n21 = "Transform (Modify) Bone - Bone: dyn_fr_jacket"!(ComponentPose=@n20, bAlphaBoolEnabled:bool=True, Translation="-3,0,12", Rotation="20,0,-30", Scale="1,1,1") -> (Pose) {
    alphaInputType=Bool
    rotationMode=BMM_Additive
    translationMode=BMM_Additive
    alphaBoolBlend {blendOption=Linear}
    boneToModify {boneName=dyn_fr_jacket}
    rotation {pitch=20, roll=-30}
    translation {x=-3, z=12}
   }
   n22 = "Transform (Modify) Bone - Bone: dyn_fr_flap_01"!(ComponentPose=@n21, bAlphaBoolEnabled:bool=True, Translation="0,0,0", Rotation="0,0,-90", Scale="1,1,1") -> (Pose) {
    alphaInputType=Bool
    rotationMode=BMM_Additive
    translationMode=BMM_Additive
    alphaBoolBlend {blendOption=Linear}
    boneToModify {boneName=dyn_fr_flap_01}
    rotation {roll=-90}
   }
   n23 = "Transform (Modify) Bone - Bone: dyn_fr_flap_02"!(ComponentPose=@n22, bAlphaBoolEnabled:bool=True, Translation="-3,0,-5", Rotation="0,0,80", Scale="1,1,1") -> (Pose) {
    alphaInputType=Bool
    rotationMode=BMM_Additive
    translationMode=BMM_Additive
    alphaBoolBlend {blendOption=Linear}
    boneToModify {boneName=dyn_fr_flap_02}
    rotation {roll=80}
    translation {x=-3, z=-5}
   }
   n24 = "Transform (Modify) Bone - Bone: dyn_fr_flap_l"!(ComponentPose=@n23, bAlphaBoolEnabled:bool=True, Translation="-4,-3,0", Rotation="0,0,10", Scale="1,1,1") -> (Pose) {
    alphaInputType=Bool
    rotationMode=BMM_Additive
    translationMode=BMM_Additive
    alphaBoolBlend {blendOption=Linear}
    boneToModify {boneName=dyn_fr_flap_l}
    rotation {roll=10}
    translation=<-4 -3>
   }
   n25 = "Transform (Modify) Bone - Bone: dyn_r_flap_b"!(ComponentPose=@n24, bAlphaBoolEnabled:bool=True, Translation="0,0,0", Rotation="5,0.000,0.000", Scale="1,1,1") -> (Pose) {
    alphaInputType=Bool
    rotationMode=BMM_Additive
    rotationSpace=BCS_BoneSpace
    scaleSpace=BCS_BoneSpace
    translationSpace=BCS_BoneSpace
    alphaBoolBlend {blendOption=Linear}
    boneToModify {boneName=dyn_r_flap_b}
    rotation {pitch=5}
   }
   n26 = "Transform (Modify) Bone - Bone: dyn_fr_flap_02"(ComponentPose=@n25, bAlphaBoolEnabled=True, Translation="0,0,0", Rotation="0,0,0", Scale="0.3,0.3,0.3") {
    alphaInputType=Bool
    rotationMode=BMM_Additive
    rotationSpace=BCS_BoneSpace
    scaleMode=BMM_Additive
    scaleSpace=BCS_BoneSpace
    translationSpace=BCS_BoneSpace
    alphaBoolBlend {blendInTime=0.3, blendOption=ExpInOut, blendOutTime=0.3}
    boneToModify {boneName=dyn_fr_flap_02}
    scale=<0.3 0.3 0.3>
   }
   n27 = "Transform (Modify) Bone - Bone: dyn_r_jacket_a_MSTR"!(ComponentPose=@n26, bAlphaBoolEnabled:bool=True, Translation="0,0,0", Rotation="-40,0,0", Scale="1,1,1") -> (Pose) {
    alphaInputType=Bool
    rotationMode=BMM_Additive
    translationMode=BMM_Additive
    alphaBoolBlend {blendOption=Linear}
    boneToModify {boneName=dyn_r_jacket_a_MSTR}
    rotation {pitch=-40}
   }
   n28 = "Transform (Modify) Bone - Bone: dyn_r_jacket_b_MSTR"(ComponentPose=@n27, bAlphaBoolEnabled=True, Translation="0,0,0", Rotation="-40,0,0", Scale="1,1,1") {
    alphaInputType=Bool
    rotationMode=BMM_Additive
    translationMode=BMM_Additive
    alphaBoolBlend {blendOption=Linear}
    boneToModify {boneName=dyn_r_jacket_b_MSTR}
    rotation {pitch=-40}
   }
   n29 = "Transform (Modify) Bone - Bone: dyn_fr_jacket_r_MSTR"!(ComponentPose=@n28, bAlphaBoolEnabled:bool=True, Translation="0,0,0", Rotation="0,0,15", Scale="1,1,1") -> (Pose) {
    alphaInputType=Bool
    rotationMode=BMM_Additive
    translationMode=BMM_Additive
    alphaBoolBlend {blendOption=Linear}
    boneToModify {boneName=dyn_fr_jacket_r_MSTR}
    rotation {roll=15}
   }
   n30 = "Transform (Modify) Bone - Bone: dyn_fr_jacket_l_MSTR"!(ComponentPose=@n29, bAlphaBoolEnabled:bool=True, Translation="0,0,0", Rotation="0,0,15", Scale="1,1,1") -> (Pose) {
    alphaInputType=Bool
    rotationMode=BMM_Additive
    translationMode=BMM_Additive
    alphaBoolBlend {blendOption=Linear}
    boneToModify {boneName=dyn_fr_jacket_l_MSTR}
    rotation {roll=15}
   }
   n31 = RigidBody!(ComponentPose=@n30, bAlphaBoolEnabled:bool, ExternalForce) -> (Pose) {
    alphaInputType=Bool
    bClampLinearTranslationLimitToRefPose=true
    bOverrideWorldGravity=true
    overridePhysicsAsset=&/M_MED_Pumpkin_Punk_Crouch_Physics
    simulationSpace=BaseBoneSpace
    alphaBoolBlend {blendOption=Linear}
    baseBoneRef {boneName=pelvis}
    componentAppliedLinearAccClamp=<5000 5000 5000>
    componentLinearAccScale=<0.25 0.25 0.25>
    overrideWorldGravity {z=-980}
   }
   n32 = RigidBody(ComponentPose=@n31) {
    alphaInputType=Bool
    bClampLinearTranslationLimitToRefPose=true
    bOverrideWorldGravity=true
    overridePhysicsAsset=&/M_MED_Pumpkin_Punk_Fr_Crouch_Physics
    simulationSpace=BaseBoneSpace
    alphaBoolBlend {blendOption=Linear}
    baseBoneRef {boneName=pelvis}
    componentAppliedLinearAccClamp=<5000 5000 5000>
    componentLinearAccScale=<0.25 0.25 0.25>
    overrideWorldGravity {z=-980}
   }
   n33 = RigidBody(ComponentPose=@n32) {
    alphaInputType=Bool
    bClampLinearTranslationLimitToRefPose=true
    bOverrideWorldGravity=true
    overridePhysicsAsset=&/M_MED_Pumpkin_Punk_Hoodie_Physics
    simulationSpace=BaseBoneSpace
    alphaBoolBlend {blendOption=Linear}
    baseBoneRef {boneName=pelvis}
    componentAppliedLinearAccClamp=<5000 5000 5000>
    componentLinearAccScale=<0.25 0.25 0.25>
    overrideWorldGravity {z=-980}
   }
   n34 = "Transform (Modify) Bone - Bone: dyn_l_flap_a"!(ComponentPose=@n33, bAlphaBoolEnabled:bool=True, Translation="0,0,-2", Rotation="0,0,0", Scale="1,1,1") -> (Pose) {
    alphaInputType=Bool
    rotationMode=BMM_Additive
    rotationSpace=BCS_BoneSpace
    scaleMode=BMM_Additive
    scaleSpace=BCS_BoneSpace
    translationMode=BMM_Additive
    translationSpace=BCS_BoneSpace
    alphaBoolBlend {blendOption=Linear}
    boneToModify {boneName=dyn_l_flap_a}
    translation {z=-2}
   }
   n35 = "Transform (Modify) Bone - Bone: dyn_l_flap_b"!(ComponentPose=@n34, bAlphaBoolEnabled:bool=True, Translation="-2,0,1.3", Rotation="0,0,0", Scale="1,1,1") -> (Pose) {
    alphaInputType=Bool
    rotationMode=BMM_Additive
    rotationSpace=BCS_BoneSpace
    scaleMode=BMM_Additive
    scaleSpace=BCS_BoneSpace
    translationMode=BMM_Additive
    translationSpace=BCS_BoneSpace
    alphaBoolBlend {blendOption=Linear}
    boneToModify {boneName=dyn_l_flap_b}
    translation {x=-2, z=1.3}
   }
   n36 = "Transform (Modify) Bone - Bone: dyn_fr_flap_01"!(ComponentPose=@n35, Alpha:float=Get ThighSumAlpha() -> (ThighSumAlpha:double), Translation="0,6,0", Rotation="50,-30,0", Scale="1,1,1") -> (Pose) {
    rotationMode=BMM_Additive
    translationMode=BMM_Additive
    alphaBoolBlend {blendOption=Linear}
    alphaScaleBiasClamp {bInterpResult=true}
    boneToModify {boneName=dyn_fr_flap_01}
    rotation {pitch=50, yaw=-30}
    translation {y=6}
   }
   n37 = Component To Local(ComponentPose=@n36)
   n38 = "Transform (Modify) Bone - Bone: dyn_r_pocket_MSTR"!(
    ComponentPose=Local To Component(LocalPose=Use cached pose 'RBAN'())
    bAlphaBoolEnabled:bool=True
    Translation="0,0,0"
    Rotation="0,25,0.000"
    Scale="1,1,1"
   ) -> (Pose) {
    alphaInputType=Bool
    rotationMode=BMM_Additive
    rotationSpace=BCS_BoneSpace
    translationMode=BMM_Additive
    alphaBoolBlend {blendOption=Linear}
    boneToModify {boneName=dyn_r_pocket_MSTR}
    rotation {yaw=25}
   }
   n39 = "Transform (Modify) Bone - Bone: dyn_l_pocket_MSTR"!(ComponentPose=@n38, bAlphaBoolEnabled:bool=True, Translation="0,0,0", Rotation="0,25,0", Scale="1,1,1") -> (Pose) {
    alphaInputType=Bool
    rotationMode=BMM_Additive
    rotationSpace=BCS_BoneSpace
    translationMode=BMM_Additive
    alphaBoolBlend {blendOption=Linear}
    boneToModify {boneName=dyn_l_pocket_MSTR}
    rotation {yaw=25}
   }
   n40 = "Transform (Modify) Bone - Bone: dyn_hood_collar_bk_r"!(ComponentPose=@n39, bAlphaBoolEnabled:bool=True, Translation="-4,0.000,-2", Rotation="15,0,0", Scale="1,1,1") -> (Pose) {
    alphaInputType=Bool
    rotationMode=BMM_Additive
    translationMode=BMM_Additive
    alphaBoolBlend {blendOption=Linear}
    boneToModify {boneName=dyn_hood_collar_bk_r}
    rotation {pitch=15}
    translation {x=-4, z=-2}
   }
   n41 = "Transform (Modify) Bone - Bone: dyn_hood_collar_r"!(ComponentPose=@n40, bAlphaBoolEnabled:bool=True, Translation="-4,0.000,-2", Rotation="15,0,0", Scale="1,1,1") -> (Pose) {
    alphaInputType=Bool
    rotationMode=BMM_Additive
    translationMode=BMM_Additive
    alphaBoolBlend {blendOption=Linear}
    boneToModify {boneName=dyn_hood_collar_r}
    rotation {pitch=15}
    translation {x=-4, z=-2}
   }
   n42 = Component To Local(ComponentPose=@n41)
   n43 = Use cached pose 'RBAN'()
   n44 = OR Boolean(A=Get bIsFalling(), B=Get bIsJumping())
   n45 = RigidBody!(
    ComponentPose=Local To Component(LocalPose=Blend Poses by bool(BlendPose_0=@n42, BlendPose_1=@n43, BlendTime_0=0.2, BlendTime_1=0.2, bActiveValue=@n44) {blendType=Linear})
    Alpha:float
    ExternalForce
   ) -> (Pose) {
    bClampLinearTranslationLimitToRefPose=true
    bOverrideWorldGravity=true
    overridePhysicsAsset=&/M_MED_Pumpkin_Punk_Physics
    simulationSpace=BaseBoneSpace
    alphaBoolBlend {blendOption=Linear}
    baseBoneRef {boneName=pelvis}
    componentAppliedLinearAccClamp=<5000 5000 5000>
    componentLinearAccScale=<0.35 0.35 0.35>
    componentLinearVelScale=<0.2 0.2 0.2>
    overrideWorldGravity {z=-980}
   }
   n46 = "Transform (Modify) Bone - Bone: dyn_fr_jacket_r_MSTR"!(ComponentPose=@n45, Alpha:float=1, Translation="0,0,0", Rotation="-10,0,0", Scale="1,1,1") -> (Pose) {
    rotationMode=BMM_Additive
    translationMode=BMM_Additive
    alphaBoolBlend {blendOption=Linear}
    boneToModify {boneName=dyn_fr_jacket_r_MSTR}
    rotation {pitch=-10}
   }
   n47 = "Transform (Modify) Bone - Bone: dyn_fr_jacket_MSTR"(ComponentPose=@n46, Alpha=1, Translation="0,0,0", Rotation="0,0,-15", Scale="1,1,1") {
    rotationMode=BMM_Additive
    translationMode=BMM_Additive
    alphaBoolBlend {blendOption=Linear}
    boneToModify {boneName=dyn_fr_jacket_MSTR}
    rotation {roll=-15}
   }
   n48 = "Transform (Modify) Bone - Bone: dyn_l_jacket_a_MSTR"(ComponentPose=@n47, Alpha=1, Translation="0,0,0", Rotation="20,0,0", Scale="1,1,1") {
    rotationMode=BMM_Additive
    translationMode=BMM_Additive
    alphaBoolBlend {blendOption=Linear}
    boneToModify {boneName=dyn_l_jacket_a_MSTR}
    rotation {pitch=20}
   }
   n49 = "Transform (Modify) Bone - Bone: dyn_bk_jacket_MSTR"!(ComponentPose=@n48, Alpha:float=1, Translation="0,0,0", Rotation="0,0,8", Scale="1,1,1") -> (Pose) {
    rotationMode=BMM_Additive
    translationMode=BMM_Additive
    alphaBoolBlend {blendOption=Linear}
    boneToModify {boneName=dyn_bk_jacket_MSTR}
    rotation {roll=8}
   }
   n50 = "Transform (Modify) Bone - Bone: dyn_l_jacket_b_MSTR"(ComponentPose=@n49, Alpha=1, Translation="1,0,0", Rotation="10,0,0", Scale="1,1,1") {
    rotationMode=BMM_Additive
    translationMode=BMM_Additive
    alphaBoolBlend {blendOption=Linear}
    boneToModify {boneName=dyn_l_jacket_b_MSTR}
    rotation {pitch=10}
    translation {x=1}
   }
   n51 = "Transform (Modify) Bone - Bone: dyn_r_jacket_b_MSTR"!(ComponentPose=@n50, Alpha:float=1, Translation="0,-4,0", Rotation="-20,0,0", Scale="1,1,1") -> (Pose) {
    rotationMode=BMM_Additive
    translationMode=BMM_Additive
    translationSpace=BCS_BoneSpace
    alphaBoolBlend {blendOption=Linear}
    boneToModify {boneName=dyn_r_jacket_b_MSTR}
    rotation {pitch=-20}
    translation {y=-4}
   }
   n52 = "Transform (Modify) Bone - Bone: dyn_r_jacket_a_MSTR"!(ComponentPose=@n51, Alpha:float=1, Translation="0,0,0", Rotation="0,0,30", Scale="1,1,1") -> (Pose) {
    rotationMode=BMM_Additive
    translationMode=BMM_Additive
    alphaBoolBlend {blendOption=Linear}
    boneToModify {boneName=dyn_r_jacket_a_MSTR}
    rotation {roll=30}
   }
   n53 = "Transform (Modify) Bone - Bone: dyn_bk_jacket_r_MSTR"!(ComponentPose=@n52, Alpha:float=1, Translation="0,0,0", Rotation="0,0,8", Scale="1,1,1") -> (Pose) {
    rotationMode=BMM_Additive
    translationMode=BMM_Additive
    alphaBoolBlend {blendOption=Linear}
    boneToModify {boneName=dyn_bk_jacket_r_MSTR}
    rotation {roll=8}
   }
   n54 = "Transform (Modify) Bone - Bone: dyn_bk_jacket_l_MSTR"!(ComponentPose=@n53, Alpha:float=1, Translation="0,0,0", Rotation="8,0,0", Scale="1,1,1") -> (Pose) {
    rotationMode=BMM_Additive
    translationMode=BMM_Additive
    alphaBoolBlend {blendOption=Linear}
    boneToModify {boneName=dyn_bk_jacket_l_MSTR}
    rotation {pitch=8}
   }
   n55 = "Transform (Modify) Bone - Bone: dyn_r_flap_b_01"(ComponentPose=@n54, Alpha=1, Translation="-5,0,0", Rotation="0,0,0", Scale="1,1,1") {
    rotationMode=BMM_Additive
    translationMode=BMM_Additive
    alphaBoolBlend {blendOption=Linear}
    alphaScaleBiasClamp {bInterpResult=true, interpSpeedDecreasing=5, interpSpeedIncreasing=5}
    boneToModify {boneName=dyn_r_flap_b_01}
    translation {x=-5}
   }
   n56 = "Transform (Modify) Bone - Bone: dyn_r_flap_b"(ComponentPose=@n55, bAlphaBoolEnabled=True, Translation="-2,0,0", Rotation="10,0,0", Scale="1,1,1") {
    alphaInputType=Bool
    rotationMode=BMM_Additive
    rotationSpace=BCS_BoneSpace
    translationMode=BMM_Additive
    alphaBoolBlend {blendInTime=0.2, blendOption=Linear, blendOutTime=0.3}
    alphaScaleBiasClamp {bInterpResult=true, interpSpeedDecreasing=5, interpSpeedIncreasing=5}
    boneToModify {boneName=dyn_r_flap_b}
    rotation {pitch=10}
    translation {x=-2}
   }
   n57 = "Constraint - dyn_r_jacket_a_MSTR (TTTT)"!(
    ComponentPose=@n56
    bAlphaBoolEnabled:bool=True
    ConstraintWeights_0=0.089
    ConstraintWeights_1:float=0.034
    ConstraintWeights_2:float=0.769
    ConstraintWeights_3:float=0.109
   ) -> (Pose) {
    alphaInputType=Bool
    alphaBoolBlend {blendOption=Linear}
    boneToModify {boneName=dyn_r_jacket_a_MSTR}
    constraintSetup [
     {transformType=Translation, targetBone {boneName=dyn_r_flap_a_02}}
     {transformType=Translation, targetBone {boneName=dyn_r_flap_b_01}}
     {transformType=Translation, targetBone {boneName=pelvis}}
     {transformType=Translation, targetBone {boneName=thigh_twist_01_r}}
    ]
    constraintWeights [1,1,1,1]
   }
   n58 = "Transform (Modify) Bone - Bone: dyn_bk_jacket_r"(ComponentPose=@n57, Alpha=1, Translation="0,0,3", Rotation="0,0,0", Scale="1,1,1") {
    rotationMode=BMM_Additive
    rotationSpace=BCS_BoneSpace
    translationMode=BMM_Additive
    translationSpace=BCS_BoneSpace
    alphaBoolBlend {blendOption=Linear}
    boneToModify {boneName=dyn_bk_jacket_r}
    translation {z=3}
   }
   n59 = "Transform (Modify) Bone - Bone: dyn_bk_flap_l"!(ComponentPose=@n58, bAlphaBoolEnabled:bool=True, Translation="0,2,0", Rotation="0,10,0", Scale="1,1,1") -> (Pose) {
    alphaInputType=Bool
    rotationMode=BMM_Additive
    translationMode=BMM_Additive
    alphaBoolBlend {blendOption=Linear}
    boneToModify {boneName=dyn_bk_flap_l}
    rotation {yaw=10}
    translation {y=2}
   }
   n60 = "Transform (Modify) Bone - Bone: dyn_fr_flap_01"(
    ComponentPose=@n59
    Alpha=Get ThighR_FPSCorrection() -> (ThighR_FPSCorrection:double)
    Translation="-2,0,0"
    Rotation="15,0,0"
    Scale="1,1,1"
   ) {
    rotationMode=BMM_Additive
    rotationSpace=BCS_BoneSpace
    translationMode=BMM_Additive
    translationSpace=BCS_BoneSpace
    alphaBoolBlend {blendOption=Linear}
    alphaScaleBiasClamp {inRange {max=0.5}}
    boneToModify {boneName=dyn_fr_flap_01}
    rotation {pitch=15}
    translation {x=-2}
   }
   n61 = Get StandingMoving_FPSCorrestion() -> (StandingMoving_FPSCorrestion:double)
   n62 = "Transform (Modify) Bone - Bone: dyn_fr_flap_l_01"(ComponentPose=@n60, Alpha=@n61, Translation="0,0,3", Rotation="0,0,0", Scale="1,1,1") {
    rotationMode=BMM_Additive
    rotationSpace=BCS_BoneSpace
    translationMode=BMM_Additive
    translationSpace=BCS_BoneSpace
    alphaBoolBlend {blendOption=Linear}
    alphaScaleBiasClamp {bInterpResult=true, inRange {max=0.7}}
    boneToModify {boneName=dyn_fr_flap_l_01}
    translation {z=3}
   }
   n63 = "Transform (Modify) Bone - Bone: dyn_fr_flap_01"(
    ComponentPose=@n62
    Alpha=Get ThighSumFPSCorrection() -> (ThighSumFPSCorrection:double)
    Translation="0,6,0"
    Rotation="0,0,-30"
    Scale="1,1,1"
   ) {
    rotationMode=BMM_Additive
    translationMode=BMM_Additive
    alphaBoolBlend {blendOption=Linear}
    alphaScaleBiasClamp {bInterpResult=true}
    boneToModify {boneName=dyn_fr_flap_01}
    rotation {roll=-30}
    translation {y=6}
   }
   n64 = "Transform (Modify) Bone - Bone: dyn_fr_flap_r_02"(ComponentPose=@n63, Alpha=@n61, Translation="0,0,3", Rotation="0,0,0", Scale="1,1,1") {
    rotationMode=BMM_Additive
    rotationSpace=BCS_BoneSpace
    translationMode=BMM_Additive
    translationSpace=BCS_BoneSpace
    alphaBoolBlend {blendOption=Linear}
    alphaScaleBiasClamp {inRange {max=0.7}}
    boneToModify {boneName=dyn_fr_flap_r_02}
    translation {z=3}
   }
   n65 = Get Jacket_LefttSide_SideDirection_Fix() -> (Jacket_LefttSide_SideDirection_Fix:double)
   n66 = "Transform (Modify) Bone - Bone: dyn_l_flap_a"!(ComponentPose=@n64, Alpha:float=@n65, Translation="0,10,0", Rotation="10,0,0", Scale="1,1,1") -> (Pose) {
    rotationMode=BMM_Additive
    rotationSpace=BCS_BoneSpace
    translationMode=BMM_Additive
    translationSpace=BCS_BoneSpace
    alphaBoolBlend {blendOption=Linear}
    alphaScaleBiasClamp {bInterpResult=true, interpSpeedDecreasing=8, interpSpeedIncreasing=8, inRange {max=0.7}}
    boneToModify {boneName=dyn_l_flap_a}
    rotation {pitch=10}
    translation {y=10}
   }
   n67 = "Transform (Modify) Bone - Bone: dyn_l_b_MSTR"!(ComponentPose=@n66, Alpha:float=@n65, Translation="0,10,0", Rotation="10,0,0", Scale="1,1,1") -> (Pose) {
    rotationMode=BMM_Additive
    rotationSpace=BCS_BoneSpace
    translationMode=BMM_Additive
    translationSpace=BCS_BoneSpace
    alphaBoolBlend {blendOption=Linear}
    alphaScaleBiasClamp {bInterpResult=true, interpSpeedDecreasing=8, interpSpeedIncreasing=8, inRange {max=0.7}}
    boneToModify {boneName=dyn_l_b_MSTR}
    rotation {pitch=10}
    translation {y=10}
   }
   n68 = Get Jacket_RightSide_SideDirection_Fix() -> (Jacket_RightSide_SideDirection_Fix:double)
   n69 = "Transform (Modify) Bone - Bone: dyn_r_flap_b"!(ComponentPose=@n67, Alpha:float=@n68, Translation="0,0,0", Rotation="10,0,0", Scale="1,1,1") -> (Pose) {
    rotationMode=BMM_Additive
    rotationSpace=BCS_BoneSpace
    translationMode=BMM_Additive
    translationSpace=BCS_BoneSpace
    alphaBoolBlend {blendOption=Linear}
    alphaScaleBiasClamp {bInterpResult=true, inRange {max=0.7}}
    boneToModify {boneName=dyn_r_flap_b}
    rotation {pitch=10}
   }
   n70 = "Transform (Modify) Bone - Bone: dyn_r_flap_a"(ComponentPose=@n69, Alpha=@n68, Translation="0,0,0", Rotation="10,0,0", Scale="1,1,1") {
    rotationMode=BMM_Additive
    rotationSpace=BCS_BoneSpace
    translationMode=BMM_Additive
    translationSpace=BCS_BoneSpace
    alphaBoolBlend {blendOption=Linear}
    alphaScaleBiasClamp {bInterpResult=true, inRange {max=0.7}}
    boneToModify {boneName=dyn_r_flap_a}
    rotation {pitch=10}
   }
   n71 = Get ThighRAlpha() -> (ThighRAlpha:double)
   n72 = "Transform (Modify) Bone - Bone: dyn_fr_flap_01"(ComponentPose=@n70, Alpha=@n71, Translation="-4,0,0", Rotation="30,-10,0", Scale="1,1,1") {
    rotationMode=BMM_Additive
    rotationSpace=BCS_BoneSpace
    translationMode=BMM_Additive
    translationSpace=BCS_BoneSpace
    alphaBoolBlend {blendOption=Linear}
    alphaScaleBiasClamp {inRange {max=0.5}}
    boneToModify {boneName=dyn_fr_flap_01}
    rotation {pitch=30, yaw=-10}
    translation {x=-4}
   }
   n73 = "Transform (Modify) Bone - Bone: dyn_fr_flap_02"!(ComponentPose=@n72, Alpha:float=@n71, Translation="-3,0,0", Rotation="20,-10,0", Scale="1,1,1") -> (Pose) {
    rotationMode=BMM_Additive
    rotationSpace=BCS_BoneSpace
    translationMode=BMM_Additive
    translationSpace=BCS_BoneSpace
    alphaBoolBlend {blendOption=Linear}
    alphaScaleBiasClamp {inRange {max=0.5}}
    boneToModify {boneName=dyn_fr_flap_02}
    rotation {pitch=20, yaw=-10}
    translation {x=-3}
   }
   n74 = "Transform (Modify) Bone - Bone: dyn_fr_flap_l_01"!(ComponentPose=@n73, bAlphaBoolEnabled:bool=false, Translation="0,2,0", Rotation="0,0,0", Scale="1,1,1") -> (Pose) {
    alphaInputType=Bool
    rotationMode=BMM_Additive
    rotationSpace=BCS_BoneSpace
    translationMode=BMM_Replace
    translationSpace=BCS_BoneSpace
    alphaBoolBlend {blendOption=Linear}
    boneToModify {boneName=dyn_fr_flap_l_01}
    translation {y=2}
   }
   n75 = "Transform (Modify) Bone - Bone: dyn_fr_jacket_01"(ComponentPose=@n74, Alpha=Get ThighSumAlpha(), Translation="-6,0,6", Rotation="0,0,0", Scale="1,1,1") {
    rotationMode=BMM_Additive
    rotationSpace=BCS_BoneSpace
    translationMode=BMM_Additive
    translationSpace=BCS_BoneSpace
    alphaBoolBlend {blendOption=Linear}
    alphaScaleBiasClamp {bInterpResult=true}
    boneToModify {boneName=dyn_fr_jacket_01}
    translation {x=-6, z=6}
   }
   n76 = "Transform (Modify) Bone - Bone: dyn_fr_jacket_l_01"(ComponentPose=@n75, Alpha=1, Translation="0,0,2", Rotation="0,0,0", Scale="1,1,1") {
    rotationMode=BMM_Additive
    rotationSpace=BCS_BoneSpace
    translationMode=BMM_Additive
    translationSpace=BCS_BoneSpace
    alphaBoolBlend {blendOption=Linear}
    alphaScaleBiasClamp {bInterpResult=true}
    boneToModify {boneName=dyn_fr_jacket_l_01}
    translation {z=2}
   }
   n77 = "Transform (Modify) Bone - Bone: dyn_l_jacket_b_01"(
    ComponentPose=@n76
    Alpha=Get ThighLJacketCorrection() -> (ThighLJacketCorrection:double)
    Translation="0,2,2"
    Rotation="0,0,0"
    Scale="1,1,1"
   ) {
    rotationMode=BMM_Additive
    rotationSpace=BCS_BoneSpace
    translationMode=BMM_Additive
    translationSpace=BCS_BoneSpace
    alphaBoolBlend {blendOption=Linear}
    alphaScaleBiasClamp {bInterpResult=true}
    boneToModify {boneName=dyn_l_jacket_b_01}
    translation {y=2, z=2}
   }
   n78 = Get bIsPlayingEmote()
   n79 = "Constraint - dyn_fr_flap_l_02 (PPPPP)"!(
    ComponentPose=@n77
    bAlphaBoolEnabled:bool=@n78
    ConstraintWeights_0=0.0039
    ConstraintWeights_1:float=0.0516
    ConstraintWeights_2:float=0.8131
    ConstraintWeights_3:float=0.0517
    ConstraintWeights_4:float=0.0786
   ) -> (Pose) {
    alphaInputType=Bool
    alphaBoolBlend {blendInTime=1, blendOption=Linear, blendOutTime=1}
    boneToModify {boneName=dyn_fr_flap_l_02}
    constraintSetup [
     {targetBone {boneName=thigh_r}}
     {targetBone {boneName=thigh_twist_01_r}}
     {targetBone {boneName=thigh_twist_01_l}}
     {targetBone {boneName=thigh_l}}
     {targetBone {boneName=pelvis}}
    ]
    constraintWeights [1,1,1,1,1]
   }
   n80 = "Transform (Modify) Bone - Bone: dyn_fr_flap_l_02"!(ComponentPose=@n79, bAlphaBoolEnabled:bool=@n78, Translation="0,0,2.5", Rotation="5,0,0", Scale="1,1,1") -> (Pose) {
    alphaInputType=Bool
    rotationMode=BMM_Additive
    rotationSpace=BCS_BoneSpace
    translationMode=BMM_Additive
    translationSpace=BCS_BoneSpace
    alphaBoolBlend {blendInTime=1, blendOption=Linear, blendOutTime=1}
    alphaScaleBiasClamp {bInterpResult=true}
    boneToModify {boneName=dyn_fr_flap_l_02}
    rotation {pitch=5}
    translation {z=2.5}
   }
   n81 = Component To Local(ComponentPose=@n80)
   n82 = Get HoodConstraintAlpha() -> (HoodConstraintAlpha:double)
   n83 = "Constraint - dyn_hood_collar_bk (P)"(
    ComponentPose=Local To Component(LocalPose=Blend Poses by bool(BlendPose_0=@n37, BlendPose_1=@n81, BlendTime_0=0.2, BlendTime_1=0.2) {blendType=Linear})
    Alpha=@n82
    ConstraintWeights_0=1
   ) {
    alphaBoolBlend {blendInTime=0.3, blendOption=ExpInOut, blendOutTime=0.3}
    alphaScaleBiasClamp {bInterpResult=true}
    boneToModify {boneName=dyn_hood_collar_bk}
    constraintSetup [{targetBone {boneName=spine_05}}]
    constraintWeights [1]
   }
   n84 = "Constraint - dyn_hood_collar_bk_r (P)"(ComponentPose=@n83, Alpha=@n82, ConstraintWeights_0=1) {
    alphaBoolBlend {blendInTime=0.3, blendOption=ExpInOut, blendOutTime=0.3}
    alphaScaleBiasClamp {bInterpResult=true}
    boneToModify {boneName=dyn_hood_collar_bk_r}
    constraintSetup [{targetBone {boneName=spine_05}}]
    constraintWeights [1]
   }
   n85 = "Constraint - dyn_hood_collar_bk_l (P)"(ComponentPose=@n84, Alpha=@n82, ConstraintWeights_0=1) {
    alphaBoolBlend {blendInTime=0.3, blendOption=ExpInOut, blendOutTime=0.3}
    alphaScaleBiasClamp {bInterpResult=true}
    boneToModify {boneName=dyn_hood_collar_bk_l}
    constraintSetup [{targetBone {boneName=spine_05}}]
    constraintWeights [1]
   }
   n86 = "Constraint - dyn_hood_l_01 (P)"(ComponentPose=@n85, Alpha=@n82, ConstraintWeights_0=1) {
    alphaBoolBlend {blendOption=Linear}
    alphaScaleBiasClamp {bInterpResult=true}
    boneToModify {boneName=dyn_hood_l_01}
    constraintSetup [{targetBone {boneName=spine_05}}]
    constraintWeights [1]
   }
   n87 = "Constraint - dyn_hood_r_01 (P)"(ComponentPose=@n86, Alpha=@n82, ConstraintWeights_0=1) {
    alphaBoolBlend {blendOption=Linear}
    alphaScaleBiasClamp {bInterpResult=true}
    boneToModify {boneName=dyn_hood_r_01}
    constraintSetup [{targetBone {boneName=spine_05}}]
    constraintWeights [1]
   }
   n88 = "Constraint - dyn_fr_flap (PP)"!(ComponentPose=@n87, Alpha=0.8, ConstraintWeights_0=0.6, ConstraintWeights_1:float=0.4) -> (Pose) {
    alphaBoolBlend {blendOption=Linear}
    boneToModify {boneName=dyn_fr_flap}
    constraintSetup [{targetBone {boneName=dyn_fr_flap_l}}, {targetBone {boneName=dyn_fr_flap_r}}]
    constraintWeights [1,1]
   }
   n89 = Get ThighSumAlpha()
   n90 = "Transform (Modify) Bone - Bone: dyn_fr_flap_r_01"(ComponentPose=@n88, Alpha=@n89, Translation="0,0,0", Rotation="30,0,-15", Scale="1,1,1") {
    rotationMode=BMM_Additive
    translationMode=BMM_Additive
    alphaBoolBlend {blendOption=Linear}
    alphaScaleBiasClamp {bInterpResult=true}
    boneToModify {boneName=dyn_fr_flap_r_01}
    rotation {pitch=30, roll=-15}
   }
   n91 = "Transform (Modify) Bone - Bone: dyn_fr_flap_r_01"(ComponentPose=@n90, Alpha=Get Thigh RCrouch Alpha() -> (Thigh RCrouch Alpha:double), Translation="0,0,0", Rotation="60,0,-10", Scale="1,1,1") {
    rotationMode=BMM_Additive
    translationMode=BMM_Additive
    alphaBoolBlend {blendOption=Linear}
    alphaScaleBiasClamp {bInterpResult=true}
    boneToModify {boneName=dyn_fr_flap_r_01}
    rotation {pitch=60, roll=-10}
   }
   n92 = "Transform (Modify) Bone - Bone: dyn_fr_flap"(ComponentPose=@n91, Alpha=@n89, Translation="0,0,0", Rotation="0,0,0", Scale="1,1,1") {
    rotationMode=BMM_Additive
    translationMode=BMM_Additive
    alphaBoolBlend {blendOption=Linear}
    alphaScaleBiasClamp {bInterpResult=true}
    boneToModify {boneName=dyn_fr_flap}
   }
   n93 = "Constraint - dyn_fr_jacket (PP)"!(ComponentPose=@n92, Alpha=1, ConstraintWeights_0=0.8, ConstraintWeights_1:float=0.2) -> (Pose) {
    alphaBoolBlend {blendOption=Linear}
    boneToModify {boneName=dyn_fr_jacket}
    constraintSetup [{targetBone {boneName=dyn_fr_flap}}, {targetBone {boneName=pelvis}}]
    constraintWeights [1,1]
   }
   n94 = "Constraint - dyn_fr_jacket_r (PP)"!(ComponentPose=@n93, Alpha=1, ConstraintWeights_0=0.8, ConstraintWeights_1:float=0.2) -> (Pose) {
    alphaBoolBlend {blendOption=Linear}
    boneToModify {boneName=dyn_fr_jacket_r}
    constraintSetup [{targetBone {boneName=dyn_fr_flap_r_01}}, {targetBone {boneName=spine_01}}]
    constraintWeights [1,1]
   }
   n95 = Get TargetingHeadUpAlpha() -> (TargetingHeadUpAlpha:double)
   n96 = "Transform (Modify) Bone - Bone: dyn_hood_collar_bk"(ComponentPose=@n94, Alpha=@n95, Translation="-5,0,-5", Rotation="0,0,0", Scale="1,1,1") {
    rotationMode=BMM_Additive
    translationMode=BMM_Additive
    alphaBoolBlend {blendOption=Linear}
    alphaScaleBiasClamp {bInterpResult=true}
    boneToModify {boneName=dyn_hood_collar_bk}
    translation {x=-5, z=-5}
   }
   n97 = "Transform (Modify) Bone - Bone: dyn_hood_collar_bk_r"!(ComponentPose=@n96, Alpha:float=@n95, Translation="-10,0,-10", Rotation="0,0,0", Scale="1,1,1") -> (Pose) {
    rotationMode=BMM_Additive
    translationMode=BMM_Additive
    alphaBoolBlend {blendOption=Linear}
    alphaScaleBiasClamp {bInterpResult=true}
    boneToModify {boneName=dyn_hood_collar_bk_r}
    translation {x=-10, z=-10}
   }
   n98 = "Transform (Modify) Bone - Bone: dyn_hood_collar_r"!(ComponentPose=@n97, Alpha:float=@n95, Translation="-5,0,-2", Rotation="0,0,0", Scale="1,1,1") -> (Pose) {
    rotationMode=BMM_Additive
    translationMode=BMM_Additive
    alphaBoolBlend {blendOption=Linear}
    alphaScaleBiasClamp {bInterpResult=true}
    boneToModify {boneName=dyn_hood_collar_r}
    translation {x=-5, z=-2}
   }
   n99 = Get Jacket_LeftSide_Corection() -> (Jacket_LeftSide_Corection:float)
   n100 = "Transform (Modify) Bone - Bone: dyn_l_jacket_b_01"(ComponentPose=@n98, Alpha=@n99, Translation="0,0,2", Rotation="-30,10,20", Scale="1,1,1") {
    rotationMode=BMM_Additive
    rotationSpace=BCS_BoneSpace
    scaleMode=BMM_Additive
    scaleSpace=BCS_BoneSpace
    translationMode=BMM_Additive
    translationSpace=BCS_BoneSpace
    alphaBoolBlend {blendOption=Linear}
    alphaScaleBiasClamp {bInterpResult=true, interpSpeedDecreasing=3, interpSpeedIncreasing=20}
    boneToModify {boneName=dyn_l_jacket_b_01}
    rotation=<rot -30 10 20>
    translation {z=2}
   }
   n101 = "Transform (Modify) Bone - Bone: dyn_fr_jacket_l_01"(ComponentPose=@n100, Alpha=@n99, Translation="-3,-4,1.5", Rotation="0,0,0", Scale="1,1,1") {
    rotationMode=BMM_Additive
    rotationSpace=BCS_BoneSpace
    scaleMode=BMM_Additive
    scaleSpace=BCS_BoneSpace
    translationMode=BMM_Additive
    translationSpace=BCS_BoneSpace
    alphaBoolBlend {blendOption=Linear}
    alphaScaleBiasClamp {bInterpResult=true, interpSpeedDecreasing=3, interpSpeedIncreasing=20}
    boneToModify {boneName=dyn_fr_jacket_l_01}
    translation=<-3 -4 1.5>
   }
   n102 = Get IsWearingBackbling() -> (IsWearingBackbling:bool)
   n103 = "Transform (Modify) Bone - Bone: dyn_hood_collar_l"!(ComponentPose=@n101, bAlphaBoolEnabled:bool=@n102, Translation="2,0,0", Rotation="0,0,0", Scale="1,1,1") -> (Pose) {
    alphaInputType=Bool
    rotationMode=BMM_Additive
    rotationSpace=BCS_BoneSpace
    scaleMode=BMM_Additive
    scaleSpace=BCS_BoneSpace
    translationMode=BMM_Additive
    translationSpace=BCS_BoneSpace
    alphaBoolBlend {blendOption=Linear}
    alphaScaleBiasClamp {bInterpResult=true}
    boneToModify {boneName=dyn_hood_collar_l}
    translation {x=2}
   }
   n104 = "Transform (Modify) Bone - Bone: dyn_hood_collar_r"!(ComponentPose=@n103, bAlphaBoolEnabled:bool=@n102, Translation="2,0,0", Rotation="0,0,0", Scale="1,1,1") -> (Pose) {
    alphaInputType=Bool
    rotationMode=BMM_Additive
    rotationSpace=BCS_BoneSpace
    scaleMode=BMM_Additive
    scaleSpace=BCS_BoneSpace
    translationMode=BMM_Additive
    translationSpace=BCS_BoneSpace
    alphaBoolBlend {blendOption=Linear}
    alphaScaleBiasClamp {bInterpResult=true}
    boneToModify {boneName=dyn_hood_collar_r}
    translation {x=2}
   }
   n105 = "Transform (Modify) Bone - Bone: dyn_hood_r"!(ComponentPose=@n104, bAlphaBoolEnabled:bool=@n102, Translation="-2,0,0", Rotation="0,0,25", Scale="1,1,1") -> (Pose) {
    alphaInputType=Bool
    rotationMode=BMM_Additive
    rotationSpace=BCS_BoneSpace
    scaleMode=BMM_Additive
    scaleSpace=BCS_BoneSpace
    translationMode=BMM_Additive
    translationSpace=BCS_BoneSpace
    alphaBoolBlend {blendOption=Linear}
    alphaScaleBiasClamp {bInterpResult=true}
    boneToModify {boneName=dyn_hood_r}
    rotation {roll=25}
    translation {x=-2}
   }
   n106 = "Transform (Modify) Bone - Bone: dyn_hood_l"!(ComponentPose=@n105, bAlphaBoolEnabled:bool=@n102, Translation="-3,0,0", Rotation="0,0,-25", Scale="1,1,1") -> (Pose) {
    alphaInputType=Bool
    rotationMode=BMM_Additive
    rotationSpace=BCS_BoneSpace
    scaleMode=BMM_Additive
    scaleSpace=BCS_BoneSpace
    translationMode=BMM_Additive
    translationSpace=BCS_BoneSpace
    alphaBoolBlend {blendOption=Linear}
    alphaScaleBiasClamp {bInterpResult=true}
    boneToModify {boneName=dyn_hood_l}
    rotation {roll=-25}
    translation {x=-3}
   }
   n107 = "Transform (Modify) Bone - Bone: dyn_hood_collar_bk_r"!(ComponentPose=@n106, bAlphaBoolEnabled:bool=@n102, Translation="-1,-5,-2", Rotation="-5,10,0", Scale="1,1,1") -> (Pose) {
    alphaInputType=Bool
    rotationMode=BMM_Additive
    rotationSpace=BCS_BoneSpace
    scaleMode=BMM_Additive
    scaleSpace=BCS_BoneSpace
    translationMode=BMM_Additive
    translationSpace=BCS_BoneSpace
    alphaBoolBlend {blendOption=Linear}
    alphaScaleBiasClamp {bInterpResult=true}
    boneToModify {boneName=dyn_hood_collar_bk_r}
    rotation {pitch=-5, yaw=10}
    translation=<-1 -5 -2>
   }
   n108 = "Transform (Modify) Bone - Bone: dyn_hood_collar_bk_l"!(ComponentPose=@n107, bAlphaBoolEnabled:bool=@n102, Translation="-1,0,-2", Rotation="-5,10,0", Scale="1,1,1") -> (Pose) {
    alphaInputType=Bool
    rotationMode=BMM_Additive
    rotationSpace=BCS_BoneSpace
    scaleMode=BMM_Additive
    scaleSpace=BCS_BoneSpace
    translationMode=BMM_Additive
    translationSpace=BCS_BoneSpace
    alphaBoolBlend {blendOption=Linear}
    alphaScaleBiasClamp {bInterpResult=true}
    boneToModify {boneName=dyn_hood_collar_bk_l}
    rotation {pitch=-5, yaw=10}
    translation {x=-1, z=-2}
   }
   n109 = "Transform (Modify) Bone - Bone: dyn_hood_collar_bk"!(ComponentPose=@n108, bAlphaBoolEnabled:bool=@n102, Translation="-5.5,0,-1.5", Rotation="-15,0,0", Scale="1,1,1") -> (Pose) {
    alphaInputType=Bool
    rotationMode=BMM_Additive
    rotationSpace=BCS_BoneSpace
    scaleMode=BMM_Additive
    scaleSpace=BCS_BoneSpace
    translationMode=BMM_Additive
    translationSpace=BCS_BoneSpace
    alphaBoolBlend {blendOption=Linear}
    alphaScaleBiasClamp {bInterpResult=true}
    boneToModify {boneName=dyn_hood_collar_bk}
    rotation {pitch=-15}
    translation {x=-5.5, z=-1.5}
   }
   n110 = "Transform (Modify) Bone - Bone: dyn_hood_collar_bk_r"(ComponentPose=@n109, bAlphaBoolEnabled=True, Translation="0,8,1", Rotation="0,0,10", Scale="1,1,1") {
    alphaInputType=Bool
    rotationMode=BMM_Additive
    rotationSpace=BCS_BoneSpace
    scaleMode=BMM_Additive
    scaleSpace=BCS_BoneSpace
    translationMode=BMM_Additive
    translationSpace=BCS_BoneSpace
    alphaBoolBlend {blendInTime=0.5, blendOption=ExpInOut, blendOutTime=0.5}
    boneToModify {boneName=dyn_hood_collar_bk_r}
    rotation {roll=10}
    translation {y=8, z=1}
   }
   n111 = "Transform (Modify) Bone - Bone: dyn_hood_collar_r"(ComponentPose=@n110, bAlphaBoolEnabled=True, Translation="3,0,-2", Rotation="0,0,0", Scale="1,1,1") {
    alphaInputType=Bool
    rotationMode=BMM_Additive
    rotationSpace=BCS_BoneSpace
    scaleMode=BMM_Additive
    scaleSpace=BCS_BoneSpace
    translationMode=BMM_Additive
    translationSpace=BCS_BoneSpace
    alphaBoolBlend {blendInTime=0.5, blendOption=Linear, blendOutTime=0.5}
    alphaScaleBiasClamp {bInterpResult=true}
    boneToModify {boneName=dyn_hood_collar_r}
    translation {x=3, z=-2}
   }
   n112 = "Transform (Modify) Bone - Bone: dyn_hood_collar_bk"(ComponentPose=@n111, bAlphaBoolEnabled=True, Translation="5,0,0", Rotation="0,0,0", Scale="1,1,1") {
    alphaInputType=Bool
    rotationMode=BMM_Additive
    rotationSpace=BCS_BoneSpace
    scaleMode=BMM_Additive
    scaleSpace=BCS_BoneSpace
    translationMode=BMM_Additive
    translationSpace=BCS_BoneSpace
    alphaBoolBlend {blendInTime=0.3, blendOption=ExpInOut, blendOutTime=0.3}
    alphaScaleBiasClamp {bInterpResult=true}
    boneToModify {boneName=dyn_hood_collar_bk}
    translation {x=5}
   }
   n113 = "Transform (Modify) Bone - Bone: dyn_hood_collar_bk_l"(ComponentPose=@n112, bAlphaBoolEnabled=True, Translation="4,0,-3", Rotation="0,0,0", Scale="1,1,1") {
    alphaInputType=Bool
    rotationMode=BMM_Additive
    rotationSpace=BCS_BoneSpace
    scaleMode=BMM_Additive
    scaleSpace=BCS_BoneSpace
    translationMode=BMM_Additive
    translationSpace=BCS_BoneSpace
    alphaBoolBlend {blendInTime=0.3, blendOption=ExpInOut, blendOutTime=0.3}
    alphaScaleBiasClamp {bInterpResult=true}
    boneToModify {boneName=dyn_hood_collar_bk_l}
    translation {x=4, z=-3}
   }
   n114 = "Transform (Modify) Bone - Bone: dyn_hood_collar_bk_r"(ComponentPose=@n113, bAlphaBoolEnabled=True, Translation="5,0,3", Rotation="0,0,0", Scale="1,1,1") {
    alphaInputType=Bool
    rotationMode=BMM_Additive
    rotationSpace=BCS_BoneSpace
    scaleMode=BMM_Additive
    scaleSpace=BCS_BoneSpace
    translationMode=BMM_Additive
    translationSpace=BCS_BoneSpace
    alphaBoolBlend {blendInTime=0.3, blendOption=ExpInOut, blendOutTime=0.3}
    alphaScaleBiasClamp {bInterpResult=true}
    boneToModify {boneName=dyn_hood_collar_bk_r}
    translation {x=5, z=3}
   }
   n115 = Get TargetingHeadUpAlpha()
   n116 = "Transform (Modify) Bone - Bone: dyn_hood_collar_bk"!(ComponentPose=@n114, Alpha:float=@n115, Translation="1,0,0", Rotation="0,0,0", Scale="1,1,1") -> (Pose) {
    rotationMode=BMM_Additive
    rotationSpace=BCS_BoneSpace
    scaleMode=BMM_Additive
    scaleSpace=BCS_BoneSpace
    translationMode=BMM_Additive
    translationSpace=BCS_BoneSpace
    alphaBoolBlend {blendInTime=0.3, blendOption=ExpInOut, blendOutTime=0.3}
    alphaScaleBiasClamp {bInterpResult=true}
    boneToModify {boneName=dyn_hood_collar_bk}
    translation {x=1}
   }
   n117 = "Transform (Modify) Bone - Bone: dyn_hood_collar_bk_l"!(ComponentPose=@n116, Alpha:float=@n115, Translation="1,0,0", Rotation="0,0,0", Scale="1,1,1") -> (Pose) {
    rotationMode=BMM_Additive
    rotationSpace=BCS_BoneSpace
    scaleMode=BMM_Additive
    scaleSpace=BCS_BoneSpace
    translationMode=BMM_Additive
    translationSpace=BCS_BoneSpace
    alphaBoolBlend {blendInTime=0.3, blendOption=ExpInOut, blendOutTime=0.3}
    alphaScaleBiasClamp {bInterpResult=true}
    boneToModify {boneName=dyn_hood_collar_bk_l}
    translation {x=1}
   }
   n118 = "Transform (Modify) Bone - Bone: dyn_hood_collar_bk_r"!(ComponentPose=@n117, Alpha:float=@n115, Translation="2,0,0", Rotation="0,0,0", Scale="1,1,1") -> (Pose) {
    rotationMode=BMM_Additive
    rotationSpace=BCS_BoneSpace
    scaleMode=BMM_Additive
    scaleSpace=BCS_BoneSpace
    translationMode=BMM_Additive
    translationSpace=BCS_BoneSpace
    alphaBoolBlend {blendInTime=0.3, blendOption=ExpInOut, blendOutTime=0.3}
    alphaScaleBiasClamp {bInterpResult=true}
    boneToModify {boneName=dyn_hood_collar_bk_r}
    translation {x=2}
   }
   n119 = "Transform (Modify) Bone - Bone: dyn_fr_flap"!(ComponentPose=@n118, bAlphaBoolEnabled:bool=True, Translation="0,0,-2", Rotation="-10,0,5", Scale="1,1,1") -> (Pose) {
    alphaInputType=Bool
    rotationMode=BMM_Additive
    translationMode=BMM_Additive
    alphaBoolBlend {blendInTime=0.3, blendOption=ExpInOut, blendOutTime=0.3}
    alphaScaleBiasClamp {bInterpResult=true}
    boneToModify {boneName=dyn_fr_flap}
    rotation {pitch=-10, roll=5}
    translation {z=-2}
   }
   n120 = "Transform (Modify) Bone - Bone: dyn_r_flap_a"!(ComponentPose=@n119, bAlphaBoolEnabled:bool=True, Translation="-2,0,2", Rotation="0,-20,50", Scale="1,1,1") -> (Pose) {
    alphaInputType=Bool
    rotationMode=BMM_Additive
    translationMode=BMM_Additive
    alphaBoolBlend {blendOption=Linear}
    alphaScaleBiasClamp {bInterpResult=true, interpSpeedDecreasing=5, interpSpeedIncreasing=5}
    boneToModify {boneName=dyn_r_flap_a}
    rotation {roll=50, yaw=-20}
    translation {x=-2, z=2}
   }
   n121 = "Transform (Modify) Bone - Bone: dyn_fr_flap_r_01"!(ComponentPose=@n120, bAlphaBoolEnabled:bool=True, Translation="-4,0,0", Rotation="0,0,-50", Scale="1,1,1") -> (Pose) {
    alphaInputType=Bool
    rotationMode=BMM_Additive
    translationMode=BMM_Additive
    alphaBoolBlend {blendOption=Linear}
    alphaScaleBiasClamp {bInterpResult=true}
    boneToModify {boneName=dyn_fr_flap_r_01}
    rotation {roll=-50}
    translation {x=-4}
   }
   n122 = "Transform (Modify) Bone - Bone: dyn_fr_flap"(ComponentPose=@n121, bAlphaBoolEnabled=True, Translation="0,0,-4", Rotation="0,0,-40", Scale="1,1,1") {
    alphaInputType=Bool
    rotationMode=BMM_Additive
    translationMode=BMM_Additive
    alphaBoolBlend {blendOption=Linear}
    alphaScaleBiasClamp {bInterpResult=true}
    boneToModify {boneName=dyn_fr_flap}
    rotation {roll=-40}
    translation {z=-4}
   }
   n123 = "Transform (Modify) Bone - Bone: dyn_fr_flap_l_01"(ComponentPose=@n122, bAlphaBoolEnabled=True, Translation="5,0,0", Rotation="0,0,-50", Scale="1,1,1") {
    alphaInputType=Bool
    rotationMode=BMM_Additive
    translationMode=BMM_Additive
    alphaBoolBlend {blendOption=Linear}
    alphaScaleBiasClamp {bInterpResult=true}
    boneToModify {boneName=dyn_fr_flap_l_01}
    rotation {roll=-50}
    translation {x=5}
   }
   n124 = "Constraint - dyn_zipper_r_MSTR (PPPPPPPP)"!(
    ComponentPose=@n123
    Alpha=1
    ConstraintWeights_0=0.058
    ConstraintWeights_1:float=0.04
    ConstraintWeights_2:float=0.21
    ConstraintWeights_3:float=0.072
    ConstraintWeights_4:float=0.158
    ConstraintWeights_5:float=0.275
    ConstraintWeights_6:float=0.135
    ConstraintWeights_7:float=0.054
   ) -> (Pose) {
    alphaBoolBlend {blendOption=Linear}
    boneToModify {boneName=dyn_zipper_r_MSTR}
    constraintSetup [
     {targetBone {boneName=dyn_fr_jacket_01}}
     {targetBone {boneName=dyn_fr_jacket_r}}
     {targetBone {boneName=dyn_fr_jacket_r_01}}
     {targetBone {boneName=dyn_r_jacket_b}}
     {targetBone {boneName=dyn_r_jacket_b_01}}
     {targetBone {boneName=pelvis}}
     {targetBone {boneName=spine_01}}
     {targetBone {boneName=spine_02}}
    ]
    constraintWeights [0,1,1,1,1,1,1,1]
   }
   n125 = "Transform (Modify) Bone - Bone: dyn_zipper_r_MSTR"(ComponentPose=@n124, Alpha=1, Translation="0.1,-0.1,0.5", Rotation="7,0,0", Scale="1,1,1") {
    rotationMode=BMM_Additive
    rotationSpace=BCS_BoneSpace
    translationMode=BMM_Additive
    translationSpace=BCS_BoneSpace
    alphaBoolBlend {blendOption=Linear}
    boneToModify {boneName=dyn_zipper_r_MSTR}
    rotation {pitch=7}
    translation=<0.1 -0.1 0.5>
   }
   n126 = "Constraint - dyn_zipper_l_MSTR (PPPPPPPP)"!(
    ComponentPose=@n125
    Alpha=1
    ConstraintWeights_0=0.04
    ConstraintWeights_1:float=0.068
    ConstraintWeights_2:float=0.144
    ConstraintWeights_3:float=0.11
    ConstraintWeights_4:float=0.242
    ConstraintWeights_5:float=0.175
    ConstraintWeights_6:float=0.18
    ConstraintWeights_7:float=0.041
   ) -> (Pose) {
    alphaBoolBlend {blendOption=Linear}
    boneToModify {boneName=dyn_zipper_l_MSTR}
    constraintSetup [
     {targetBone {boneName=dyn_fr_jacket_01}}
     {targetBone {boneName=dyn_fr_jacket_l}}
     {targetBone {boneName=dyn_fr_jacket_l_01}}
     {targetBone {boneName=spine_01}}
     {targetBone {boneName=pelvis}}
     {targetBone {boneName=dyn_l_jacket_b}}
     {targetBone {boneName=dyn_l_jacket_b_01}}
     {targetBone {boneName=spine_02}}
    ]
    constraintWeights [1,1,1,1,1,1,1,1]
   }
   n127 = "Transform (Modify) Bone - Bone: dyn_zipper_l_MSTR"(ComponentPose=@n126, Alpha=1, Translation="0.3,-0.1,1", Rotation="0,0,0", Scale="1,1,1") {
    translationMode=BMM_Additive
    translationSpace=BCS_BoneSpace
    alphaBoolBlend {blendOption=Linear}
    boneToModify {boneName=dyn_zipper_l_MSTR}
    translation=<0.3 -0.1 1>
   }
   n128 = "Transform (Modify) Bone - Bone: dyn_zipper_l_MSTR"(ComponentPose=@n127, Alpha=Get ThighLAlpha() -> (ThighLAlpha:double), Translation="0,0,0.5", Rotation="40,-30,0", Scale="1,1,1") {
    rotationMode=BMM_Additive
    rotationSpace=BCS_BoneSpace
    translationMode=BMM_Additive
    translationSpace=BCS_BoneSpace
    alphaBoolBlend {blendInTime=0.2, blendOption=Linear, blendOutTime=0.2}
    alphaScaleBiasClamp {bInterpResult=true}
    boneToModify {boneName=dyn_zipper_l_MSTR}
    rotation {pitch=40, yaw=-30}
    translation {z=0.5}
   }
   n129 = RigidBody!(ComponentPose=Local To Component(LocalPose=Use cached pose 'RBAN'()), bAlphaBoolEnabled:bool, ExternalForce) -> (Pose) {
    alphaInputType=Bool
    bClampLinearTranslationLimitToRefPose=true
    bOverrideWorldGravity=true
    overridePhysicsAsset=&/M_MED_Pumpkin_Punk_Physics_SkyDiving
    simulationSpace=BaseBoneSpace
    alphaBoolBlend {blendOption=Linear}
    baseBoneRef {boneName=pelvis}
    componentAppliedLinearAccClamp=<2000 2000 2000>
    overrideWorldGravity {z=-980}
   }
   n130 = "Transform (Modify) Bone - Bone: dyn_l_flap_a"!(ComponentPose=@n129, bAlphaBoolEnabled:bool=True, Translation="0,2,-5", Rotation="30,-60,0", Scale="1,1,1") -> (Pose) {
    alphaInputType=Bool
    rotationMode=BMM_Additive
    rotationSpace=BCS_BoneSpace
    translationMode=BMM_Additive
    translationSpace=BCS_BoneSpace
    alphaBoolBlend {blendOption=Linear}
    boneToModify {boneName=dyn_l_flap_a}
    rotation {pitch=30, yaw=-60}
    translation {y=2, z=-5}
   }
   n131 = "Transform (Modify) Bone - Bone: dyn_r_flap_a"(ComponentPose=@n130, bAlphaBoolEnabled=True, Translation="0,3,-5", Rotation="30,0,0", Scale="1,1,1") {
    alphaInputType=Bool
    rotationMode=BMM_Additive
    rotationSpace=BCS_BoneSpace
    translationMode=BMM_Additive
    translationSpace=BCS_BoneSpace
    alphaBoolBlend {blendOption=Linear}
    boneToModify {boneName=dyn_r_flap_a}
    rotation {pitch=30}
    translation {y=3, z=-5}
   }
   n132 = "Transform (Modify) Bone - Bone: dyn_l_flap_b"(ComponentPose=@n131, bAlphaBoolEnabled=True, Translation="0,3,-5", Rotation="30,0,0", Scale="1,1,1") {
    alphaInputType=Bool
    rotationMode=BMM_Additive
    rotationSpace=BCS_BoneSpace
    translationMode=BMM_Additive
    translationSpace=BCS_BoneSpace
    alphaBoolBlend {blendOption=Linear}
    boneToModify {boneName=dyn_l_flap_b}
    rotation {pitch=30}
    translation {y=3, z=-5}
   }
   n133 = "Transform (Modify) Bone - Bone: dyn_bk_jacket_r"(ComponentPose=@n132, Alpha=1, Translation="0,0,0", Rotation="30,0,0", Scale="1,1,1") {
    rotationMode=BMM_Replace
    rotationSpace=BCS_BoneSpace
    scaleSpace=BCS_BoneSpace
    translationMode=BMM_Replace
    translationSpace=BCS_BoneSpace
    alphaBoolBlend {blendOption=Linear}
    boneToModify {boneName=dyn_bk_jacket_r}
    rotation {pitch=30}
   }
   n134 = "Transform (Modify) Bone - Bone: dyn_bk_jacket"(ComponentPose=@n133, Alpha=1, Translation="0,0,2", Rotation="10,0,0", Scale="1,1,1") {
    rotationMode=BMM_Replace
    rotationSpace=BCS_BoneSpace
    scaleSpace=BCS_BoneSpace
    translationMode=BMM_Replace
    translationSpace=BCS_BoneSpace
    alphaBoolBlend {blendOption=Linear}
    boneToModify {boneName=dyn_bk_jacket}
    rotation {pitch=10}
    translation {z=2}
   }
   n135 = "Transform (Modify) Bone - Bone: dyn_bk_jacket_l"(ComponentPose=@n134, Alpha=1, Translation="0,5,0.5", Rotation="30,0,0", Scale="1,1,1") {
    rotationMode=BMM_Replace
    rotationSpace=BCS_BoneSpace
    scaleSpace=BCS_BoneSpace
    translationMode=BMM_Replace
    translationSpace=BCS_BoneSpace
    alphaBoolBlend {blendOption=Linear}
    boneToModify {boneName=dyn_bk_jacket_l}
    rotation {pitch=30}
    translation {y=5, z=0.5}
   }
   n136 = Get bIsSkyDiving()
   Output Pose(Result=Blend Poses by bool(BlendPose_0=Component To Local(ComponentPose=@n128), BlendPose_1=Component To Local(ComponentPose=@n135), bActiveValue=NOT Boolean(A=@n136)) {blendType=Linear}) {layerGroup=None}
   n137 = "Transform (Modify) Bone - Bone: dyn_r_flap_a"!(
    ComponentPose=Local To Component(LocalPose=Copy Pose From Mesh() {bUseAttachedParent=true})
    Alpha:float=1
    Translation="0.000,0.000,2"
    Rotation="0,0,0"
    Scale="1,1,1"
   ) -> (Pose) {
    rotationMode=BMM_Additive
    rotationSpace=BCS_BoneSpace
    scaleSpace=BCS_BoneSpace
    translationMode=BMM_Replace
    translationSpace=BCS_BoneSpace
    alphaBoolBlend {blendOption=Linear}
    boneToModify {boneName=dyn_r_flap_a}
    translation {z=2}
   }
   n138 = "Transform (Modify) Bone - Bone: dyn_r_flap_b"(ComponentPose=@n137, Alpha=1, Translation="0.000,0.000,2", Rotation="0,0,0", Scale="1,1,1") {
    rotationMode=BMM_Additive
    rotationSpace=BCS_BoneSpace
    scaleSpace=BCS_BoneSpace
    translationMode=BMM_Replace
    translationSpace=BCS_BoneSpace
    alphaBoolBlend {blendOption=Linear}
    boneToModify {boneName=dyn_r_flap_b}
    translation {z=2}
   }
   n139 = "Transform (Modify) Bone - Bone: dyn_l_flap_a"!(ComponentPose=@n138, Alpha:float=1, Translation="0,0,2", Rotation="15,0,0", Scale="1,1,1") -> (Pose) {
    rotationMode=BMM_Additive
    rotationSpace=BCS_BoneSpace
    scaleSpace=BCS_BoneSpace
    translationMode=BMM_Replace
    translationSpace=BCS_BoneSpace
    alphaBoolBlend {blendOption=Linear}
    boneToModify {boneName=dyn_l_flap_a}
    rotation {pitch=15}
    translation {z=2}
   }
   n140 = "Transform (Modify) Bone - Bone: dyn_l_flap_b"!(ComponentPose=@n139, Alpha:float=1, Translation="0.000,0.000,2", Rotation="15,0.000,0.000", Scale="1,1,1") -> (Pose) {
    rotationMode=BMM_Additive
    rotationSpace=BCS_BoneSpace
    scaleSpace=BCS_BoneSpace
    translationMode=BMM_Replace
    translationSpace=BCS_BoneSpace
    alphaBoolBlend {blendOption=Linear}
    boneToModify {boneName=dyn_l_flap_b}
    rotation {pitch=15}
    translation {z=2}
   }
   n141 = "Transform (Modify) Bone - Bone: dyn_bk_flap_l_01"!(ComponentPose=@n140, Alpha:float=1, Translation="0,0,0", Rotation="10,0.000,0.000", Scale="1,1,1") -> (Pose) {
    rotationMode=BMM_Additive
    rotationSpace=BCS_BoneSpace
    scaleSpace=BCS_BoneSpace
    translationMode=BMM_Replace
    translationSpace=BCS_BoneSpace
    alphaBoolBlend {blendOption=Linear}
    boneToModify {boneName=dyn_bk_flap_l_01}
    rotation {pitch=10}
   }
   n142 = "Transform (Modify) Bone - Bone: dyn_bk_flap"(ComponentPose=@n141, Alpha=1, Translation="0,0,0", Rotation="10,0.000,0.000", Scale="1,1,1") {
    rotationMode=BMM_Additive
    rotationSpace=BCS_BoneSpace
    scaleSpace=BCS_BoneSpace
    translationMode=BMM_Replace
    translationSpace=BCS_BoneSpace
    alphaBoolBlend {blendOption=Linear}
    boneToModify {boneName=dyn_bk_flap}
    rotation {pitch=10}
   }
   n143 = "Transform (Modify) Bone - Bone: dyn_bk_flap_r_01"!(ComponentPose=@n142, Alpha:float=1, Translation="0,0,0", Rotation="10,0,0", Scale="1,1,1") -> (Pose) {
    rotationMode=BMM_Additive
    rotationSpace=BCS_BoneSpace
    scaleSpace=BCS_BoneSpace
    translationMode=BMM_Replace
    translationSpace=BCS_BoneSpace
    alphaBoolBlend {blendOption=Linear}
    boneToModify {boneName=dyn_bk_flap_r_01}
    rotation {pitch=10}
   }
   n144 = "Constraint - dyn_fr_flap_l_03 (PPPPPPPP)"!(
    ComponentPose=@n143
    Alpha=1
    ConstraintWeights_0=0
    ConstraintWeights_1:float=0.0039
    ConstraintWeights_2:float=0.0516
    ConstraintWeights_3:float=0.001
    ConstraintWeights_4:float=0.8131
    ConstraintWeights_5:float=0.0517
    ConstraintWeights_6:float=0.0786
    ConstraintWeights_7:float=0.0001
   ) -> (Pose) {
    alphaBoolBlend {blendOption=Linear}
    boneToModify {boneName=dyn_fr_flap_l_03}
    constraintSetup [
     {targetBone {boneName=calf_twist_02_l}}
     {targetBone {boneName=thigh_r}}
     {targetBone {boneName=thigh_twist_01_r}}
     {targetBone {boneName=spine_01}}
     {targetBone {boneName=thigh_twist_01_l}}
     {targetBone {boneName=thigh_l}}
     {targetBone {boneName=pelvis}}
     {targetBone {boneName=calf_l}}
    ]
    constraintWeights [1,1,1,1,1,1,1,1]
   }
   n145 = "Constraint - dyn_fr_MSTR (PPPPPPPP)"!(
    ComponentPose=@n144
    Alpha=1
    ConstraintWeights_0=0
    ConstraintWeights_1:float=0.0295
    ConstraintWeights_2:float=0.3927
    ConstraintWeights_3:float=0.0027
    ConstraintWeights_4:float=0.3628
    ConstraintWeights_5:float=0.0265
    ConstraintWeights_6:float=0.1858
    ConstraintWeights_7:float=0
   ) -> (Pose) {
    alphaBoolBlend {blendOption=Linear}
    boneToModify {boneName=dyn_fr_MSTR}
    constraintSetup [
     {targetBone {boneName=calf_r}}
     {targetBone {boneName=thigh_r}}
     {targetBone {boneName=thigh_twist_01_r}}
     {targetBone {boneName=spine_01}}
     {targetBone {boneName=thigh_twist_01_l}}
     {targetBone {boneName=thigh_l}}
     {targetBone {boneName=pelvis}}
     {targetBone {boneName=calf_l}}
    ]
    constraintWeights [1,1,1,1,1,1,1,1]
   }
   n146 = "Constraint - dyn_fr_flap_r_03 (PPPPPPPP)"!(
    ComponentPose=@n145
    Alpha=1
    ConstraintWeights_0=0
    ConstraintWeights_1:float=0.0541
    ConstraintWeights_2:float=0.7121
    ConstraintWeights_3:float=0.0007
    ConstraintWeights_4:float=0.053
    ConstraintWeights_5:float=0.0043
    ConstraintWeights_6:float=0.1757
    ConstraintWeights_7:float=0
   ) -> (Pose) {
    alphaBoolBlend {blendOption=Linear}
    boneToModify {boneName=dyn_fr_flap_r_03}
    constraintSetup [
     {targetBone {boneName=calf_r}}
     {targetBone {boneName=thigh_r}}
     {targetBone {boneName=thigh_twist_01_r}}
     {targetBone {boneName=spine_01}}
     {targetBone {boneName=thigh_twist_01_l}}
     {targetBone {boneName=thigh_l}}
     {targetBone {boneName=pelvis}}
     {targetBone {boneName=calf_l}}
    ]
    constraintWeights [1,1,1,1,1,1,1,1]
   }
   n147 = "Constraint - dyn_r_flap_b (P)"(ComponentPose=@n146, Alpha=1, ConstraintWeights_0=1) {
    alphaBoolBlend {blendOption=Linear}
    boneToModify {boneName=dyn_r_flap_b}
    constraintSetup [{targetBone {boneName=thigh_r}}]
    constraintWeights [1]
   }
   n148 = "Constraint - dyn_r_flap_a_02 (P)"(ComponentPose=@n147, Alpha=1, ConstraintWeights_0=1) {
    alphaBoolBlend {blendOption=Linear}
    boneToModify {boneName=dyn_r_flap_a_02}
    constraintSetup [{targetBone {boneName=thigh_r}}]
    constraintWeights [1]
   }
   n149 = "Constraint - dyn_l_flap_b_02 (P)"(ComponentPose=@n148, Alpha=1, ConstraintWeights_0=1) {
    alphaBoolBlend {blendOption=Linear}
    boneToModify {boneName=dyn_l_flap_b_02}
    constraintSetup [{targetBone {boneName=thigh_l}}]
    constraintWeights [1]
   }
   n150 = "Constraint - dyn_l_flap_a_02 (P)"(ComponentPose=@n149, Alpha=1, ConstraintWeights_0=1) {
    alphaBoolBlend {blendOption=Linear}
    boneToModify {boneName=dyn_l_flap_a_02}
    constraintSetup [{targetBone {boneName=thigh_l}}]
    constraintWeights [1]
   }
   n151 = "Constraint - dyn_bk_flap_l_02 (PPPP)"!(
    ComponentPose=@n150
    Alpha=1
    ConstraintWeights_0=0.0176
    ConstraintWeights_1:float=0.6335
    ConstraintWeights_2:float=0.0745
    ConstraintWeights_3:float=0.2743
   ) -> (Pose) {
    alphaBoolBlend {blendOption=Linear}
    boneToModify {boneName=dyn_bk_flap_l_02}
    constraintSetup [
     {targetBone {boneName=thigh_twist_01_r}}
     {targetBone {boneName=thigh_twist_01_l}}
     {targetBone {boneName=thigh_l}}
     {targetBone {boneName=pelvis}}
    ]
    constraintWeights [1,1,1,1]
   }
   n152 = "Constraint - dyn_bk_flap_02 (PPPPPPPP)"!(
    ComponentPose=@n151
    Alpha=1
    ConstraintWeights_0=0.0389
    ConstraintWeights_1:float=0.0607
    ConstraintWeights_2:float=0.0255
    ConstraintWeights_3:float=0.0894
    ConstraintWeights_4:float=0.0386
    ConstraintWeights_5:float=0.14
    ConstraintWeights_6:float=0.4244
    ConstraintWeights_7:float=0.1826
   ) -> (Pose) {
    alphaBoolBlend {blendOption=Linear}
    boneToModify {boneName=dyn_bk_flap_02}
    constraintSetup [
     {targetBone {boneName=dyn_bk_flap_r_01}}
     {targetBone {boneName=dyn_bk_flap_r_03}}
     {targetBone {boneName=dyn_bk_flap_r_02}}
     {targetBone {boneName=dyn_bk_flap_l_03}}
     {targetBone {boneName=dyn_bk_flap_l_01}}
     {targetBone {boneName=pelvis}}
     {targetBone {boneName=dyn_bk_flap_02}}
     {targetBone {boneName=dyn_bk_flap_01}}
    ]
    constraintWeights [1,1,1,1,1,1,1,1]
   }
   n153 = "Constraint - dyn_bk_flap_r_02 (PPPP)"!(
    ComponentPose=@n152
    Alpha=1
    ConstraintWeights_0=0.076
    ConstraintWeights_1:float=0.6115
    ConstraintWeights_2:float=0.0213
    ConstraintWeights_3:float=0.2912
   ) -> (Pose) {
    alphaBoolBlend {blendOption=Linear}
    boneToModify {boneName=dyn_bk_flap_r_02}
    constraintSetup [
     {targetBone {boneName=thigh_r}}
     {targetBone {boneName=thigh_twist_01_r}}
     {targetBone {boneName=thigh_twist_01_l}}
     {targetBone {boneName=pelvis}}
    ]
    constraintWeights [1,1,1,1]
   }
   Save cached pose 'RBAN'(Pose=Blend Poses by bool(
     BlendPose_0=Copy Pose From Mesh() {bUseAttachedParent=true}
     BlendPose_1=Component To Local(ComponentPose=@n153)
     bActiveValue=Get bRigidBodyEnabled()
    ) {blendType=Linear}) {cacheName=RBAN}
   n154 = Control Rig(Source=Use cached pose 'anim physics'()) {
    alphaBoolBlend {blendOption=Linear}
    controlRigAssetReference {blueprintRigClass=/BRCosmetics/Characters/Player/Male/Medium/Bodies/M_MED_Pumpkin_Punk/Meshes/M_MED_Pumpkin_Punk_CtrlRig/M_MED_Pumpkin_Punk_T_CtrlRig.M_MED_Pumpkin_Punk_T_CtrlRig_C}
   }
   n155 = Control Rig(Source=Use cached pose 'anim physics'()) {
    alphaBoolBlend {blendOption=Linear}
    controlRigAssetReference {blueprintRigClass=/BRCosmetics/Characters/Player/Male/Medium/Bodies/M_MED_Pumpkin_Punk/Meshes/M_MED_Pumpkin_Punk_CtrlRig/M_MED_Pumpkin_Punk_CtrlRig.M_MED_Pumpkin_Punk_CtrlRig_C}
   }
   Save cached pose 'anim physics'(Pose=Blend Poses by bool(BlendPose_0=@n154, BlendPose_1=@n155, BlendTime_0=0.2, BlendTime_1=0.2, bActiveValue=Get bIsTargeting()) {blendType=Linear}) {cacheName=anim physics}
   Get ThighR_Roll_FPSCorrection() -> (ThighR_Roll_FPSCorrection:double)
   Get ThighLAlpha()
  ]}
 EventGraph {nodes [
   n1 = Event BlueprintUpdateAnimation()
   > Is Valid(InputObject=Get SkeletalMesh() -> (SkeletalMesh:Skeletal Mesh Component Object Reference))
   > Set FPS_correction(FPS_correction:double=MapRangeClamped(Value=@n1.DeltaTimeX, InRangeA=0.03, InRangeB=0.017, OutRangeA=0.5)) -> (Output_Get:double)
   > n2 = Sequence!(execute) -> (then_0, then_1, then_2:Exec, then_3:Exec, then_4:Exec)
   n3 = Get SkeletalMesh()
   n4 = Get bIsTargeting()
   n5 = Set TargetingHeadUpAlpha(
    execute=@n2.then_2
    TargetingHeadUpAlpha:double=float * float(
     A=GetAlphaFromReferenceTransformDelta(SkeletalMeshComponent=@n3, BoneName=neck_01, BaseName=root, InputRange=(X=5,Y=-30), TwistAxis="1,0,0")
     B="To Float (Boolean)"(InBool=@n4)
    ) {is=title K2Node_PromotableOperator}
   ) -> (Output_Get:double)
   > Set BackblingOnAndTargetingUp(BackblingOnAndTargetingUp:bool="float >= float"(A=@n5, B=0.01)) -> (Output_Get:bool)
   n6 = Get SkeletalMesh()
   n7 = Get bIsCrouching()
   n8 = GetAlphaFromReferenceTransformDelta(SkeletalMeshComponent=@n6, BoneName=thigh_l, BaseName=pelvis, InputRange=(X=-10,Y=6), TwistAxis="1,0,0")
   n9 = "To Float (Boolean)"(InBool=NOT Boolean(A=Get bIsInFrontEnd()))
   n10 = "Clamp (Float)"(
    Value=float + float!(
     A=float * float(A=@n8, B=@n9)
     B=GetAlphaFromReferenceTransformDelta(SkeletalMeshComponent=@n6, BoneName=thigh_l, BaseName=pelvis, DeltaComponentType=EulerX, InputRange=(X=25,Y=60), TwistAxis="1,0,0")
     C:double=SelectFloat(A=0.25, B=0, bPickA=@n7)
     D:double
     E:double
    ) -> (ReturnValue)
    Min=0
    Max=1
   )
   n11 = Dot Product!(
    A:Vector="Normalize 2D (Vector)"!(A:Vector=Get PawnVelocity(), Tolerance:float=0.0001) -> (ReturnValue:Vector)
    B:Vector=GetActorForwardVector!(self:Actor Object Reference=TryGetPawnOwner()) -> (ReturnValue:Vector)
   ) -> (ReturnValue:double)
   > Set Jacket_LefttSide_SideDirection_Fix(Jacket_LefttSide_SideDirection_Fix:double="Clamp (Float)"(
     Value=Power!(Base:double=@n10, Exp:double=0.5) -> (ReturnValue:double) {is=title K2Node_CallFunction}
     Min=0
     Max=MapRangeClamped(Value=@n11, InRangeA=0.6, InRangeB=0)
    )) -> (Output_Get:double)
   n12 = Get FPS_correction() -> (FPS_correction:double)
   n13 = Get SkeletalMesh()
   n14 = Get bIsCrouching()
   n15 = GetAlphaFromReferenceTransformDelta(SkeletalMeshComponent=@n13, BoneName=thigh_l, BaseName=pelvis, InputRange=(X=-10,Y=6), TwistAxis="1,0,0")
   n16 = "To Float (Boolean)"(InBool=NOT Boolean(A=Get bIsInFrontEnd()))
   n17 = "Clamp (Float)"(
    Value=float + float(
     A=float * float(A=@n15, B=@n16)
     B=GetAlphaFromReferenceTransformDelta(SkeletalMeshComponent=@n13, BoneName=thigh_l, BaseName=pelvis, DeltaComponentType=EulerX, InputRange=(X=25,Y=60), TwistAxis="1,0,0")
     C=SelectFloat(A=0.25, B=0, bPickA=@n14)
    )
    Min=0
    Max=1
   )
   > Set Jacket_RightSide_SideDirection_Fix(Jacket_RightSide_SideDirection_Fix:double=float + float!(A=Power(Base=@n17, Exp=0.5), B=MapRangeClamped(Value=@n12, InRangeA=1, InRangeB=0.3)) -> (ReturnValue)) -> (Output_Get:double)
   n18 = Get SkeletalMesh()
   n19 = NOT Boolean(A=Get bIsInFrontEnd())
   n20 = float + float!(
    A=float * float(
     A=GetAlphaFromReferenceTransformDelta(SkeletalMeshComponent=@n18, BoneName=thigh_l, BaseName=pelvis, InputRange=(X=-10,Y=6), TwistAxis="1,0,0")
     B="To Float (Boolean)"(InBool=@n19)
    )
    B=GetAlphaFromReferenceTransformDelta(SkeletalMeshComponent=@n18, BoneName=thigh_l, BaseName=pelvis, DeltaComponentType=EulerX, InputRange=(X=25,Y=60), TwistAxis="1,0,0")
    C:double=SelectFloat(A=0.25, B=0, bPickA=Get bIsCrouching())
    D:double
    E:double
   ) -> (ReturnValue)
   Set Jacket_LeftSide_Corection(execute=@n2.then_3, Jacket_LeftSide_Corection:float=Power(Base="Clamp (Float)"(Value=@n20, Min=0, Max=1), Exp=0.5)) -> (Output_Get:float)
   n21 = Get SkeletalMesh()
   > n22 = Set thighLAlpha(thighLAlpha:double=GetAlphaFromReferenceTransformDelta(SkeletalMeshComponent=@n21, BoneName=thigh_l, BaseName=pelvis, InputRange=(X=-20,Y=-90), OutputRange=(X=0,Y=0.7), TwistAxis="1,0,0")) -> (Output_Get:double)
   > n23 = Set thighRAlpha(thighRAlpha:double=GetAlphaFromReferenceTransformDelta(SkeletalMeshComponent=@n21, BoneName=thigh_r, BaseName=pelvis, InputRange=(X=0,Y=-90), OutputRange=(X=0,Y=0.7), TwistAxis="1,0,1")) -> (Output_Get:double)
   > Set ThighSumAlpha(ThighSumAlpha:double=float / float(A=float + float!(A=@n22, B=@n23) -> (ReturnValue), B=2)) -> (Output_Get:double)
   n24 = AND Boolean(A=Get bIsCrouching(), B=Get bIsMoving2D())
   > Set Thigh RCrouch Alpha(Thigh RCrouch Alpha:double=float * float(A=@n23, B="To Float (Boolean)"(InBool=@n24))) -> (Output_Get:double)
   n25 = Get ThighRAlpha()
   n26 = "To Float (Boolean)"(InBool=NOT Boolean(A=Get bIsPlayingEmote()))
   n27 = Get FPS_correction()
   > Set ThighR_FPSCorrection(ThighR_FPSCorrection:double=float * float!(A=MapRangeClamped(Value=@n25, InRangeA=0, InRangeB=0.7), B=MapRangeClamped(Value=@n27, InRangeA=1, InRangeB=0.5), C:double=6, D:double=@n26) -> (ReturnValue)) -> (Output_Get:double)
   n28 = Get FPS_correction()
   n29 = Get ThighLAlpha()
   > Set ThighL_FPSCorrection(ThighL_FPSCorrection:double=float * float(A=MapRangeClamped(Value=@n29, InRangeA=0, InRangeB=0.7), B=MapRangeClamped(Value=@n28, InRangeA=1, InRangeB=0.5), C=4, D=@n26)) -> (Output_Get:double)
   n30 = Get FPS_correction()
   > Set ThighR_Roll_FPSCorrection(ThighR_Roll_FPSCorrection:double=float * float!(
     A=GetAlphaFromReferenceTransformDelta(
      SkeletalMeshComponent=@n21
      BoneName=thigh_r
      BaseName=pelvis
      DeltaComponentType=EulerX
      InputRange=(X=-80,Y=10)
      OutputRange=(X=1,Y=0)
      TwistAxis="1,0,0"
     )
     B=MapRangeClamped(Value=@n30, InRangeA=1, InRangeB=0.5)
     C=@n26
    ) -> (ReturnValue)) -> (Output_Get:double)
   > Set ThighSumFPSCorrection(ThighSumFPSCorrection:double=float * float!(A=Get ThighSumAlpha(), B=Get FPS_correction()) -> (ReturnValue)) -> (Output_Get:double)
   n31 = Get IsStandingMoving() -> (IsStandingMoving:bool)
   > Set StandingMoving_FPSCorrestion(StandingMoving_FPSCorrestion:double=float * float(A=Get FPS_correction(), B="To Float (Boolean)"(InBool=@n31))) -> (Output_Get:double)
   > Set ThighLJacketCorrection(ThighLJacketCorrection:double=MapRangeClamped(Value=Get ThighLAlpha(), InRangeA=0, InRangeB=0.05)) -> (Output_Get:double)
   Set HoodConstraintAlpha(execute=@n2.then_0, HoodConstraintAlpha:double=SelectFloat(A=0.95, B=0, bPickA=Get bIsMoving2D())) -> (Output_Get:double)
   n32 = Get bIsMoving2D()
   Set IsNotCrouchMoving(execute=@n2.then_1, IsNotCrouchMoving:bool=AND Boolean(A=Get bIsCrouching(), B=NOT Boolean(A=@n32))) -> (Output_Get:bool)
   > Set Is CrouchMoving(Is CrouchMoving:bool=AND Boolean(A=Get bIsCrouching(), B=Get bIsMoving2D())) -> (Output_Get:bool)
   n33 = Get bIsCrouching()
   > Set IsStandingMoving(IsStandingMoving:bool=AND Boolean(A=Get bIsMoving2D(), B=NOT Boolean(A=@n33))) -> (Output_Get:bool)
   > Set BackblingOnAndTargeting(BackblingOnAndTargeting:bool=AND Boolean(A=Get IsWearingBackbling(), B=Get bIsTargeting())) -> (Output_Get:bool)
   Event OnCharacterCustomizationLoaded()
   > Set IsWearingBackbling(IsWearingBackbling:bool=CheckForBackbling()) -> (Output_Get:bool)
   > Set SkeletalMesh(SkeletalMesh:Skeletal Mesh Component Object Reference=GetOwnerSkeletalMeshComponent()) -> (Output_Get:Skeletal Mesh Component Object Reference)
  ]}
]
variables [
 BackblingOnAndTargeting {type=boolean}
 BackblingOnAndTargetingUp {type=boolean}
 FPS_correction {type=number}
 HoodConstraintAlpha {type=number}
 Is CrouchMoving {type=boolean}
 IsNotCrouchMoving {type=boolean}
 IsStandingMoving {type=boolean}
 IsWearingBackbling {type=boolean}
 Jacket_LeftSide_Corection {type=number}
 Jacket_LefttSide_SideDirection_Fix {type=number}
 Jacket_RightSide_SideDirection_Fix {type=number}
 SkeletalMesh {type=SkeletalMeshComponent}
 StandingMoving_FPSCorrestion {type=number}
 TargetingHeadUpAlpha {type=number}
 Thigh RCrouch Alpha {type=number}
 ThighL_FPSCorrection {type=number}
 ThighLAlpha {type=number}
 ThighLJacketCorrection {type=number}
 ThighR_FPSCorrection {type=number}
 ThighR_Roll_FPSCorrection {type=number}
 ThighRAlpha {type=number}
 ThighSumAlpha {type=number}
 ThighSumFPSCorrection {type=number}
]
