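# Physics asset for the Pumpkin Punk male medium head cosmetic, providing bodies for the head, stem and a three-bone leaf chain.
# Capsule bodies are linked in series from head through dyn_stem_05 to the three dyn_leaf bones.
# The stem joint has very small angular limit values, while each leaf joint locks swing1 and twist, so the leaf bends only on the swing2 axis.
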PhysicsAsset /BRCosmetics/Characters/Player/Male/Medium/Bodies/M_MED_Pumpkin_Punk/Meshes/Parts/Male_Medium_Pumpkin_Punk_Head_Physics
bodies {
 dyn_leaf_01 [{shapeType=Capsule, length=1, radius=1, shapeName=dyn_leaf_01_capsule}]
 dyn_leaf_02 [{shapeType=Capsule, length=1, radius=1, shapeName=dyn_leaf_02_capsule}]
 dyn_leaf_03 [{shapeType=Capsule, length=1, radius=1, shapeName=dyn_leaf_03_capsule}]
 dyn_stem_05 [{shapeType=Capsule, length=1.6997, radius=0.5024, center=<1.022 0 0>, rotation=<rot 90 0.0395 0.0395>}]
 head [{shapeType=Capsule, length=4.0357, radius=16.3008, center=<-0.5591 0.5617 0.0101>}]
}
constraints [
 {bone1Name=dyn_stem_05, bone2Name=head, swing1LimitDegrees=1, swing2LimitDegrees=1, twistLimitDegrees=1}
 {
  bone1Name=dyn_leaf_01
  bone2Name=dyn_stem_05
  swing1LimitDegrees=45
  swing1Motion=Locked
  swing2LimitDegrees=10
  twistLimitDegrees=15
  twistMotion=Locked
 }
 {
  bone1Name=dyn_leaf_02
  bone2Name=dyn_leaf_01
  swing1LimitDegrees=45
  swing1Motion=Locked
  swing2LimitDegrees=10
  twistLimitDegrees=15
  twistMotion=Locked
 }
 {
  bone1Name=dyn_leaf_03
  bone2Name=dyn_leaf_02
  swing1LimitDegrees=45
  swing1Motion=Locked
  swing2LimitDegrees=10
  twistLimitDegrees=15
  twistMotion=Locked
 }
]
