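# Animation Blueprint for the Pumpkin Punk male medium character's head part, adding secondary physics motion to it.
# It copies the pose from the attached parent mesh, then runs a RigidBody node with the head's physics asset.
# Simulation runs in spine_05 bone space, with custom gravity and scaled, clamped component motion.
# Each update lowers the simulation alpha slightly while bIsTargeting is true and restores full strength otherwise.
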AnimBlueprint /BRCosmetics/Characters/Player/Male/Medium/Bodies/M_MED_Pumpkin_Punk/Meshes/Parts/Male_Medium_Pumpkin_Punk_Head_AnimBP
target_skeleton=&/Male_Medium_Pumpkin_Punk_Head_Skeleton
graphs [
 AnimGraph {nodes [
   n1 = RigidBody(
    ComponentPose=Local To Component(LocalPose=Copy Pose From Mesh() {bUseAttachedParent=true})
    Alpha=Get Targeting Alpha() -> (Targeting Alpha:double)
   ) {
    bClampLinearTranslationLimitToRefPose=true
    bOverrideWorldGravity=true
    overridePhysicsAsset=&/Male_Medium_Pumpkin_Punk_Head_Physics
    simulationSpace=BaseBoneSpace
    alphaBoolBlend {blendOption=Linear}
    baseBoneRef {boneName=spine_05}
    componentAppliedLinearAccClamp=<100 100 100>
    componentLinearAccScale=<0.3 0.3 0.3>
    componentLinearVelScale=<0.1 0.1 0.1>
    overrideWorldGravity {z=-980}
   }
   Output Pose(Result=Component To Local(ComponentPose=@n1)) {layerGroup=None}
  ]}
 EventGraph {nodes [
   Event BlueprintUpdateAnimation()
   > n1 = Branch(Condition=Get bIsTargeting())
   > Set Targeting Alpha(Targeting Alpha:double=0.7) -> (Output_Get:double)
   Set Targeting Alpha(execute=@n1.else, Targeting Alpha=1)
   TryGetPawnOwner()
  ]}
]
variables [Targeting Alpha {type=number}]
